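# Control Rig for the female medium chainmail armor body that procedurally corrects shirt, back-shoulder and necklace bones during movement.
# A four-way sequence handles swing, crouch, tactical sprint or jump, and necklace adjustments, using parent constraints, aims and transform offsets.
# Smoothed movement variables weight the offsets, with directional ones applied only while crouched.
# Spherical pose readers on thighs (crouch only) and right upper arm drive shirt and necklace corrections; Skydive is unused.
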ControlRigBlueprint /BRCosmetics/Characters/Player/Female/Medium/Bodies/F_MED_Chainmail_Armor/Meshes/F_MED_Chainmail_Armor_CtrlRig_Shirt
graphs [
 ContainedGraph {nodes [
   n1 = RigVMFunctionEntryNode()
   SequenceAggregate = &RigVMFunction_Sequence(ExecuteContext=@n1.ExecuteContext)
   SequenceAggregate_1 = &RigVMFunction_Sequence(ExecuteContext=@SequenceAggregate.B)
   SequenceAggregate_2 = &RigVMFunction_Sequence(ExecuteContext=@SequenceAggregate_1.B)
   SequenceAggregate_3 = &RigVMFunction_Sequence(ExecuteContext=@SequenceAggregate_2.B)
   RigVMFunctionReturnNode(A=@SequenceAggregate.A, B=@SequenceAggregate_1.A, C=@SequenceAggregate_2.A, D=@SequenceAggregate_3.A, E=@SequenceAggregate_3.B)
  ]}
 RigVMFunctionLibrary {links [], nodes []}
 RigVMModel {nodes [
   RigUnit_BeginExecution = &RigUnit_BeginExecution()
   SequenceAggregate = RigVMAggregateNode(ExecuteContext=@RigUnit_BeginExecution.ExecutePin)
   AlphaInterp = &RigVMFunction_AlphaInterp(InterpSpeedDecreasing=2, InterpSpeedIncreasing=2, Scale=1, bClampResult=False, Value=RigVMVariableNode() {variable=Swing}.Value)
   &RigUnit_ParentConstraint(
    AdvancedSettings=(RotationOrderForFilter=XZY)
    Child=(Type=Bone,Name="dyn_shirt_mid_l_03")
    Parents=((Item=(Type=Bone,Name="spine_02"),Weight=1))
    Weight=1
    Parents.0.Weight=@AlphaInterp.Result
    ExecutePin=@SequenceAggregate.A
   )
   > &RigUnit_ParentConstraint(
    AdvancedSettings=(RotationOrderForFilter=XZY)
    Child=(Type=Bone,Name="dyn_shirt_bk_l_03")
    Parents=((Item=(Type=Bone,Name="spine_02"),Weight=1))
    Weight=1
    Parents.0.Weight=@AlphaInterp.Result
   )
   > &RigUnit_ModifyTransforms(
    ItemToModify=((Item=(Type=Bone,Name="dyn_shirt_bk_l_MSTR"),Transform=(Rotation=(X=0,Y=0,Z=0,W=1),Translation=(X=0,Y=0,Z=0),Scale3D=(X=1,Y=1,Z=1))))
    Weight=@AlphaInterp.Result
    ItemToModify.0.Transform.Rotation=&RigVMFunction_MathQuaternionFromEuler(Euler=(X=0,Y=-20,Z=0)).Result
   )
   > &RigUnit_ModifyTransforms(
    ItemToModify=((Item=(Type=Bone,Name="dyn_shirt_mid_l_MSTR"),Transform=(Rotation=(X=0,Y=0,Z=0,W=1),Translation=(X=0,Y=0,Z=0),Scale3D=(X=1,Y=1,Z=1))))
    Weight=@AlphaInterp.Result
    ItemToModify.0.Transform.Rotation=&RigVMFunction_MathQuaternionFromEuler(Euler=(X=0,Y=-20,Z=0)).Result
   )
   n1 = &RigVMFunction_AlphaInterp(InterpSpeedDecreasing=2, InterpSpeedIncreasing=2, Scale=1, bClampResult=False, Value=RigVMVariableNode() {variable=Swing}.Value)
   > &RigUnit_ParentConstraint(
    AdvancedSettings=(RotationOrderForFilter=XZY)
    Child=(Type=Bone,Name="dyn_shirt_mid_r_03")
    Parents=((Item=(Type=Bone,Name="spine_02"),Weight=1))
    Weight=0.5
    Parents.0.Weight=@n1.Result
   )
   > &RigUnit_ParentConstraint(
    AdvancedSettings=(RotationOrderForFilter=XZY)
    Child=(Type=Bone,Name="dyn_shirt_bk_r_03")
    Parents=((Item=(Type=Bone,Name="spine_02"),Weight=1))
    Weight=0.5
    Parents.0.Weight=@n1.Result
   )
   > &RigUnit_ModifyTransforms(
    ItemToModify=((Item=(Type=Bone,Name="dyn_shirt_fr_r_MSTR"),Transform=(Rotation=(X=0,Y=0,Z=0,W=1),Translation=(X=0,Y=0,Z=0),Scale3D=(X=1,Y=1,Z=1))))
    Weight=@n1.Result
    ItemToModify.0.Transform.Rotation=&RigVMFunction_MathQuaternionFromEuler(Euler=(X=0,Y=-15,Z=0)).Result
   )
   > &RigUnit_ModifyTransforms(
    ItemToModify=((Item=(Type=Bone,Name="dyn_shirt_bk_r_03"),Transform=(Rotation=(X=0,Y=0,Z=0,W=1),Translation=(X=0,Y=10,Z=20),Scale3D=(X=1,Y=1,Z=1))))
    Weight=@n1.Result
    ItemToModify.0.Transform.Rotation=&RigVMFunction_MathQuaternionFromEuler(Euler=(X=0,Y=-35,Z=0)).Result
   )
   > &RigUnit_ModifyTransforms(
    ItemToModify=((Item=(Type=Bone,Name="dyn_shirt_mid_r_01"),Transform=(Rotation=(X=0,Y=0,Z=0,W=1),Translation=(X=0,Y=0,Z=0),Scale3D=(X=1,Y=1,Z=1))))
    Weight=@n1.Result
    ItemToModify.0.Transform.Rotation=&RigVMFunction_MathQuaternionFromEuler(Euler=(X=0,Y=-20,Z=0)).Result
   )
   Branch_1_RigVMUnitNode = &RigVMFunction_ControlFlowBranch(ExecuteContext=@SequenceAggregate.B, Condition=RigVMVariableNode() {variable=Crouch}.Value)
   /Script/ControlRig.RigUnit_AimItem(
    DebugSettings=(Scale=10,WorldOffset=(Rotation=(X=0,Y=0,Z=0,W=1),Translation=(X=0,Y=0,Z=0),Scale3D=(X=1,Y=1,Z=1)))
    Item=(Type=Bone,Name="dyn_shirt_bk_l_MSTR")
    Primary=(Weight=1,Axis=(X=1,Y=0,Z=0),Target=(X=-74.5835,Y=-45.9119,Z=-3.5451),Kind=Location,Space=(Type=Bone,Name="pelvis"))
    Secondary=(Weight=0,Axis=(X=0,Y=0,Z=1),Target=(X=-102.1545,Y=-50.3476,Z=-60.987),Kind=Location,Space=(Type=Bone,Name="None"))
    Weight=1
    ExecutePin=@Branch_1_RigVMUnitNode.True
   ) {is=struct RigVMUnitNode}
   > &RigUnit_AimItem(
    DebugSettings=(Scale=10,WorldOffset=(Rotation=(X=0,Y=0,Z=0,W=1),Translation=(X=0,Y=0,Z=0),Scale3D=(X=1,Y=1,Z=1)))
    Item=(Type=Bone,Name="dyn_shirt_bk_l_MSTR")
    Primary=(Weight=1,Axis=(X=1,Y=0,Z=0),Target=(X=-74.5835,Y=-45.9119,Z=-3.5451),Kind=Location,Space=(Type=Bone,Name="pelvis"))
    Secondary=(Weight=0,Axis=(X=0,Y=0,Z=1),Target=(X=-102.1545,Y=-50.3476,Z=-60.987),Kind=Location,Space=(Type=Bone,Name="None"))
    Weight=1
   )
   > &RigUnit_AimItem(
    DebugSettings=(Scale=10,WorldOffset=(Rotation=(X=0,Y=0,Z=0,W=1),Translation=(X=0,Y=0,Z=0),Scale3D=(X=1,Y=1,Z=1)))
    Item=(Type=Bone,Name="dyn_shoulder_r_bk")
    Primary=(Weight=1,Axis=(X=1,Y=0,Z=0),Target=(X=0,Y=-10,Z=5),Kind=Location,Space=(Type=Bone,Name="spine_05"))
    Secondary=(Weight=0,Axis=(X=0,Y=0,Z=1),Target=(X=-102.1545,Y=-50.3476,Z=-60.987),Kind=Location,Space=(Type=Bone,Name="None"))
    Weight=0.5
   )
   > &RigUnit_ParentConstraint(
    AdvancedSettings=(RotationOrderForFilter=XZY)
    Child=(Type=Bone,Name="dyn_shirt_bk_r_04")
    Parents=((Item=(Type=Bone,Name="pelvis"),Weight=1))
    Weight=0.5
   )
   > &RigUnit_ParentConstraint(
    AdvancedSettings=(RotationOrderForFilter=XZY)
    Child=(Type=Bone,Name="dyn_shirt_bk_l_04")
    Parents=((Item=(Type=Bone,Name="pelvis"),Weight=1))
    Weight=0.5
   )
   > &RigUnit_ParentConstraint(
    AdvancedSettings=(RotationOrderForFilter=XZY)
    Child=(Type=Bone,Name="dyn_shoulder_r_bk")
    Parents="((Item=(Type=Bone,Name=\"spine_02\"),Weight=0),(Item=(Type=Bone,Name=\"spine_04\"),Weight=0.0466),(Item=(Type=Bone,Name=\"spine_03\"),Weight=0.0003),(Item=(Type=Bone,Name=\"spine_05\"),Weight=0.1514),(Item=(Type=Bone,Name=\"clavicle_l\"),Weight=0.0023),(Item=(Type=B ...(truncated)"
    Weight=0.3
   )
   > &RigUnit_ParentConstraint(
    AdvancedSettings=(RotationOrderForFilter=XZY)
    Child=(Type=Bone,Name="dyn_shoulder_l_bk")
    Filter=(TranslationFilter=(bX=false,bY=false,bZ=false),RotationFilter=(bX=True,bY=True,bZ=True),ScaleFilter=(bX=True,bY=True,bZ=True))
    Parents="((Item=(Type=Bone,Name=\"upperarm_twist_01_l\"),Weight=0.3247),(Item=(Type=Bone,Name=\"spine_04\"),Weight=0.0563),(Item=(Type=Bone,Name=\"spine_05\"),Weight=0.1149),(Item=(Type=Bone,Name=\"clavicle_l\"),Weight=0.4799),(Item=(Type=Bone,Name=\"dyn_shirt_mid_l_01\"),We ...(truncated)"
    Weight=0.5
   )
   n2 = &RigVMFunction_AlphaInterp(InterpSpeedDecreasing=7, InterpSpeedIncreasing=5, Scale=1, bClampResult=False, Value=RigVMVariableNode() {variable=Left}.Value)
   > &RigUnit_ModifyTransforms(
    ItemToModify=((Item=(Type=Bone,Name="dyn_shoulder_l_bk"),Transform=(Rotation=(X=0,Y=0,Z=0,W=1),Translation=(X=0,Y=-2,Z=0),Scale3D=(X=1,Y=1,Z=1))))
    Weight=@n2.Result
    ItemToModify.0.Transform.Rotation=&RigVMFunction_MathQuaternionFromEuler(Euler=(X=-15,Y=0,Z=-20)).Result
   )
   > &RigUnit_ModifyTransforms(
    ItemToModify=((Item=(Type=Bone,Name="dyn_shoulder_r_bk"),Transform=(Rotation=(X=0,Y=0,Z=0,W=1),Translation=(X=0,Y=-3,Z=0),Scale3D=(X=1,Y=1,Z=1))))
    Weight=@n2.Result
    ItemToModify.0.Transform.Rotation=&RigVMFunction_MathQuaternionFromEuler(Euler=(X=0,Y=-30,Z=-20), RotationOrder=XYZ).Result
   )
   n3 = &RigVMFunction_AlphaInterp(InterpSpeedDecreasing=7, InterpSpeedIncreasing=5, Scale=1, bClampResult=False, Value=RigVMVariableNode() {variable=Right}.Value)
   > &RigUnit_ModifyTransforms(
    ItemToModify=((Item=(Type=Bone,Name="dyn_shoulder_l_bk"),Transform=(Rotation=(X=0,Y=0,Z=0,W=1),Translation=(X=0,Y=-2,Z=0),Scale3D=(X=1,Y=1,Z=1))))
    Weight=@n3.Result
    ItemToModify.0.Transform.Rotation=&RigVMFunction_MathQuaternionFromEuler(Euler=(X=-10,Y=0,Z=-20)).Result
   )
   > &RigUnit_ModifyTransforms(
    ItemToModify=((Item=(Type=Bone,Name="dyn_shoulder_r_bk"),Transform=(Rotation=(X=0,Y=0,Z=0,W=1),Translation=(X=0,Y=-3,Z=0),Scale3D=(X=1,Y=1,Z=1))))
    Weight=@n3.Result
    ItemToModify.0.Transform.Rotation=&RigVMFunction_MathQuaternionFromEuler(Euler=(X=0,Y=-30,Z=-15), RotationOrder=XZY).Result
   )
   n4 = &RigVMFunction_AlphaInterp(InterpSpeedDecreasing=7, InterpSpeedIncreasing=5, Scale=1, bClampResult=False, Value=RigVMVariableNode() {variable=Back}.Value)
   > &RigUnit_ModifyTransforms(
    ItemToModify=((Item=(Type=Bone,Name="dyn_shoulder_l_bk"),Transform=(Rotation=(X=0,Y=0,Z=0,W=1),Translation=(X=0,Y=-1,Z=0),Scale3D=(X=1,Y=1,Z=1))))
    Weight=@n4.Result
    ItemToModify.0.Transform.Rotation=&RigVMFunction_MathQuaternionFromEuler(Euler=(X=5,Y=0,Z=0)).Result
   )
   > &RigUnit_ModifyTransforms(
    ItemToModify=((Item=(Type=Bone,Name="dyn_shoulder_r_bk"),Transform=(Rotation=(X=0,Y=0,Z=0,W=1),Translation=(X=0,Y=-2,Z=0),Scale3D=(X=1,Y=1,Z=1))))
    Weight=@n4.Result
    ItemToModify.0.Transform.Rotation=&RigVMFunction_MathQuaternionFromEuler(Euler=(X=0,Y=0,Z=-2), RotationOrder=XZY).Result
   )
   n5 = &RigVMFunction_AlphaInterp(InterpSpeedDecreasing=4, InterpSpeedIncreasing=1, Scale=1, bClampResult=False, Value=RigVMVariableNode() {variable=Forward}.Value)
   > &RigUnit_AimItem(
    DebugSettings=(Scale=10,WorldOffset=(Rotation=(X=0,Y=0,Z=0,W=1),Translation=(X=0,Y=0,Z=0),Scale3D=(X=1,Y=1,Z=1)))
    Item=(Type=Bone,Name="dyn_shirt_bk_l_MSTR")
    Primary=(Weight=1,Axis=(X=1,Y=0,Z=0),Target=(X=-104.1113,Y=-96.0145,Z=-20.572),Kind=Location,Space=(Type=Bone,Name="pelvis"))
    Secondary=(Weight=0,Axis=(X=0,Y=0,Z=1),Target=(X=-102.1545,Y=-50.3476,Z=-60.987),Kind=Location,Space=(Type=Bone,Name="None"))
    Weight=@n5.Result
   )
   > &RigUnit_AimItem(
    DebugSettings=(Scale=10,WorldOffset=(Rotation=(X=0,Y=0,Z=0,W=1),Translation=(X=0,Y=0,Z=0),Scale3D=(X=1,Y=1,Z=1)))
    Item=(Type=Bone,Name="dyn_shirt_mid_l_MSTR")
    Primary=(Weight=1,Axis=(X=1,Y=0,Z=0),Target=(X=-104.174,Y=-41.316,Z=-72.2932),Kind=Location,Space=(Type=Bone,Name="spine_03"))
    Secondary=(Weight=0,Axis=(X=0,Y=0,Z=1),Target=(X=-102.1545,Y=-50.3476,Z=-60.987),Kind=Location,Space=(Type=Bone,Name="None"))
    Weight=@n5.Result
   )
   > &RigUnit_AimItem(
    DebugSettings=(Scale=10,WorldOffset=(Rotation=(X=0,Y=0,Z=0,W=1),Translation=(X=0,Y=0,Z=0),Scale3D=(X=1,Y=1,Z=1)))
    Item=(Type=Bone,Name="dyn_shoulder_r_bk")
    Primary=(Weight=1,Axis=(X=1,Y=0,Z=0),Target=(X=0,Y=-30,Z=-30),Kind=Location,Space=(Type=Bone,Name="spine_03"))
    Secondary=(Weight=0,Axis=(X=0,Y=0,Z=1),Target=(X=-102.1545,Y=-50.3476,Z=-60.987),Kind=Location,Space=(Type=Bone,Name="None"))
    Weight=@n5.Result
   )
   > &RigUnit_ModifyTransforms(
    ItemToModify=((Item=(Type=Bone,Name="dyn_shoulder_r_bk"),Transform=(Rotation=(X=0,Y=0,Z=0,W=1),Translation=(X=0,Y=-3.5,Z=0),Scale3D=(X=1,Y=1,Z=1))))
    Weight=@n5.Result
    ItemToModify.0.Transform.Rotation=&RigVMFunction_MathQuaternionFromEuler(Euler=(X=0,Y=0,Z=0), RotationOrder=XZY).Result
   )
   > n6 = &RigUnit_SphericalPoseReader(
    ActiveRegionScaleFactors=(PositiveWidth=3.5,NegativeWidth=0.5,PositiveHeight=3,NegativeHeight=3)
    ActiveRegionSize=0.1
    Debug=(bDrawDebug=true,bDraw2D=false,bDrawLocalAxes=false,DebugScale=25,DebugSegments=20,DebugThickness=0.25)
    DriverAxis=(X=-1,Y=0,Z=0)
    DriverItem=(Type=Bone,Name="thigh_l")
    FalloffRegionScaleFactors=(PositiveWidth=1,NegativeWidth=1,PositiveHeight=1,NegativeHeight=1)
    FalloffSize=0.45
    OptionalParentItem=(Type=Bone,Name="pelvis")
    RotationOffset=(X=90,Y=0,Z=0)
   )
   > &RigUnit_ModifyTransforms(
    ItemToModify=((Item=(Type=Bone,Name="dyn_shirt_fr_r_04"),Transform=(Rotation=(X=0,Y=0,Z=0,W=1),Translation=(X=0,Y=0,Z=-2),Scale3D=(X=1,Y=1,Z=1))))
    ItemToModify.0.Transform.Rotation=&RigVMFunction_MathQuaternionFromEuler(Euler=(X=0,Y=0,Z=5)).Result
    Weight=@n6.OutputParam
   )
   > &RigUnit_ModifyTransforms(
    ItemToModify=((Item=(Type=Bone,Name="dyn_shirt_mid_l_MSTR"),Transform=(Rotation=(X=0,Y=0,Z=0,W=1),Translation=(X=0,Y=0,Z=0),Scale3D=(X=1,Y=1,Z=1))))
    ItemToModify.0.Transform.Rotation=&RigVMFunction_MathQuaternionFromEuler(Euler=(X=0,Y=0,Z=0)).Result
    Weight=@n6.OutputParam
   )
   > &RigUnit_ModifyTransforms(
    ItemToModify=((Item=(Type=Bone,Name="dyn_shirt_bk_l_MSTR"),Transform=(Rotation=(X=0,Y=0,Z=0,W=1),Translation=(X=0,Y=0,Z=0),Scale3D=(X=1,Y=1,Z=1))))
    ItemToModify.0.Transform.Rotation=&RigVMFunction_MathQuaternionFromEuler(Euler=(X=0,Y=0,Z=0)).Result
    Weight=@n6.OutputParam
   )
   > n7 = &RigUnit_SphericalPoseReader(
    ActiveRegionScaleFactors=(PositiveWidth=3.5,NegativeWidth=0.5,PositiveHeight=3,NegativeHeight=3)
    ActiveRegionSize=0.1
    Debug=(bDrawDebug=false,bDraw2D=false,bDrawLocalAxes=false,DebugScale=25,DebugSegments=20,DebugThickness=0.25)
    DriverAxis=(X=1,Y=0,Z=0)
    DriverItem=(Type=Bone,Name="thigh_r")
    FalloffRegionScaleFactors=(PositiveWidth=1,NegativeWidth=1,PositiveHeight=1,NegativeHeight=1)
    FalloffSize=0.45
    FlipHeightScaling=false
    FlipWidthScaling=true
    OptionalParentItem=(Type=Bone,Name="pelvis")
    RotationOffset=(X=-70,Y=0,Z=20)
   )
   > &RigUnit_ModifyTransforms(
    ItemToModify=((Item=(Type=Bone,Name="dyn_shirt_fr_r_MSTR"),Transform=(Rotation=(X=0,Y=0,Z=0,W=1),Translation=(X=0,Y=1.5,Z=0),Scale3D=(X=1,Y=1,Z=1))))
    ItemToModify.0.Transform.Rotation=&RigVMFunction_MathQuaternionFromEuler(Euler=(X=0,Y=0,Z=0)).Result
    Weight=@n7.OutputParam
   )
   > &RigUnit_ModifyTransforms(
    ItemToModify=((Item=(Type=Bone,Name="dyn_shirt_mid_r_03"),Transform=(Rotation=(X=0,Y=0,Z=0,W=1),Translation=(X=0,Y=0,Z=0),Scale3D=(X=1,Y=1,Z=1))))
    ItemToModify.0.Transform.Rotation=&RigVMFunction_MathQuaternionFromEuler(Euler=(X=0,Y=0,Z=0)).Result
    Weight=@n7.OutputParam
   )
   > &RigUnit_ModifyTransforms(
    ItemToModify=((Item=(Type=Bone,Name="dyn_shirt_bk_r_MSTR"),Transform=(Rotation=(X=0,Y=0,Z=0,W=1),Translation=(X=0,Y=0,Z=0),Scale3D=(X=1,Y=1,Z=1))))
    ItemToModify.0.Transform.Rotation=&RigVMFunction_MathQuaternionFromEuler(Euler=(X=0,Y=0,Z=0)).Result
    Weight=@n7.OutputParam
   )
   Branch_1_1_RigVMUnitNode = &RigVMFunction_ControlFlowBranch(ExecuteContext=@SequenceAggregate.C, Condition=RigVMVariableNode() {variable=Tact_Sprint}.Value)
   n8 = &RigVMFunction_AlphaInterp(InterpSpeedDecreasing=5, InterpSpeedIncreasing=1, Scale=1, bClampResult=False, Value=RigVMVariableNode() {variable=Jumps}.Value)
   &RigUnit_ModifyTransforms(
    ItemToModify=((Item=(Type=Bone,Name="dyn_shirt_bk_l_MSTR"),Transform=(Rotation=(X=0,Y=0,Z=0,W=1),Translation=(X=0,Y=0,Z=0),Scale3D=(X=1,Y=1,Z=1))))
    Weight=@n8.Result
    ExecutePin=@Branch_1_1_RigVMUnitNode.False
    ItemToModify.0.Transform.Rotation=&RigVMFunction_MathQuaternionFromEuler(Euler=(X=0,Y=-10,Z=0)).Result
   )
   > &RigUnit_ModifyTransforms(
    ItemToModify=((Item=(Type=Bone,Name="dyn_shirt_bk_r_MSTR"),Transform=(Rotation=(X=0,Y=0,Z=0,W=1),Translation=(X=0,Y=0,Z=0),Scale3D=(X=1,Y=1,Z=1))))
    Weight=@n8.Result
    ItemToModify.0.Transform.Rotation=&RigVMFunction_MathQuaternionFromEuler(Euler=(X=0,Y=-10,Z=0)).Result
   )
   &RigUnit_ModifyTransforms(
    ItemToModify=((Item=(Type=Bone,Name="dyn_shirt_fr_l_MSTR"),Transform=(Rotation=(X=0,Y=0,Z=0,W=1),Translation=(X=0,Y=0,Z=0),Scale3D=(X=1,Y=1,Z=1))))
    Weight=1
    ExecutePin=@Branch_1_1_RigVMUnitNode.True
    ItemToModify.0.Transform.Rotation=&RigVMFunction_MathQuaternionFromEuler(Euler=(X=0,Y=-10,Z=0)).Result
   )
   > &RigUnit_ModifyTransforms(
    ItemToModify=((Item=(Type=Bone,Name="dyn_shirt_bk_l_MSTR"),Transform=(Rotation=(X=0,Y=0,Z=0,W=1),Translation=(X=0,Y=0,Z=0),Scale3D=(X=1,Y=1,Z=1))))
    Weight=1
    ItemToModify.0.Transform.Rotation=&RigVMFunction_MathQuaternionFromEuler(Euler=(X=0,Y=-10,Z=0)).Result
   )
   > &RigUnit_ModifyTransforms(
    ItemToModify=((Item=(Type=Bone,Name="dyn_shirt_bk_r_MSTR"),Transform=(Rotation=(X=0,Y=0,Z=0,W=1),Translation=(X=0,Y=0,Z=0),Scale3D=(X=1,Y=1,Z=1))))
    Weight=1
    ItemToModify.0.Transform.Rotation=&RigVMFunction_MathQuaternionFromEuler(Euler=(X=0,Y=-30,Z=0)).Result
   )
   n9 = &RigUnit_SphericalPoseReader(
    ActiveRegionScaleFactors=(PositiveWidth=1,NegativeWidth=1,PositiveHeight=0.6,NegativeHeight=1)
    ActiveRegionSize=0.15
    Debug=(bDrawDebug=false,bDraw2D=false,bDrawLocalAxes=false,DebugScale=25,DebugSegments=20,DebugThickness=0.25)
    DriverAxis=(X=-1,Y=0,Z=0)
    DriverItem=(Type=Bone,Name="upperarm_r")
    FalloffRegionScaleFactors=(PositiveWidth=1,NegativeWidth=1,PositiveHeight=0.2,NegativeHeight=1)
    FalloffSize=0.45
    FlipHeightScaling=false
    FlipWidthScaling=true
    OptionalParentItem=(Type=Bone,Name="spine_04")
    RotationOffset=(X=-100,Y=30,Z=40)
    ExecutePin=@SequenceAggregate.D
   )
   > &RigUnit_ModifyTransforms(
    ItemToModify=((Item=(Type=Bone,Name="dyn_necklace_01"),Transform=(Rotation=(X=0,Y=0,Z=0,W=1),Translation=(X=0,Y=0,Z=0),Scale3D=(X=1,Y=1,Z=1))))
    ItemToModify.0.Transform.Rotation=&RigVMFunction_MathQuaternionFromEuler(Euler=(X=0,Y=0,Z=3)).Result
    Weight=@n9.OutputParam
   )
   > n10 = &RigUnit_SphericalPoseReader(
    ActiveRegionScaleFactors=(PositiveWidth=1.5,NegativeWidth=1.5,PositiveHeight=1.5,NegativeHeight=0.5)
    ActiveRegionSize=0.15
    Debug=(bDrawDebug=false,bDraw2D=false,bDrawLocalAxes=false,DebugScale=25,DebugSegments=20,DebugThickness=0.25)
    DriverAxis=(X=-1,Y=0,Z=0)
    DriverItem=(Type=Bone,Name="upperarm_r")
    FalloffRegionScaleFactors=(PositiveWidth=0.5,NegativeWidth=0.5,PositiveHeight=0.5,NegativeHeight=0.5)
    FalloffSize=0.45
    FlipHeightScaling=false
    FlipWidthScaling=true
    OptionalParentItem=(Type=Bone,Name="spine_04")
    RotationOffset=(X=0,Y=40,Z=45)
   )
   > &RigUnit_ModifyTransforms(
    ItemToModify=((Item=(Type=Bone,Name="dyn_necklace_01"),Transform=(Rotation=(X=0,Y=0,Z=0,W=1),Translation=(X=0,Y=0,Z=0),Scale3D=(X=1,Y=1,Z=1))))
    ItemToModify.0.Transform.Rotation=&RigVMFunction_MathQuaternionFromEuler(Euler=(X=0,Y=0,Z=2)).Result
    Weight=@n10.OutputParam
   )
  ]}
]
hierarchy [
 root {loc=<0 0 0>, rot=<0 0 0 1>}
 pelvis {parent=root, loc=<0 -0.5301 98.0571>, rot=<0.0127 0.707 -0.0127 -0.707>}
 spine_01 {parent=pelvis, loc=<6.1208 -0.9328 0>, rot=<0 0 0.034 -0.9994>}
 spine_02 {parent=spine_01, loc=<6.5996 -0.2095 0>, rot=<0 0 0.0238 -0.9997>}
 spine_03 {parent=spine_02, loc=<6.4914 -0.1237 0>, rot=<0 0 0.0149 -0.9999>}
 spine_04 {parent=spine_03, loc=<6.9598 0.0794 0>, rot=<0 0 -0.0132 -0.9999>}
 spine_05 {parent=spine_04, loc=<6.9987 0 0>, rot=<0 0 -0.0015 -1>}
 clavicle_l {parent=spine_05, loc=<5.4831 -0.1053 -3.2946>, rot=<0.0465 -0.7289 -0.009 -0.6829>}
 upperarm_l {parent=clavicle_l, loc=<9.8658 -1.2136 -0.4105>, rot=<0.0164 -0.3245 -0.0269 -0.9454>}
 lowerarm_l {parent=upperarm_l, loc=<26.465 -0.8345 0.259>, rot=<0.0036 0.0161 -0.2298 -0.9731>}
 hand_l {parent=lowerarm_l, loc=<22.7355 -0.3203 -0.6105>, rot=<-0.6784 -0.0223 0.0735 -0.7307>}
 thumb_01_l {parent=hand_l, loc=<1.3118 -0.3191 -2.459>, rot=<0.6546 -0.4042 0.0664 -0.6354>}
 thumb_02_l {parent=thumb_01_l, loc=<5.9336 -0.0448 0.0401>, rot=<-0.0037 0.0213 0.298 -0.9543>}
 thumb_03_l {parent=thumb_02_l, loc=<2.2223 0.3682 -0.0907>, rot=<0.0118 -0.013 0.0066 -0.9998>}
 index_metacarpal_l {parent=hand_l, loc=<1.6535 0.0183 -2.96>, rot=<0.0158 -0.0124 -0.0173 -0.9996>}
 index_01_l {parent=index_metacarpal_l, loc=<8.3008 0.0357 -0.0524>, rot=<-0.0044 0.0485 0.0924 -0.9945>}
 index_02_l {parent=index_01_l, loc=<3.3587 -0.3004 0.0452>, rot=<-0.002 0.0027 0.3225 -0.9466>}
 index_03_l {parent=index_02_l, loc=<2.5578 -0.123 -0.0252>, rot=<-0.002 0.0067 0.2558 -0.9667>}
 middle_metacarpal_l {parent=hand_l, loc=<2.1856 -0.1725 -0.5984>, rot=<0.006 0.0168 -0.0329 -0.9993>}
 middle_01_l {parent=middle_metacarpal_l, loc=<7.3275 -0.2447 0.0129>, rot=<-0.0208 0.0042 0.0858 -0.9961>}
 middle_02_l {parent=middle_01_l, loc=<4.0123 0.1401 0.1175>, rot=<-0.0218 0.0258 0.3512 -0.9357>}
 middle_03_l {parent=middle_02_l, loc=<2.6077 0.0534 0.1373>, rot=<0.0082 0.0188 0.1708 -0.9851>}
 ring_metacarpal_l {parent=hand_l, loc=<2.2074 -0.3625 1.3638>, rot=<-0.0284 0.0722 -0.0458 -0.9959>}
 ring_01_l {parent=ring_metacarpal_l, loc=<7.0295 -0.6197 -0.1446>, rot=<-0.0489 -0.0252 0.1892 -0.9804>}
 ring_02_l {parent=ring_01_l, loc=<3.6174 0.4647 -0.1027>, rot=<0 0 0.2806 -0.9598>}
 ring_03_l {parent=ring_02_l, loc=<2.4152 0.114 0.0199>, rot=<0 0 0.1189 -0.9929>}
 pinky_metacarpal_l {parent=hand_l, loc=<1.7616 -0.9753 3.3814>, rot=<-0.0397 0.0814 -0.069 -0.9935>}
 pinky_01_l {parent=pinky_metacarpal_l, loc=<7.1709 -0.4831 -0.1094>, rot=<-0.0654 -0.0035 0.2242 -0.9723>}
 pinky_02_l {parent=pinky_01_l, loc=<2.8274 0.0893 -0.1043>, rot=<0.0037 0.0011 0.2854 -0.9584>}
 pinky_03_l {parent=pinky_02_l, loc=<2.2046 0.0307 -0.0078>, rot=<-0.0002 0.0098 0.1445 -0.9895>}
 deform_wrist_dn_base_l {parent=hand_l, loc=<0 -1.5326 0>, rot=<0 0 0 1>}
 deform_wrist_dn_l {parent=deform_wrist_dn_base_l, loc=<0 0 0>, rot=<0 0 0 1>}
 deform_wrist_up_base_l {parent=hand_l, loc=<0 2.3009 0>, rot=<0 0 0 1>}
 deform_wrist_up_l {parent=deform_wrist_up_base_l, loc=<0 0 0>, rot=<0 0 0 1>}
 dyn_wrist_l {parent=hand_l, loc=<2.0835 2.157 -0.2958>, rot=<-0.247 -0.9637 -0.0209 0.0991>}
 lowerarm_twist_02_l {parent=lowerarm_l, loc=<10.6707 0 0>, rot=<0 0 0 1>}
 lowerarm_twist_01_l {parent=lowerarm_l, loc=<18.316 0 0>, rot=<0 0 0 1>}
 deform_elbow_in_base_l {parent=lowerarm_l, loc=<0.1834 3.6062 -0.2572>, rot=<0 0 0 1>}
 deform_elbow_in_l {parent=deform_elbow_in_base_l, loc=<0 0 0>, rot=<0 0 0 1>}
 deform_elbow_out_base_l {parent=lowerarm_l, loc=<0 -4 0>, rot=<0 0 0 1>}
 deform_elbow_out_l {parent=deform_elbow_out_base_l, loc=<0 0 0>, rot=<0 0 0 1>}
 dyn_elbow_l {parent=lowerarm_l, loc=<-0.9946 -0.7385 5.9106>, rot=<-0.1921 -0.9812 -0.0024 -0.0183>}
 dyn_elbow_l_top {parent=dyn_elbow_l, loc=<0.417 0.0246 0.576>, rot=<0.0084 -0.1004 0.0034 -0.9949>}
 dyn_elbow_l_bot {parent=dyn_elbow_l, loc=<-1.3315 0.1681 0.3795>, rot=<0.0578 -0.984 -0.0037 -0.1684>}
 upperarm_twist_01_l {parent=upperarm_l, loc=<7.2135 -0.1134 0>, rot=<0 0 0 1>}
 deform_upperarm_bk_base_l {parent=upperarm_twist_01_l, loc=<-6.4053 -4.3746 0.314>, rot=<0 0 0 1>}
 deform_upperarm_bk_l {parent=deform_upperarm_bk_base_l, loc=<0 0 0>, rot=<0 0 0 1>}
 deform_upperarm_fr_base_l {parent=upperarm_twist_01_l, loc=<-5.6312 5.092 0.3354>, rot=<0 0 0 1>}
 deform_upperarm_fr_l {parent=deform_upperarm_fr_base_l, loc=<0 0 0>, rot=<0 0 0 1>}
 deform_upperarm_up_base_l {parent=upperarm_twist_01_l, loc=<-8.891 0.4825 3.9517>, rot=<0 0 0 1>}
 deform_upperarm_up_l {parent=deform_upperarm_up_base_l, loc=<0 0 0>, rot=<0 0 0 1>}
 upperarm_twist_02_l {parent=upperarm_l, loc=<19.7998 -0.1134 0>, rot=<0 0 0 1>}
 dyn_shoulder_l {parent=upperarm_l, loc=<-1.7117 0.1764 5.8586>, rot=<-0.0135 0.1096 0.0578 -0.9922>}
 dyn_shoulder_l_fr {parent=dyn_shoulder_l, loc=<2.5219 6.7358 -2.2184>, rot=<-0.1735 -0.8245 -0.1691 -0.5113>}
 dyn_shoulder_l_bk {parent=dyn_shoulder_l, loc=<0.2987 -8.4292 -8.8552>, rot=<0.1053 -0.7261 0.0218 -0.6791>}
 attach_cape {parent=spine_05, loc=<9.3216 -8.4501 0>, rot=<0.0616 -0.7044 -0.0616 -0.7044>}
 attach_backpack {parent=spine_05, loc=<0 -8.35 0>, rot=<-0.0093 -0.707 0.0093 -0.707>}
 neck_01 {parent=spine_05, loc=<11.3138 0.1529 0>, rot=<0 0 -0.2003 -0.9797>}
 neck_02 {parent=neck_01, loc=<4.5627 0 0>, rot=<0 0 0.0504 -0.9987>}
 head {parent=neck_02, loc=<5.3914 0 0>, rot=<0 0 0.115 -0.9934>}
 clavicle_r {parent=spine_05, loc=<5.4831 -0.1053 3.2946>, rot=<0.7289 0.0465 -0.6829 0.009>}
 upperarm_r {parent=clavicle_r, loc=<-9.8658 1.2136 0.4105>, rot=<0.0164 -0.3245 -0.0269 -0.9454>}
 lowerarm_r {parent=upperarm_r, loc=<-26.465 0.8345 -0.259>, rot=<0.0036 0.0161 -0.2298 -0.9731>}
 hand_r {parent=lowerarm_r, loc=<-22.7355 0.3203 0.6106>, rot=<-0.6784 -0.0223 0.0735 -0.7307>}
 thumb_01_r {parent=hand_r, loc=<-1.3118 0.3191 2.459>, rot=<0.6546 -0.4042 0.0664 -0.6354>}
 thumb_02_r {parent=thumb_01_r, loc=<-5.9336 0.0448 -0.0401>, rot=<-0.0037 0.0213 0.298 -0.9543>}
 thumb_03_r {parent=thumb_02_r, loc=<-2.2223 -0.3682 0.0907>, rot=<0.0118 -0.013 0.0066 -0.9998>}
 index_metacarpal_r {parent=hand_r, loc=<-1.6535 -0.0183 2.96>, rot=<0.0158 -0.0124 -0.0173 -0.9996>}
 index_01_r {parent=index_metacarpal_r, loc=<-8.3008 -0.0357 0.0524>, rot=<-0.0044 0.0485 0.0924 -0.9945>}
 index_02_r {parent=index_01_r, loc=<-3.3587 0.3004 -0.0452>, rot=<-0.002 0.0027 0.3225 -0.9466>}
 index_03_r {parent=index_02_r, loc=<-2.5578 0.123 0.0252>, rot=<-0.002 0.0067 0.2558 -0.9667>}
 middle_metacarpal_r {parent=hand_r, loc=<-2.1856 0.1725 0.5984>, rot=<0.006 0.0168 -0.0329 -0.9993>}
 middle_01_r {parent=middle_metacarpal_r, loc=<-7.3275 0.2447 -0.0129>, rot=<-0.0208 0.0042 0.0858 -0.9961>}
 middle_02_r {parent=middle_01_r, loc=<-4.0123 -0.1401 -0.1175>, rot=<-0.0218 0.0258 0.3512 -0.9357>}
 middle_03_r {parent=middle_02_r, loc=<-2.6077 -0.0534 -0.1373>, rot=<0.0082 0.0188 0.1708 -0.9851>}
 ring_metacarpal_r {parent=hand_r, loc=<-2.2074 0.3625 -1.3638>, rot=<-0.0284 0.0722 -0.0458 -0.9959>}
 ring_01_r {parent=ring_metacarpal_r, loc=<-7.0295 0.6197 0.1446>, rot=<-0.0489 -0.0252 0.1892 -0.9804>}
 ring_02_r {parent=ring_01_r, loc=<-3.6174 -0.4647 0.1027>, rot=<0 0 0.2806 -0.9598>}
 ring_03_r {parent=ring_02_r, loc=<-2.4152 -0.114 -0.0199>, rot=<0 0 0.1189 -0.9929>}
 pinky_metacarpal_r {parent=hand_r, loc=<-1.7616 0.9753 -3.3814>, rot=<-0.0397 0.0814 -0.069 -0.9935>}
 pinky_01_r {parent=pinky_metacarpal_r, loc=<-7.1709 0.4831 0.1094>, rot=<-0.0654 -0.0035 0.2242 -0.9723>}
 pinky_02_r {parent=pinky_01_r, loc=<-2.8274 -0.0893 0.1043>, rot=<0.0037 0.0011 0.2854 -0.9584>}
 pinky_03_r {parent=pinky_02_r, loc=<-2.2046 -0.0307 0.0078>, rot=<-0.0002 0.0098 0.1445 -0.9895>}
 deform_wrist_dn_base_r {parent=hand_r, loc=<-0.0001 1.5328 0>, rot=<0 0 0 1>}
 deform_wrist_dn_r {parent=deform_wrist_dn_base_r, loc=<0 0 0>, rot=<0 0 0 1>}
 deform_wrist_up_base_r {parent=hand_r, loc=<-0.0002 -2.3007 -0.0001>, rot=<0 0 0 1>}
 deform_wrist_up_r {parent=deform_wrist_up_base_r, loc=<0 0 0>, rot=<0 0 0 1>}
 dyn_wrist_r {parent=hand_r, loc=<2.3279 -2.1858 0.2766>, rot=<0.9968 -0.0732 -0.0219 0.0218>}
 lowerarm_twist_02_r {parent=lowerarm_r, loc=<-10.6707 0 0>, rot=<0 0 0 -1>}
 lowerarm_twist_01_r {parent=lowerarm_r, loc=<-18.316 0 0>, rot=<0 0 0 -1>}
 deform_elbow_in_base_r {parent=lowerarm_r, loc=<-0.1834 -3.6062 0.2572>, rot=<0 0 0 -1>}
 deform_elbow_in_r {parent=deform_elbow_in_base_r, loc=<0 0 0>, rot=<0 0 0 1>}
 deform_elbow_out_base_r {parent=lowerarm_r, loc=<0 4 -0.0001>, rot=<0 0 0 -1>}
 deform_elbow_out_r {parent=deform_elbow_out_base_r, loc=<0 0 0>, rot=<0 0 0 1>}
 upperarm_twist_01_r {parent=upperarm_r, loc=<-7.2135 0.1134 0>, rot=<0 0 0 -1>}
 deform_upperarm_bk_base_r {parent=upperarm_twist_01_r, loc=<6.4053 4.3746 -0.314>, rot=<0 0 0 1>}
 deform_upperarm_bk_r {parent=deform_upperarm_bk_base_r, loc=<0 0 0>, rot=<0 0 0 1>}
 deform_upperarm_fr_base_r {parent=upperarm_twist_01_r, loc=<5.6312 -5.092 -0.3354>, rot=<0 0 0 1>}
 deform_upperarm_fr_r {parent=deform_upperarm_fr_base_r, loc=<0 0 0>, rot=<0 0 0 1>}
 deform_upperarm_up_base_r {parent=upperarm_twist_01_r, loc=<8.891 -0.4825 -3.9517>, rot=<0 0 0 1>}
 deform_upperarm_up_r {parent=deform_upperarm_up_base_r, loc=<0 0 0>, rot=<0 0 0 1>}
 dyn_sleeve_r_fr {parent=upperarm_twist_01_r, loc=<-0.9361 -4.593 1.6161>, rot=<-0.0215 -0.7098 0.7016 -0.059>}
 dyn_sleeve_r_top {parent=upperarm_twist_01_r, loc=<1.9433 -0.1716 -3.8645>, rot=<0.013 -0.9919 -0.0053 -0.1259>}
 dyn_sleeve_r_bk {parent=upperarm_twist_01_r, loc=<0.3436 5.3229 2.0348>, rot=<-0.036 0.6991 0.7113 0.0639>}
 dyn_sleeve_r_bot {parent=upperarm_twist_01_r, loc=<-2.1723 -0.3855 7.2215>, rot=<-0.0128 -0.008 0.999 0.0417>}
 upperarm_twist_02_r {parent=upperarm_r, loc=<-19.7998 0.1134 0>, rot=<0 0 0 -1>}
 dyn_shoulder_r {parent=upperarm_r, loc=<2.9665 0.0073 -4.9494>, rot=<-0.0409 -0.0076 0.9991 0.0056>}
 dyn_shoulder_r_fr {parent=dyn_shoulder_r, loc=<4.8817 8.0234 3.7292>, rot=<0.1186 0.4908 -0.2094 -0.8373>}
 dyn_shoulder_r_bk {parent=dyn_shoulder_r, loc=<1.7458 -8.1681 4.9145>, rot=<-0.0759 0.5304 0.113 -0.8367>}
 dyn_beads_fr {parent=spine_05, loc=<8.2042 4.4398 0>, rot=<-0.5242 -0.8516 0.0001 -0.0011>}
 dyn_necklace_01 {parent=spine_05, loc=<5.5194 6.9071 -0.1067>, rot=<-0.2185 -0.9758 -0.0018 -0.0059>}
 dyn_necklace_02 {parent=dyn_necklace_01, loc=<6.9294 -0.0001 0>, rot=<0 0 0 1>}
 dyn_beads_l {parent=spine_05, loc=<9.9472 2.5274 -4.4031>, rot=<-0.3553 -0.7145 0.2926 -0.5269>}
 dyn_beads_r {parent=spine_05, loc=<10.1063 2.6277 4.1807>, rot=<-0.4334 -0.7401 -0.2279 0.461>}
 dyn_shirt_fr_l_MSTR {parent=spine_03, loc=<9.9762 8.7595 -10.2576>, rot=<0.0448 -0.9372 -0.3413 -0.0565>}
 dyn_shirt_fr_l_01 {parent=dyn_shirt_fr_l_MSTR, loc=<0 0 0>, rot=<0 0 0 1>}
 dyn_shirt_fr_l_02 {parent=dyn_shirt_fr_l_01, loc=<10.2705 0 0.0001>, rot=<0 0 0 1>}
 dyn_shirt_fr_l_03 {parent=dyn_shirt_fr_l_02, loc=<8.4855 0 0>, rot=<0 0 0 1>}
 dyn_shirt_fr_l_04 {parent=dyn_shirt_fr_l_03, loc=<11.9056 0 0>, rot=<0 0 0 1>}
 dyn_shirt_fr_l_05 {parent=dyn_shirt_fr_l_04, loc=<12.1305 0 0>, rot=<0 0 0 1>}
 dyn_shirt_mid_l_MSTR {parent=spine_03, loc=<6.9442 -0.8257 -12.0881>, rot=<0.039 -0.996 -0.0056 -0.0796>}
 dyn_shirt_mid_l_01 {parent=dyn_shirt_mid_l_MSTR, loc=<0 0 0>, rot=<0 0 0 1>}
 dyn_shirt_mid_l_02 {parent=dyn_shirt_mid_l_01, loc=<8.5336 0 0>, rot=<0 0 0 1>}
 dyn_shirt_mid_l_03 {parent=dyn_shirt_mid_l_02, loc=<7.8113 0 0>, rot=<0 0 0 1>}
 dyn_shirt_mid_l_04 {parent=dyn_shirt_mid_l_03, loc=<9.6793 0 0>, rot=<0 0 0 1>}
 dyn_shirt_mid_l_05 {parent=dyn_shirt_mid_l_04, loc=<12.8104 0 0>, rot=<0 0 0 1>}
 dyn_shirt_bk_l_MSTR {parent=spine_03, loc=<7.6951 -8.8397 -5.9544>, rot=<0.0256 -0.8095 0.5827 -0.0674>}
 dyn_shirt_bk_l_01 {parent=dyn_shirt_bk_l_MSTR, loc=<0 0 0>, rot=<0 0 0 1>}
 dyn_shirt_bk_l_02 {parent=dyn_shirt_bk_l_01, loc=<9.9965 0 0>, rot=<0 0 0 1>}
 dyn_shirt_bk_l_03 {parent=dyn_shirt_bk_l_02, loc=<7.597 0 0>, rot=<0 0 0 1>}
 dyn_shirt_bk_l_04 {parent=dyn_shirt_bk_l_03, loc=<9.6794 0 0>, rot=<0 0 0 1>}
 dyn_shirt_bk_l_05 {parent=dyn_shirt_bk_l_04, loc=<12.6866 0 0>, rot=<0 0 0 1>}
 dyn_shirt_fr_r_MSTR {parent=spine_03, loc=<9.2464 6.8054 11.3206>, rot=<0.0676 0.2998 0.9503 -0.0497>}
 dyn_shirt_fr_r_01 {parent=dyn_shirt_fr_r_MSTR, loc=<0 0 0>, rot=<0 0 0 1>}
 dyn_shirt_fr_r_02 {parent=dyn_shirt_fr_r_01, loc=<9.7783 0 0>, rot=<0 0 0 1>}
 dyn_shirt_fr_r_03 {parent=dyn_shirt_fr_r_02, loc=<8.5888 0 0.0001>, rot=<0 0 0 1>}
 dyn_shirt_fr_r_04 {parent=dyn_shirt_fr_r_03, loc=<10.7819 0 -0.0001>, rot=<0 0 0 1>}
 dyn_shirt_fr_r_05 {parent=dyn_shirt_fr_r_04, loc=<11.6643 0 -0.0001>, rot=<0 0 0 1>}
 dyn_rivet_03 {parent=dyn_shirt_fr_r_03, loc=<-1.4093 -7.0734 -0.2338>, rot=<-0.578 0.3435 0.6363 -0.3782>}
 dyn_rivet_02 {parent=dyn_shirt_fr_r_02, loc=<-2.4563 -7.0347 0.2691>, rot=<-0.5668 0.3196 0.6758 -0.3462>}
 dyn_rivet_01 {parent=dyn_shirt_fr_r_01, loc=<-1.8546 -4.6253 -0.3018>, rot=<-0.5637 0.3048 0.6954 -0.3252>}
 dyn_shirt_mid_r_MSTR {parent=spine_03, loc=<6.7383 -0.8168 11.6073>, rot=<0.0742 0.0043 0.9959 -0.0515>}
 dyn_shirt_mid_r_01 {parent=dyn_shirt_mid_r_MSTR, loc=<0 0 0>, rot=<0 0 0 1>}
 dyn_shirt_mid_r_02 {parent=dyn_shirt_mid_r_01, loc=<8.6352 0 0>, rot=<0 0 0 1>}
 dyn_shirt_mid_r_03 {parent=dyn_shirt_mid_r_02, loc=<7.7597 0 0.0001>, rot=<0 0 0 1>}
 dyn_shirt_mid_r_04 {parent=dyn_shirt_mid_r_03, loc=<9.8164 0 0>, rot=<0 0 0 1>}
 dyn_shirt_mid_r_05 {parent=dyn_shirt_mid_r_04, loc=<12.2276 0 0>, rot=<0 0 0 1>}
 dyn_shirt_bk_r_MSTR {parent=spine_03, loc=<7.66 -8.6377 5.8226>, rot=<0.053 -0.5395 0.8391 -0.0458>}
 dyn_shirt_bk_r_01 {parent=dyn_shirt_bk_r_MSTR, loc=<0 0 0>, rot=<0 0 0 1>}
 dyn_shirt_bk_r_02 {parent=dyn_shirt_bk_r_01, loc=<10.4227 0 0>, rot=<0 0 0 1>}
 dyn_shirt_bk_r_03 {parent=dyn_shirt_bk_r_02, loc=<7.4781 0 0>, rot=<0 0 0 1>}
 dyn_shirt_bk_r_04 {parent=dyn_shirt_bk_r_03, loc=<9.3435 0 0>, rot=<0 0 0 1>}
 dyn_shirt_bk_r_05 {parent=dyn_shirt_bk_r_04, loc=<12.3043 0 0>, rot=<0 0 0 1>}
 thigh_l {parent=pelvis, loc=<-1.0855 -0.2109 -8.7053>, rot=<0.0009 0.0355 0.0155 -0.9993>}
 calf_l {parent=thigh_l, loc=<-42.5284 -0.0019 -0.7297>, rot=<0.0006 -0.0161 -0.0391 -0.9991>}
 foot_l {parent=calf_l, loc=<-46.7401 -2.5014 -0.592>, rot=<0.0427 -0.0238 0.0353 -0.9982>}
 ball_l {parent=foot_l, loc=<-3.5658 10.8066 -0.1544>, rot=<0.0001 -0.0001 -0.7063 -0.7079>}
 calf_twist_02_l {parent=calf_l, loc=<-19.1546 -1.7943 -0.2638>, rot=<0.0237 0.0095 -0.0035 -0.9997>}
 dyn_shoelace_l_in_01 {parent=calf_twist_02_l, loc=<-2.4025 4.1455 4.5878>, rot=<-0.1872 -0.9607 -0.0019 0.2051>}
 dyn_shoelace_l_in_02 {parent=dyn_shoelace_l_in_01, loc=<3.0822 0 0>, rot=<0.0137 0.0491 0.0324 -0.9982>}
 dyn_shoelace_l_in_03 {parent=dyn_shoelace_l_in_02, loc=<4.7508 0 0>, rot=<-0.0004 -0.0018 -0.0072 -1>}
 dyn_shoelace_l_in_04 {parent=dyn_shoelace_l_in_03, loc=<4.5386 0 0>, rot=<0.004 0.0147 -0.0128 -0.9998>}
 dyn_shoelace_l_in_05 {parent=dyn_shoelace_l_in_04, loc=<3.2559 -0.0409 -0.0025>, rot=<-0.0002 -0.0009 0.015 -0.9999>}
 dyn_shoelace_l_out_01 {parent=calf_twist_02_l, loc=<-1.8713 4.4433 -4.7792>, rot=<-0.1439 -0.9741 -0.0329 -0.1715>}
 dyn_shoelace_l_out_02 {parent=dyn_shoelace_l_out_01, loc=<2.5041 0.1822 0.0014>, rot=<-0.0037 -0.0185 -0.0044 -0.9998>}
 dyn_shoelace_l_out_03 {parent=dyn_shoelace_l_out_02, loc=<3.8756 0.2955 -0.0058>, rot=<0.0072 0.0223 -0.1087 -0.9938>}
 dyn_shoelace_l_out_04 {parent=dyn_shoelace_l_out_03, loc=<4.1064 -0.3022 -0.0029>, rot=<-0.0055 -0.0144 0.0548 -0.9984>}
 dyn_shoelace_l_out_05 {parent=dyn_shoelace_l_out_04, loc=<1.8997 0 0>, rot=<0.0357 0.1178 0.0166 -0.9923>}
 calf_twist_01_l {parent=calf_l, loc=<-35.8483 -1.9197 -0.5794>, rot=<0.0237 0.0095 -0.0035 -0.9997>}
 dyn_knee_guard {parent=calf_l, loc=<0.5996 -1.395 -5.5698>, rot=<0.4582 0.5416 -0.4278 -0.5601>}
 dyn_knee_cap {parent=dyn_knee_guard, loc=<8.6808 -3.7381 1.1082>, rot=<0.702 0.0748 -0.0858 -0.703>}
 thigh_twist_01_l {parent=thigh_l, loc=<-27.952 0 0>, rot=<0.0013 0 0 -1>}
 deform_groin_base_l {parent=thigh_l, loc=<0.4801 5.6776 -0.2486>, rot=<0 0 0 -1>}
 deform_groin_l {parent=deform_groin_base_l, loc=<0 0 0>, rot=<0 0 0 1>}
 dyn_loop_l_top {parent=thigh_l, loc=<-1.9855 7.0335 0.4185>, rot=<-0.0209 -0.9992 0.0004 0.0349>}
 dyn_loop_l_in {parent=thigh_l, loc=<-7.7405 6.5065 3.4561>, rot=<-0.0476 -0.9983 0.0013 0.0337>}
 dyn_loop_l_out {parent=thigh_l, loc=<-7.5058 6.1538 -3.3037>, rot=<-0.1032 -0.9941 -0.0035 -0.0333>}
 thigh_r {parent=pelvis, loc=<-1.0855 -0.2109 8.7053>, rot=<0.0355 -0.0009 0.9993 0.0155>}
 calf_r {parent=thigh_r, loc=<42.5284 0.0019 0.7297>, rot=<0.0006 -0.0161 -0.0391 -0.9991>}
 foot_r {parent=calf_r, loc=<46.7401 2.5014 0.592>, rot=<0.0427 -0.0238 0.0353 -0.9982>}
 ball_r {parent=foot_r, loc=<3.5658 -10.8066 0.1544>, rot=<0.0001 -0.0001 -0.7063 -0.7079>}
 calf_twist_02_r {parent=calf_r, loc=<19.1546 1.7943 0.2638>, rot=<0.0237 0.0095 -0.0035 -0.9997>}
 dyn_shoelace_r_in_01 {parent=calf_twist_02_r, loc=<2.3229 -3.4021 -4.8339>, rot=<0.9765 -0.0537 -0.2065 -0.0301>}
 dyn_shoelace_r_in_02 {parent=dyn_shoelace_r_in_01, loc=<1.6969 0 0>, rot=<-0.005 -0.075 -0.037 -0.9965>}
 dyn_shoelace_r_in_03 {parent=dyn_shoelace_r_in_02, loc=<5.4234 0 0>, rot=<-0.0025 -0.0258 0.0062 -0.9996>}
 dyn_shoelace_r_in_04 {parent=dyn_shoelace_r_in_03, loc=<5.6703 0 0>, rot=<0.0004 0.0045 -0.0064 -1>}
 dyn_shoelace_r_in_05 {parent=dyn_shoelace_r_in_04, loc=<3.0098 0 0>, rot=<-0.0017 -0.0112 -0.0427 -0.999>}
 dyn_shoelace_r_in_06 {parent=dyn_shoelace_r_in_05, loc=<3.5664 -0.0038 -0.0117>, rot=<-0.0009 -0.0049 0.0016 -1>}
 dyn_shoelace_r_out_01 {parent=calf_twist_02_r, loc=<2.3383 -4.4687 4.692>, rot=<0.9656 -0.0924 0.2418 -0.0233>}
 dyn_shoelace_r_out_02 {parent=dyn_shoelace_r_out_01, loc=<1.1775 0 0>, rot=<0.009 0.1053 0.0025 -0.9944>}
 dyn_shoelace_r_out_03 {parent=dyn_shoelace_r_out_02, loc=<3.5556 0 0>, rot=<0.0014 0.0269 0.0311 -0.9992>}
 dyn_shoelace_r_out_04 {parent=dyn_shoelace_r_out_03, loc=<5.1329 0 0>, rot=<0.0009 0.0231 -0.0186 -0.9996>}
 dyn_shoelace_r_out_05 {parent=dyn_shoelace_r_out_04, loc=<4.7405 0 0>, rot=<-0.0005 -0.007 -0.0131 -0.9999>}
 dyn_shoelace_r_out_06 {parent=dyn_shoelace_r_out_05, loc=<2.7048 -0.0672 0.0421>, rot=<0.0011 0.0244 0.039 -0.9989>}
 calf_twist_01_r {parent=calf_r, loc=<35.8483 1.9197 0.5794>, rot=<0.0237 0.0095 -0.0035 -0.9997>}
 thigh_twist_01_r {parent=thigh_r, loc=<27.952 0 0>, rot=<0.0013 0 0 -1>}
 deform_groin_base_r {parent=thigh_r, loc=<-0.4801 -5.6776 0.2486>, rot=<0 0 0 -1>}
 deform_groin_r {parent=deform_groin_base_r, loc=<0 0 0>, rot=<0 0 0 1>}
 deform_glute_base_l {parent=pelvis, loc=<-7.3466 -9.0213 -9.1245>, rot=<0.0172 0.031 -0.4477 -0.8935>}
 deform_glute_l {parent=deform_glute_base_l, loc=<0 0 0>, rot=<0 0 0 1>}
 deform_glute_base_r {parent=pelvis, loc=<-7.3464 -9.0213 9.1245>, rot=<0.031 -0.0172 0.8935 -0.4477>}
 deform_glute_r {parent=deform_glute_base_r, loc=<0 0 0>, rot=<0 0 0 1>}
 dyn_loop_c {parent=pelvis, loc=<-0.8324 8.0425 -0.0104>, rot=<-0.0179 -0.9998 0 0>}
]
variables [
 Back {type=double, default=0}
 Crouch {type=bool, default=False}
 Forward {type=double, default=0}
 Jumps {type=double, default=0}
 Left {type=double, default=0}
 Right {type=double, default=0}
 Skydive {type=bool, default=False}
 Swing {type=double, default=0}
 Tact_Sprint {type=bool, default=False}
]
