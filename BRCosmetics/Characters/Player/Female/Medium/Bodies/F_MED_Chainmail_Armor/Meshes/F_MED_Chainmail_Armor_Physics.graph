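# Physics asset for the F_MED_Chainmail_Armor female medium player cosmetic body, set up for a dangling necklace on the upper torso.
# Capsule bodies cover two upper spine bones and two dyn_necklace bones.
# Constraints chain the second necklace segment to the first, the first to spine_05, and spine_05 to spine_04.
# Necklace joints have tight swing and almost no twist, while the spine joint allows a much wider range.
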PhysicsAsset /BRCosmetics/Characters/Player/Female/Medium/Bodies/F_MED_Chainmail_Armor/Meshes/F_MED_Chainmail_Armor_Physics
bodies {
 dyn_necklace_01 [{shapeType=Capsule, length=2.5, radius=1, shapeName=dyn_necklace_01_capsule, center=<4 1 0>, rotation=<rot 90 0 0>}]
 dyn_necklace_02 [{shapeType=Capsule, length=7.3472, radius=1.5, shapeName=dyn_necklace_02_capsule, center=<5 1.5 0>, rotation=<rot 90 0 0>}]
 spine_04 [{shapeType=Capsule, length=12.8307, radius=9.8, center=<-1.5454 2.9333 0>}]
 spine_05 [{shapeType=Capsule, length=18.5, radius=10.1308, center=<-1.0962 -0.4288 0>}]
}
constraints [
 {bone1Name=spine_05, bone2Name=spine_04, swing1LimitDegrees=45, swing2LimitDegrees=45, twistLimitDegrees=45}
 {bone1Name=dyn_necklace_01, bone2Name=spine_05, swing1LimitDegrees=15, swing2LimitDegrees=15, twistLimitDegrees=1}
 {bone1Name=dyn_necklace_02, bone2Name=dyn_necklace_01, swing1LimitDegrees=15, swing2LimitDegrees=15, twistLimitDegrees=1}
]
