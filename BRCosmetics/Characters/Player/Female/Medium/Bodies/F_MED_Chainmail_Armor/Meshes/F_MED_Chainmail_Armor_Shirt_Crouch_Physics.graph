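# Physics asset for the crouch variant of the female medium Chainmail Armor shirt, defining bodies for its dynamic shirt bones.
# Six front, mid and back chains per left and right side each link five capsules below an MSTR root, alongside pelvis, spine and thigh capsules.
# Most chain joints limit swing and tightly limit twist, but the right-mid chain is looser.
# Cross constraints tie the two lowest mid segments to back ones.
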PhysicsAsset /BRCosmetics/Characters/Player/Female/Medium/Bodies/F_MED_Chainmail_Armor/Meshes/F_MED_Chainmail_Armor_Shirt_Crouch_Physics
bodies {
 dyn_shirt_bk_l_01 [{shapeType=Capsule, length=25, radius=4, shapeName=dyn_shirt_bk_l_01_capsule, center=<3 0 4>, rotation=<rot 0 0 90>}]
 dyn_shirt_bk_l_02 [{shapeType=Capsule, length=14, radius=4.5, shapeName=dyn_shirt_bk_l_02_capsule, center=<3 0 4>, rotation=<rot 0 0 90>}]
 dyn_shirt_bk_l_03 [{shapeType=Capsule, length=14, radius=4.5, shapeName=dyn_shirt_bk_l_03_capsule, center=<3 0 4>, rotation=<rot 0 0 90>}]
 dyn_shirt_bk_l_04 [{shapeType=Capsule, length=14, radius=6, shapeName=dyn_shirt_bk_l_04_capsule, center=<3 0 4>, rotation=<rot 0 0 90>}]
 dyn_shirt_bk_l_05 [{shapeType=Capsule, length=14, radius=6, shapeName=dyn_shirt_bk_l_05_capsule, center=<3 0 4>, rotation=<rot 0 0 90>}]
 dyn_shirt_bk_l_MSTR [{shapeType=Capsule, length=0.505, radius=1, shapeName=dyn_shirt_bk_l_MSTR_capsule, center=<5 0 0>, rotation=<rot 0 -90 0>}]
 dyn_shirt_bk_r_01 [{shapeType=Capsule, length=25, radius=4, shapeName=dyn_shirt_bk_r_01_capsule, center=<3 0 4>, rotation=<rot 0 0 90>}]
 dyn_shirt_bk_r_02 [{shapeType=Capsule, length=14, radius=4.5, shapeName=dyn_shirt_bk_r_02_capsule, center=<3 0 4>, rotation=<rot 0 0 90>}]
 dyn_shirt_bk_r_03 [{shapeType=Capsule, length=14, radius=4.5, shapeName=dyn_shirt_bk_r_03_capsule, center=<3 0 4>, rotation=<rot 0 0 90>}]
 dyn_shirt_bk_r_04 [{shapeType=Capsule, length=14, radius=6, shapeName=dyn_shirt_bk_r_04_capsule, center=<3 0 4>, rotation=<rot 0 0 90>}]
 dyn_shirt_bk_r_05 [{shapeType=Capsule, length=14, radius=6, shapeName=dyn_shirt_bk_r_05_capsule, center=<3 0 4>, rotation=<rot 0 0 90>}]
 dyn_shirt_bk_r_MSTR [{shapeType=Capsule, length=0.505, radius=1, shapeName=dyn_shirt_bk_r_MSTR_capsule, center=<5 0 0>, rotation=<rot 0 -90 0>}]
 dyn_shirt_fr_l_01 [{shapeType=Capsule, length=10, radius=5, shapeName=dyn_shirt_fr_l_01_capsule, center=<3 0 3>, rotation=<rot 0 0.0001 70>}]
 dyn_shirt_fr_l_02 [{shapeType=Capsule, length=15.25, radius=5, shapeName=dyn_shirt_fr_l_02_capsule, center=<3 0 3>, rotation=<rot 0 0.0001 70>}]
 dyn_shirt_fr_l_03 [{shapeType=Capsule, length=15, radius=6, shapeName=dyn_shirt_fr_l_03_capsule, center=<3 0 1>, rotation=<rot 0 0 90>}]
 dyn_shirt_fr_l_04 [{shapeType=Capsule, length=20, radius=8, shapeName=dyn_shirt_fr_l_04_capsule, center=<3 0 2>, rotation=<rot 0 0.0001 70>}]
 dyn_shirt_fr_l_05 [{shapeType=Capsule, length=20, radius=8, shapeName=dyn_shirt_fr_l_05_capsule, center=<3 0 2>, rotation=<rot 0 0.0001 70>}]
 dyn_shirt_fr_l_MSTR [{shapeType=Capsule, length=0.505, radius=1, shapeName=dyn_shirt_fr_l_MSTR_capsule, center=<5 0 0>, rotation=<rot 0 -90 0>}]
 dyn_shirt_fr_r_01 [{shapeType=Capsule, length=10, radius=5, shapeName=dyn_shirt_fr_r_01_capsule, center=<3 0 3>, rotation=<rot 0 0 90>}]
 dyn_shirt_fr_r_02 [{shapeType=Capsule, length=10, radius=5, shapeName=dyn_shirt_fr_r_02_capsule, center=<3 0 3>, rotation=<rot 0 0 90>}]
 dyn_shirt_fr_r_03 [{shapeType=Capsule, length=17.0312, radius=3.9688, shapeName=dyn_shirt_fr_r_03_capsule, center=<3 0 4>, rotation=<rot 0.0001 -0.0001 105>}]
 dyn_shirt_fr_r_04 [{shapeType=Capsule, length=22.0312, radius=5.9688, shapeName=dyn_shirt_fr_r_04_capsule, center=<3 0 2>, rotation=<rot 0.0001 -0.0001 105>}]
 dyn_shirt_fr_r_05 [{shapeType=Capsule, length=22.0312, radius=5.9688, shapeName=dyn_shirt_fr_r_05_capsule, center=<3 0 2>, rotation=<rot 0 90 90>}]
 dyn_shirt_fr_r_MSTR [{shapeType=Capsule, length=0.505, radius=1, shapeName=dyn_shirt_fr_r_MSTR_capsule, center=<5 0 0>, rotation=<rot 0 -90 0>}]
 dyn_shirt_mid_l_01 [{shapeType=Capsule, length=25, radius=5, shapeName=dyn_shirt_mid_l_01_capsule, center=<3 0 1>, rotation=<rot 0 0 90>}]
 dyn_shirt_mid_l_02 [{shapeType=Capsule, length=16.75, radius=4.5, shapeName=dyn_shirt_mid_l_02_capsule, center=<3 0 4>, rotation=<rot 0 0 100>}]
 dyn_shirt_mid_l_03 [{shapeType=Capsule, length=15, radius=4.5, shapeName=dyn_shirt_mid_l_03_capsule, center=<3 0 4>, rotation=<rot 0 0 100>}]
 dyn_shirt_mid_l_04 [{shapeType=Capsule, length=20, radius=8, shapeName=dyn_shirt_mid_l_04_capsule, center=<3 0 4>, rotation=<rot 0 0.0001 85>}]
 dyn_shirt_mid_l_05 [{shapeType=Capsule, length=20, radius=8, shapeName=dyn_shirt_mid_l_05_capsule, center=<3 0 4>, rotation=<rot 0 0.0001 85>}]
 dyn_shirt_mid_l_MSTR [{shapeType=Capsule, length=0.505, radius=1, shapeName=dyn_shirt_mid_l_MSTR_capsule, center=<5 0 0>, rotation=<rot 0 -90 0>}]
 dyn_shirt_mid_r_01 [{shapeType=Capsule, length=25, radius=5, shapeName=dyn_shirt_mid_r_01_capsule, center=<3 0 1>, rotation=<rot 0 0 70>}]
 dyn_shirt_mid_r_02 [{shapeType=Capsule, length=26.5312, radius=2.9688, shapeName=dyn_shirt_mid_r_02_capsule, center=<3 0 4>, rotation=<rot 0 -0.0002 110>}]
 dyn_shirt_mid_r_03 [{shapeType=Capsule, length=16.5312, radius=2.9688, shapeName=dyn_shirt_mid_r_03_capsule, center=<3 0 4>, rotation=<rot 0 -0.0002 110>}]
 dyn_shirt_mid_r_04 [{shapeType=Capsule, length=21.5312, radius=6.4688, shapeName=dyn_shirt_mid_r_04_capsule, center=<3 0 4>, rotation=<rot 0 -0.0002 110>}]
 dyn_shirt_mid_r_05 [{shapeType=Capsule, length=21.5312, radius=6.4688, shapeName=dyn_shirt_mid_r_05_capsule, center=<3 0 4>, rotation=<rot 0 -0.0002 110>}]
 dyn_shirt_mid_r_MSTR [{shapeType=Capsule, length=0.505, radius=1, shapeName=dyn_shirt_mid_r_MSTR_capsule, center=<5 0 0>, rotation=<rot 0 -90 0>}]
 pelvis [{shapeType=Capsule, length=9.1322, radius=11.837, center=<-7.3577 0.3845 0>}]
 spine_01 [{shapeType=Capsule, length=14.0837, radius=11.6729, shapeName=spine_01_capsule, rotation=<rot -90 0 -0.0004>}]
 spine_02 [{shapeType=Capsule, length=11.62, radius=10.6261, shapeName=spine_02_capsule, center=<0 1.2163 0>, rotation=<rot -90 179.9998 179.9999>}]
 spine_03 [{shapeType=Capsule, length=11.0936, radius=12.1525, shapeName=spine_02_capsule, center=<0 1.5209 0>, rotation=<rot -90 179.9998 179.9999>}]
 thigh_l [{shapeType=Capsule, length=41.9424, radius=8.5576, center=<-27.2993 0 -0.0102>, rotation=<rot 90 0 0>}]
 thigh_r [{shapeType=Capsule, length=32.6924, radius=7.8076, center=<27.9441 1.2257 -0.4754>, rotation=<rot 90 0 0>}]
}
constraints [
 {bone1Name=thigh_l, bone2Name=pelvis, swing1LimitDegrees=45, swing2LimitDegrees=45, twistLimitDegrees=45}
 {bone1Name=thigh_r, bone2Name=pelvis, swing1LimitDegrees=45, swing2LimitDegrees=45, twistLimitDegrees=45}
 {
  bone1Name=spine_02
  bone2Name=spine_01
  swing1LimitDegrees=45
  swing1Motion=Free
  swing2LimitDegrees=45
  swing2Motion=Free
  twistLimitDegrees=45
  twistMotion=Free
 }
 {
  bone1Name=spine_01
  bone2Name=pelvis
  swing1LimitDegrees=45
  swing1Motion=Free
  swing2LimitDegrees=45
  swing2Motion=Free
  twistLimitDegrees=45
  twistMotion=Free
 }
 {bone1Name=dyn_shirt_bk_l_01, bone2Name=dyn_shirt_bk_l_MSTR, swing1LimitDegrees=45, swing2LimitDegrees=45, twistLimitDegrees=2}
 {bone1Name=dyn_shirt_bk_l_02, bone2Name=dyn_shirt_bk_l_01, swing1LimitDegrees=45, swing2LimitDegrees=45, twistLimitDegrees=2}
 {bone1Name=dyn_shirt_bk_l_03, bone2Name=dyn_shirt_bk_l_02, swing1LimitDegrees=45, swing2LimitDegrees=45, twistLimitDegrees=2}
 {bone1Name=dyn_shirt_bk_l_04, bone2Name=dyn_shirt_bk_l_03, swing1LimitDegrees=45, swing2LimitDegrees=45, twistLimitDegrees=2}
 {bone1Name=dyn_shirt_bk_l_05, bone2Name=dyn_shirt_bk_l_04, swing1LimitDegrees=45, swing2LimitDegrees=45, twistLimitDegrees=2}
 {bone1Name=dyn_shirt_bk_r_01, bone2Name=dyn_shirt_bk_r_MSTR, swing1LimitDegrees=45, swing2LimitDegrees=45, twistLimitDegrees=2}
 {bone1Name=dyn_shirt_bk_r_02, bone2Name=dyn_shirt_bk_r_01, swing1LimitDegrees=45, swing2LimitDegrees=45, twistLimitDegrees=2}
 {bone1Name=dyn_shirt_bk_r_03, bone2Name=dyn_shirt_bk_r_02, swing1LimitDegrees=45, swing2LimitDegrees=45, twistLimitDegrees=2}
 {bone1Name=dyn_shirt_bk_r_04, bone2Name=dyn_shirt_bk_r_03, swing1LimitDegrees=45, swing2LimitDegrees=45, twistLimitDegrees=2}
 {bone1Name=dyn_shirt_bk_r_05, bone2Name=dyn_shirt_bk_r_04, swing1LimitDegrees=45, swing2LimitDegrees=45, twistLimitDegrees=2}
 {bone1Name=dyn_shirt_fr_l_01, bone2Name=dyn_shirt_fr_l_MSTR, swing1LimitDegrees=45, swing2LimitDegrees=45, twistLimitDegrees=2}
 {bone1Name=dyn_shirt_fr_l_02, bone2Name=dyn_shirt_fr_l_01, swing1LimitDegrees=45, swing2LimitDegrees=45, twistLimitDegrees=2}
 {bone1Name=dyn_shirt_fr_l_03, bone2Name=dyn_shirt_fr_l_02, swing1LimitDegrees=45, swing2LimitDegrees=45, twistLimitDegrees=2}
 {bone1Name=dyn_shirt_fr_l_04, bone2Name=dyn_shirt_fr_l_03, swing1LimitDegrees=45, swing2LimitDegrees=45, twistLimitDegrees=2}
 {bone1Name=dyn_shirt_fr_l_05, bone2Name=dyn_shirt_fr_l_04, swing1LimitDegrees=45, swing2LimitDegrees=45, twistLimitDegrees=2}
 {bone1Name=dyn_shirt_mid_l_01, bone2Name=dyn_shirt_mid_l_MSTR, swing1LimitDegrees=35, swing2LimitDegrees=35, twistLimitDegrees=2}
 {bone1Name=dyn_shirt_mid_l_02, bone2Name=dyn_shirt_mid_l_01, swing1LimitDegrees=45, swing2LimitDegrees=45, twistLimitDegrees=0.2}
 {bone1Name=dyn_shirt_mid_l_03, bone2Name=dyn_shirt_mid_l_02, swing1LimitDegrees=45, swing2LimitDegrees=45, twistLimitDegrees=0.2}
 {bone1Name=dyn_shirt_mid_l_04, bone2Name=dyn_shirt_mid_l_03, swing1LimitDegrees=45, swing2LimitDegrees=45, twistLimitDegrees=0.2}
 {bone1Name=dyn_shirt_mid_l_05, bone2Name=dyn_shirt_mid_l_04, swing1LimitDegrees=45, swing2LimitDegrees=45, twistLimitDegrees=0.2}
 {
  bone1Name=spine_03
  bone2Name=spine_02
  swing1LimitDegrees=45
  swing1Motion=Free
  swing2LimitDegrees=45
  swing2Motion=Free
  twistLimitDegrees=45
  twistMotion=Free
 }
 {bone1Name=dyn_shirt_fr_r_01, bone2Name=dyn_shirt_fr_r_MSTR, swing1LimitDegrees=45, swing2LimitDegrees=45, twistLimitDegrees=2}
 {bone1Name=dyn_shirt_fr_r_02, bone2Name=dyn_shirt_fr_r_01, swing1LimitDegrees=45, swing2LimitDegrees=45, twistLimitDegrees=2}
 {bone1Name=dyn_shirt_fr_r_03, bone2Name=dyn_shirt_fr_r_02, swing1LimitDegrees=45, swing2LimitDegrees=45, twistLimitDegrees=2}
 {bone1Name=dyn_shirt_fr_r_04, bone2Name=dyn_shirt_fr_r_03, swing1LimitDegrees=45, swing2LimitDegrees=45, twistLimitDegrees=2}
 {bone1Name=dyn_shirt_fr_r_05, bone2Name=dyn_shirt_fr_r_04, swing1LimitDegrees=45, swing2LimitDegrees=45, twistLimitDegrees=2}
 {
  bone1Name=dyn_shirt_mid_r_01
  bone2Name=dyn_shirt_mid_r_MSTR
  swing1LimitDegrees=45
  swing1Motion=Free
  swing2LimitDegrees=45
  swing2Motion=Free
  twistLimitDegrees=45
  twistMotion=Free
 }
 {bone1Name=dyn_shirt_mid_r_02, bone2Name=dyn_shirt_mid_r_01, swing1LimitDegrees=45, swing2LimitDegrees=45, twistLimitDegrees=45}
 {bone1Name=dyn_shirt_mid_r_03, bone2Name=dyn_shirt_mid_r_02, swing1LimitDegrees=45, swing2LimitDegrees=45, twistLimitDegrees=45}
 {
  bone1Name=dyn_shirt_mid_r_04
  bone2Name=dyn_shirt_mid_r_03
  swing1LimitDegrees=45
  swing1Motion=Free
  swing2LimitDegrees=45
  swing2Motion=Free
  twistLimitDegrees=45
  twistMotion=Free
 }
 {
  bone1Name=dyn_shirt_mid_r_05
  bone2Name=dyn_shirt_mid_r_04
  swing1LimitDegrees=45
  swing1Motion=Free
  swing2LimitDegrees=45
  swing2Motion=Free
  twistLimitDegrees=45
  twistMotion=Free
 }
 {bone1Name=dyn_shirt_mid_r_05, bone2Name=dyn_shirt_bk_r_05, swing1LimitDegrees=45, swing2LimitDegrees=45, twistLimitDegrees=45}
 {bone1Name=dyn_shirt_mid_r_04, bone2Name=dyn_shirt_bk_r_04, swing1LimitDegrees=45, swing2LimitDegrees=45, twistLimitDegrees=45}
 {
  bone1Name=dyn_shirt_mid_l_05
  bone2Name=dyn_shirt_bk_l_05
  swing1LimitDegrees=45
  swing1Motion=Free
  swing2LimitDegrees=45
  swing2Motion=Free
  twistLimitDegrees=45
  twistMotion=Free
 }
 {
  bone1Name=dyn_shirt_mid_l_04
  bone2Name=dyn_shirt_bk_l_04
  swing1LimitDegrees=45
  swing1Motion=Free
  swing2LimitDegrees=45
  swing2Motion=Free
  twistLimitDegrees=45
  twistMotion=Free
 }
]
