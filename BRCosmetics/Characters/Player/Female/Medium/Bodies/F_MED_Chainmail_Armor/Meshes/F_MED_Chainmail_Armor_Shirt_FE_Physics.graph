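# Physics asset for the female medium chainmail armor shirt, defining capsule bodies for its dynamic cloth-like bone chains.
# Six chains (front, mid, back on each side) run from an MSTR bone through five segments, with limited swing and very tight twist.
# Free-rotation cross constraints tie mid to front chains and left back to right back, linking the panels laterally.
# Pelvis, spine and thigh capsules provide the body shapes around the chains.
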PhysicsAsset /BRCosmetics/Characters/Player/Female/Medium/Bodies/F_MED_Chainmail_Armor/Meshes/F_MED_Chainmail_Armor_Shirt_FE_Physics
bodies {
 dyn_shirt_bk_l_01 [{shapeType=Capsule, length=25, radius=3, shapeName=dyn_shirt_bk_l_01_capsule, center=<8 0 3>, rotation=<rot 0 0 90>}]
 dyn_shirt_bk_l_02 [{shapeType=Capsule, length=25, radius=3, shapeName=dyn_shirt_bk_l_02_capsule, center=<8 0 3>, rotation=<rot 0 0 90>}]
 dyn_shirt_bk_l_03 [{shapeType=Capsule, length=36.75, radius=6, shapeName=dyn_shirt_bk_l_03_capsule, center=<8 0 5>, rotation=<rot 0 0 90>}]
 dyn_shirt_bk_l_04 [{shapeType=Capsule, length=36.75, radius=6, shapeName=dyn_shirt_bk_l_04_capsule, center=<8 0 5>, rotation=<rot 0 0 90>}]
 dyn_shirt_bk_l_05 [{shapeType=Capsule, length=45, radius=6, shapeName=dyn_shirt_bk_l_05_capsule, center=<8 0 5>, rotation=<rot 0 0 90>}]
 dyn_shirt_bk_l_MSTR [{shapeType=Capsule, length=0.505, radius=1, shapeName=dyn_shirt_bk_l_MSTR_capsule, center=<5 0 3>, rotation=<rot 0 -90 0>}]
 dyn_shirt_bk_r_01 [{shapeType=Capsule, length=25, radius=3, shapeName=dyn_shirt_bk_r_01_capsule, center=<8 0 3>, rotation=<rot 0 0 90>}]
 dyn_shirt_bk_r_02 [{shapeType=Capsule, length=25, radius=3, shapeName=dyn_shirt_bk_r_02_capsule, center=<8 0 3>, rotation=<rot 0 0 90>}]
 dyn_shirt_bk_r_03 [{shapeType=Capsule, length=36.75, radius=6, shapeName=dyn_shirt_bk_r_03_capsule, center=<8 0 5>, rotation=<rot 0 0 90>}]
 dyn_shirt_bk_r_04 [{shapeType=Capsule, length=36.75, radius=6, shapeName=dyn_shirt_bk_r_04_capsule, center=<8 0 5>, rotation=<rot 0 0 90>}]
 dyn_shirt_bk_r_05 [{shapeType=Capsule, length=45, radius=6, shapeName=dyn_shirt_bk_r_05_capsule, center=<8 0 5>, rotation=<rot 0 0 90>}]
 dyn_shirt_bk_r_MSTR [{shapeType=Capsule, length=0.505, radius=1, shapeName=dyn_shirt_bk_r_MSTR_capsule, center=<5 0 3>, rotation=<rot 0 -90 0>}]
 dyn_shirt_fr_l_01 [{shapeType=Capsule, length=10, radius=3, shapeName=dyn_shirt_fr_l_01_capsule, center=<5 0 3>, rotation=<rot 0 0 90>}]
 dyn_shirt_fr_l_02 [{shapeType=Capsule, length=5.5, radius=3, shapeName=dyn_shirt_fr_l_02_capsule, center=<5 0 3>, rotation=<rot 0 0 90>}]
 dyn_shirt_fr_l_03 [{shapeType=Capsule, length=10, radius=6, shapeName=dyn_shirt_fr_l_03_capsule, center=<5 0 5>, rotation=<rot 0 0 90>}]
 dyn_shirt_fr_l_04 [{shapeType=Capsule, length=15, radius=6, shapeName=dyn_shirt_fr_l_04_capsule, center=<5 0 5>, rotation=<rot 0 0 90>}]
 dyn_shirt_fr_l_05 [{shapeType=Capsule, length=15, radius=6, shapeName=dyn_shirt_fr_l_05_capsule, center=<7 0 5>, rotation=<rot 0 0 90>}]
 dyn_shirt_fr_l_MSTR [{shapeType=Capsule, length=0.505, radius=1, shapeName=dyn_shirt_fr_l_MSTR_capsule, center=<5 0 3>, rotation=<rot 0 -90 0>}]
 dyn_shirt_fr_r_01 [{shapeType=Capsule, length=10, radius=3, shapeName=dyn_shirt_fr_r_01_capsule, center=<5 0 3>, rotation=<rot 0 0 90>}]
 dyn_shirt_fr_r_02 [{shapeType=Capsule, length=10, radius=3, shapeName=dyn_shirt_fr_r_02_capsule, center=<5 0 3>, rotation=<rot 0 0 90>}]
 dyn_shirt_fr_r_03 [{shapeType=Capsule, length=15, radius=6, shapeName=dyn_shirt_fr_r_03_capsule, center=<5 0 5>, rotation=<rot 0 0 90>}]
 dyn_shirt_fr_r_04 [{shapeType=Capsule, length=15, radius=6, shapeName=dyn_shirt_fr_r_04_capsule, center=<5 0 5>, rotation=<rot 0 0 90>}]
 dyn_shirt_fr_r_05 [{shapeType=Capsule, length=15, radius=6, shapeName=dyn_shirt_fr_r_05_capsule, center=<7 0 5>, rotation=<rot 0 0 90>}]
 dyn_shirt_fr_r_MSTR [{shapeType=Capsule, length=0.505, radius=1, shapeName=dyn_shirt_fr_r_MSTR_capsule, center=<5 0 3>, rotation=<rot 0 -90 0>}]
 dyn_shirt_mid_l_01 [{shapeType=Capsule, length=25, radius=3, shapeName=dyn_shirt_mid_l_01_capsule, center=<5 0 3>, rotation=<rot 0 0.0001 85>}]
 dyn_shirt_mid_l_02 [{shapeType=Capsule, length=25, radius=3, shapeName=dyn_shirt_mid_l_02_capsule, center=<5 0 3>, rotation=<rot 0 0.0002 70>}]
 dyn_shirt_mid_l_03 [{shapeType=Capsule, length=27.75, radius=6, shapeName=dyn_shirt_mid_l_03_capsule, center=<5 0 5>, rotation=<rot 0 0.0002 70>}]
 dyn_shirt_mid_l_04 [{shapeType=Capsule, length=32.75, radius=6, shapeName=dyn_shirt_mid_l_04_capsule, center=<5 0 5>, rotation=<rot 0 0.0002 70>}]
 dyn_shirt_mid_l_05 [{shapeType=Capsule, length=38.5, radius=6, shapeName=dyn_shirt_mid_l_05_capsule, center=<5 0 5>, rotation=<rot 0 0.0002 70>}]
 dyn_shirt_mid_l_MSTR [{shapeType=Capsule, length=0.505, radius=1, shapeName=dyn_shirt_mid_l_MSTR_capsule, center=<5 0 3>, rotation=<rot 0 -90 0>}]
 dyn_shirt_mid_r_01 [{shapeType=Capsule, length=25, radius=3, shapeName=dyn_shirt_mid_r_01_capsule, center=<5 0 3>, rotation=<rot 0 0 90>}]
 dyn_shirt_mid_r_02 [{shapeType=Capsule, length=25, radius=3, shapeName=dyn_shirt_mid_r_02_capsule, center=<5 0 3>, rotation=<rot 0 -0.0001 115>}]
 dyn_shirt_mid_r_03 [{shapeType=Capsule, length=36.75, radius=6, shapeName=dyn_shirt_mid_r_03_capsule, center=<5 0 5>, rotation=<rot 0 -0.0001 115>}]
 dyn_shirt_mid_r_04 [{shapeType=Capsule, length=41.75, radius=6, shapeName=dyn_shirt_mid_r_04_capsule, center=<5 0 5>, rotation=<rot 0 -0.0001 115>}]
 dyn_shirt_mid_r_05 [{shapeType=Capsule, length=52.75, radius=6, shapeName=dyn_shirt_mid_r_05_capsule, center=<5 0 5>, rotation=<rot 0 -0.0001 115>}]
 dyn_shirt_mid_r_MSTR [{shapeType=Capsule, length=0.505, radius=1, shapeName=dyn_shirt_mid_r_MSTR_capsule, center=<5 0 3>, rotation=<rot 0 -90 0>}]
 pelvis [{shapeType=Capsule, length=10.8822, radius=11.837, center=<-4.4622 -0.5194 0>}]
 spine_01 [{shapeType=Capsule, length=16.0837, radius=10.6729, shapeName=spine_01_capsule, center=<-5.4558 0 0>, rotation=<rot -90 0 -0.0004>}]
 spine_02 [{shapeType=Capsule, length=15.87, radius=9.3761, shapeName=spine_02_capsule, center=<0.1559 0 0>, rotation=<rot -90 179.9998 179.9999>}]
 spine_03 [{shapeType=Capsule, length=10.8436, radius=8.6525, shapeName=spine_02_capsule, rotation=<rot -90 179.9998 179.9999>}]
 thigh_l [{shapeType=Capsule, length=46.4424, radius=8.3076, center=<-27.2993 0 0>, rotation=<rot 90 0 0>}]
 thigh_r [{shapeType=Capsule, length=39.9424, radius=9.0576, center=<25.3461 1.2257 -0.4754>, rotation=<rot 90 0 0>}]
}
constraints [
 {bone1Name=thigh_l, bone2Name=pelvis, swing1LimitDegrees=45, swing2LimitDegrees=45, twistLimitDegrees=45}
 {bone1Name=thigh_r, bone2Name=pelvis, swing1LimitDegrees=45, swing2LimitDegrees=45, twistLimitDegrees=45}
 {
  bone1Name=spine_02
  bone2Name=spine_01
  swing1LimitDegrees=45
  swing1Motion=Free
  swing2LimitDegrees=45
  swing2Motion=Free
  twistLimitDegrees=45
  twistMotion=Free
 }
 {
  bone1Name=spine_01
  bone2Name=pelvis
  swing1LimitDegrees=45
  swing1Motion=Free
  swing2LimitDegrees=45
  swing2Motion=Free
  twistLimitDegrees=45
  twistMotion=Free
 }
 {bone1Name=dyn_shirt_bk_l_01, bone2Name=dyn_shirt_bk_l_MSTR, swing1LimitDegrees=45, swing2LimitDegrees=45, twistLimitDegrees=2}
 {bone1Name=dyn_shirt_bk_l_02, bone2Name=dyn_shirt_bk_l_01, swing1LimitDegrees=45, swing2LimitDegrees=45, twistLimitDegrees=2}
 {bone1Name=dyn_shirt_bk_l_03, bone2Name=dyn_shirt_bk_l_02, swing1LimitDegrees=45, swing2LimitDegrees=45, twistLimitDegrees=2}
 {bone1Name=dyn_shirt_bk_l_04, bone2Name=dyn_shirt_bk_l_03, swing1LimitDegrees=45, swing2LimitDegrees=45, twistLimitDegrees=2}
 {bone1Name=dyn_shirt_bk_l_05, bone2Name=dyn_shirt_bk_l_04, swing1LimitDegrees=45, swing2LimitDegrees=45, twistLimitDegrees=2}
 {bone1Name=dyn_shirt_bk_r_01, bone2Name=dyn_shirt_bk_r_MSTR, swing1LimitDegrees=45, swing2LimitDegrees=45, twistLimitDegrees=2}
 {bone1Name=dyn_shirt_bk_r_02, bone2Name=dyn_shirt_bk_r_01, swing1LimitDegrees=45, swing2LimitDegrees=45, twistLimitDegrees=2}
 {bone1Name=dyn_shirt_bk_r_03, bone2Name=dyn_shirt_bk_r_02, swing1LimitDegrees=45, swing2LimitDegrees=45, twistLimitDegrees=2}
 {bone1Name=dyn_shirt_bk_r_04, bone2Name=dyn_shirt_bk_r_03, swing1LimitDegrees=45, swing2LimitDegrees=45, twistLimitDegrees=2}
 {bone1Name=dyn_shirt_bk_r_05, bone2Name=dyn_shirt_bk_r_04, swing1LimitDegrees=45, swing2LimitDegrees=45, twistLimitDegrees=2}
 {bone1Name=dyn_shirt_fr_l_01, bone2Name=dyn_shirt_fr_l_MSTR, swing1LimitDegrees=45, swing2LimitDegrees=45, twistLimitDegrees=2}
 {bone1Name=dyn_shirt_fr_l_02, bone2Name=dyn_shirt_fr_l_01, swing1LimitDegrees=45, swing2LimitDegrees=45, twistLimitDegrees=2}
 {bone1Name=dyn_shirt_fr_l_03, bone2Name=dyn_shirt_fr_l_02, swing1LimitDegrees=45, swing2LimitDegrees=45, twistLimitDegrees=2}
 {bone1Name=dyn_shirt_fr_l_04, bone2Name=dyn_shirt_fr_l_03, swing1LimitDegrees=30, swing2LimitDegrees=30, twistLimitDegrees=0.5}
 {bone1Name=dyn_shirt_fr_l_05, bone2Name=dyn_shirt_fr_l_04, swing1LimitDegrees=30, swing2LimitDegrees=30, twistLimitDegrees=0.5}
 {bone1Name=dyn_shirt_fr_r_01, bone2Name=dyn_shirt_fr_r_MSTR, swing1LimitDegrees=45, swing2LimitDegrees=45, twistLimitDegrees=2}
 {bone1Name=dyn_shirt_fr_r_02, bone2Name=dyn_shirt_fr_r_01, swing1LimitDegrees=45, swing2LimitDegrees=45, twistLimitDegrees=2}
 {bone1Name=dyn_shirt_fr_r_03, bone2Name=dyn_shirt_fr_r_02, swing1LimitDegrees=45, swing2LimitDegrees=45, twistLimitDegrees=2}
 {bone1Name=dyn_shirt_fr_r_04, bone2Name=dyn_shirt_fr_r_03, swing1LimitDegrees=30, swing2LimitDegrees=30, twistLimitDegrees=0.5}
 {bone1Name=dyn_shirt_fr_r_05, bone2Name=dyn_shirt_fr_r_04, swing1LimitDegrees=30, swing2LimitDegrees=30, twistLimitDegrees=0.5}
 {bone1Name=dyn_shirt_mid_l_01, bone2Name=dyn_shirt_mid_l_MSTR, swing1LimitDegrees=30, swing2LimitDegrees=30, twistLimitDegrees=2}
 {bone1Name=dyn_shirt_mid_l_02, bone2Name=dyn_shirt_mid_l_01, swing1LimitDegrees=30, swing2LimitDegrees=30, twistLimitDegrees=2}
 {bone1Name=dyn_shirt_mid_l_03, bone2Name=dyn_shirt_mid_l_02, swing1LimitDegrees=30, swing2LimitDegrees=30, twistLimitDegrees=2}
 {bone1Name=dyn_shirt_mid_l_04, bone2Name=dyn_shirt_mid_l_03, swing1LimitDegrees=30, swing2LimitDegrees=30, twistLimitDegrees=2}
 {bone1Name=dyn_shirt_mid_l_05, bone2Name=dyn_shirt_mid_l_04, swing1LimitDegrees=30, swing2LimitDegrees=30, twistLimitDegrees=2}
 {bone1Name=dyn_shirt_mid_r_01, bone2Name=dyn_shirt_mid_r_MSTR, swing1LimitDegrees=30, swing2LimitDegrees=30, twistLimitDegrees=2}
 {bone1Name=dyn_shirt_mid_r_02, bone2Name=dyn_shirt_mid_r_01, swing1LimitDegrees=30, swing2LimitDegrees=30, twistLimitDegrees=2}
 {bone1Name=dyn_shirt_mid_r_03, bone2Name=dyn_shirt_mid_r_02, swing1LimitDegrees=30, swing2LimitDegrees=30, twistLimitDegrees=2}
 {bone1Name=dyn_shirt_mid_r_04, bone2Name=dyn_shirt_mid_r_03, swing1LimitDegrees=30, swing2LimitDegrees=30, twistLimitDegrees=2}
 {bone1Name=dyn_shirt_mid_r_05, bone2Name=dyn_shirt_mid_r_04, swing1LimitDegrees=30, swing2LimitDegrees=30, twistLimitDegrees=2}
 {
  bone1Name=spine_03
  bone2Name=spine_02
  swing1LimitDegrees=45
  swing1Motion=Free
  swing2LimitDegrees=45
  swing2Motion=Free
  twistLimitDegrees=45
  twistMotion=Free
 }
 {
  bone1Name=dyn_shirt_bk_r_04
  bone2Name=dyn_shirt_bk_l_04
  swing1LimitDegrees=45
  swing1Motion=Free
  swing2LimitDegrees=45
  swing2Motion=Free
  twistLimitDegrees=45
  twistMotion=Free
 }
 {
  bone1Name=dyn_shirt_mid_r_03
  bone2Name=dyn_shirt_fr_r_03
  swing1LimitDegrees=45
  swing1Motion=Free
  swing2LimitDegrees=45
  swing2Motion=Free
  twistLimitDegrees=45
  twistMotion=Free
 }
 {
  bone1Name=dyn_shirt_mid_r_02
  bone2Name=dyn_shirt_fr_r_02
  swing1LimitDegrees=45
  swing1Motion=Free
  swing2LimitDegrees=45
  swing2Motion=Free
  twistLimitDegrees=45
  twistMotion=Free
 }
 {
  bone1Name=dyn_shirt_mid_l_03
  bone2Name=dyn_shirt_fr_l_03
  swing1LimitDegrees=45
  swing1Motion=Free
  swing2LimitDegrees=45
  swing2Motion=Free
  twistLimitDegrees=45
  twistMotion=Free
 }
 {
  bone1Name=dyn_shirt_mid_l_02
  bone2Name=dyn_shirt_fr_l_02
  swing1LimitDegrees=45
  swing1Motion=Free
  swing2LimitDegrees=45
  swing2Motion=Free
  twistLimitDegrees=45
  twistMotion=Free
 }
 {
  bone1Name=dyn_shirt_mid_l_05
  bone2Name=dyn_shirt_fr_l_05
  swing1LimitDegrees=45
  swing1Motion=Free
  swing2LimitDegrees=45
  swing2Motion=Free
  twistLimitDegrees=45
  twistMotion=Free
 }
 {
  bone1Name=dyn_shirt_mid_r_05
  bone2Name=dyn_shirt_fr_r_05
  swing1LimitDegrees=45
  swing1Motion=Free
  swing2LimitDegrees=45
  swing2Motion=Free
  twistLimitDegrees=45
  twistMotion=Free
 }
]
