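# Animation Blueprint for the female medium Chainmail Armor body cosmetic, driving secondary motion and correctives on its shirt, shoulder, necklace and rivet bones.
# The cached 'Dynamics' pose feeds front-end, crouch and default branches, each chaining pelvis- and neck-space RigidBody simulations with gravity override and bone fixes.
# Bool blends select the branch, then rivet constraints feed a shirt Control Rig driven by movement variables.
# Layered constraint, Look At and transform correctives follow.
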AnimBlueprint /BRCosmetics/Characters/Player/Female/Medium/Bodies/F_MED_Chainmail_Armor/Meshes/F_MED_Chainmail_Armor_AnimBP
target_skeleton=&/F_MED_Chainmail_Armor_Skeleton
graphs [
 AnimGraph {nodes [
   n1 = "Constraint - dyn_shoulder_l_bk (PPPPPPPP)"!(
    ComponentPose
    Alpha=1
    ConstraintWeights_0=0.3247
    ConstraintWeights_1:float=0.0563
    ConstraintWeights_2:float=0.1149
    ConstraintWeights_3:float=0.4799
    ConstraintWeights_4:float=0.0057
    ConstraintWeights_5:float=0.0066
    ConstraintWeights_6:float=0.006
    ConstraintWeights_7:float=0.0057
   ) -> (Pose) {
    alphaBoolBlend {blendOption=Linear}
    alphaScaleBiasClamp {bClampResult=true, bInterpResult=true, clampMax=0.3, interpSpeedDecreasing=1, interpSpeedIncreasing=0.5}
    boneToModify {boneName=dyn_shoulder_l_bk}
    constraintSetup [
     {targetBone {boneName=upperarm_twist_01_l}}
     {targetBone {boneName=spine_04}}
     {targetBone {boneName=spine_05}}
     {targetBone {boneName=clavicle_l}}
     {targetBone {boneName=dyn_shirt_mid_l_01}}
     {targetBone {boneName=dyn_shirt_mid_l_02}}
     {targetBone {boneName=dyn_shirt_bk_l_01}}
     {targetBone {boneName=dyn_shirt_bk_l_02}}
    ]
    constraintWeights [1,1,1,1,1,1,1,1]
   }
   n2 = "Constraint - dyn_shoulder_r_bk (PPPPPPP)"!(
    ComponentPose=@n1
    Alpha=1
    ConstraintWeights_0=0
    ConstraintWeights_1:float=0.0466
    ConstraintWeights_2:float=0.0003
    ConstraintWeights_3:float=0.1514
    ConstraintWeights_4:float=0.0023
    ConstraintWeights_5:float=0.4705
    ConstraintWeights_6:float=0.3287
   ) -> (Pose) {
    alphaBoolBlend {blendOption=Linear}
    alphaScaleBiasClamp {bClampResult=true, bInterpResult=true, clampMax=0.3, interpSpeedDecreasing=1, interpSpeedIncreasing=0.5}
    boneToModify {boneName=dyn_shoulder_r_bk}
    constraintSetup [
     {targetBone {boneName=spine_02}}
     {targetBone {boneName=spine_04}}
     {targetBone {boneName=spine_03}}
     {targetBone {boneName=spine_05}}
     {targetBone {boneName=clavicle_l}}
     {targetBone {boneName=clavicle_r}}
     {targetBone {boneName=upperarm_twist_01_r}}
    ]
    constraintWeights [1,1,1,1,1,1,1]
   }
   n3 = "Constraint - dyn_shoulder_l_fr (PPPPPPPP)"!(
    ComponentPose=@n2
    Alpha=1
    ConstraintWeights_0=0.6436
    ConstraintWeights_1:float=0.0394
    ConstraintWeights_2:float=0.0591
    ConstraintWeights_3:float=0.2512
    ConstraintWeights_4:float=0.0049
    ConstraintWeights_5:float=0.0003
    ConstraintWeights_6:float=0.001
    ConstraintWeights_7:float=0.0005
   ) -> (Pose) {
    alphaBoolBlend {blendOption=Linear}
    alphaScaleBiasClamp {bClampResult=true, bInterpResult=true, clampMax=0.3, interpSpeedDecreasing=1, interpSpeedIncreasing=0.5}
    boneToModify {boneName=dyn_shoulder_l_fr}
    constraintSetup [
     {targetBone {boneName=upperarm_twist_01_l}}
     {targetBone {boneName=spine_04}}
     {targetBone {boneName=spine_05}}
     {targetBone {boneName=clavicle_l}}
     {targetBone {boneName=dyn_shirt_fr_l_01}}
     {targetBone {boneName=dyn_shirt_fr_l_02}}
     {targetBone {boneName=dyn_shirt_mid_l_01}}
     {targetBone {boneName=dyn_shirt_mid_l_02}}
    ]
    constraintWeights [1,1,1,1,1,1,1,1]
   }
   "Constraint - dyn_shoulder_r_fr (PPPP)"!(
    ComponentPose=@n3
    Alpha=1
    ConstraintWeights_0=0.0177
    ConstraintWeights_1:float=0.0884
    ConstraintWeights_2:float=0.3037
    ConstraintWeights_3:float=0.5902
   ) -> (Pose) {
    alphaBoolBlend {blendOption=Linear}
    alphaScaleBiasClamp {bClampResult=true, bInterpResult=true, clampMax=0.3, interpSpeedDecreasing=1, interpSpeedIncreasing=0.5}
    boneToModify {boneName=dyn_shoulder_r_fr}
    constraintSetup [
     {targetBone {boneName=spine_04}}
     {targetBone {boneName=spine_05}}
     {targetBone {boneName=clavicle_r}}
     {targetBone {boneName=upperarm_twist_01_r}}
    ]
    constraintWeights [1,1,1,1]
   }
   Look At - Bone: dyn_shirt_fr_r_MSTR(Alpha=1) {
    alphaBoolBlend {blendOption=Linear}
    alphaScaleBiasClamp {bClampResult=true, bInterpResult=true, interpSpeedDecreasing=5, interpSpeedIncreasing=7}
    boneToModify {boneName=dyn_shirt_fr_r_MSTR}
    lookAtLocation=<-17.2755 4.7777 46.7251>
    lookAtTarget {boneReference {boneName=pelvis}}
    lookAt_Axis {axis=<1 0>}
   }
   "Transform (Modify) Bone - Bone: dyn_shirt_fr_l_02"!(ComponentPose, Alpha=1, Rotation="4.813,-1.301,-15.035") -> (Pose) {
    rotationMode=BMM_Additive
    rotationSpace=BCS_BoneSpace
    scaleSpace=BCS_BoneSpace
    translationMode=BMM_Additive
    translationSpace=BCS_BoneSpace
    alphaBoolBlend {blendOption=Linear}
    alphaScaleBiasClamp {bInterpResult=true, interpSpeedIncreasing=7}
    boneToModify {boneName=dyn_shirt_fr_l_02}
    rotation=<rot 4.813 -1.301 -15.035>
    translation=<0.0544 -0.3067 -9.6907>
   }
   n4 = Get GravityOverride()
   n5 = RigidBody!(ComponentPose=Local To Component(LocalPose=Use cached pose 'Dynamics'()), Alpha, OverrideWorldGravity:Vector=@n4, ExternalForce) -> (Pose) {
    bClampLinearTranslationLimitToRefPose=true
    bOverrideWorldGravity=true
    overridePhysicsAsset=&/F_MED_Chainmail_Armor_Shirts_Physics
    simulationSpace=BaseBoneSpace
    alphaBoolBlend {blendOption=Linear}
    baseBoneRef {boneName=pelvis}
    componentAppliedLinearAccClamp=<2500 2500 2500>
    componentLinearAccScale=<0.4 0.3 0.1>
    componentLinearVelScale=<0.4 0.3 0.1>
    simSpaceSettings {maxAngularAcceleration=1000, maxAngularVelocity=3, maxLinearAcceleration=250, maxLinearVelocity=500, velocityScaleZ=0, worldAlpha=1}
   }
   n6 = RigidBody(ComponentPose=@n5, OverrideWorldGravity=@n4) {
    bClampLinearTranslationLimitToRefPose=true
    bOverrideWorldGravity=true
    overridePhysicsAsset=&/F_MED_Chainmail_Armor_Physics
    simulationSpace=BaseBoneSpace
    alphaBoolBlend {blendOption=Linear}
    baseBoneRef {boneName=neck_01}
    componentAppliedLinearAccClamp=<3000 3000 3000>
    componentLinearAccScale=<0.4 0.3 0.3>
    componentLinearVelScale=<0.4 0.3 0.3>
    simSpaceSettings {maxAngularAcceleration=1000, maxAngularVelocity=3, maxLinearAcceleration=250, maxLinearVelocity=500, velocityScaleZ=0, worldAlpha=1}
   }
   n7 = Look At - Bone: dyn_shirt_fr_l_MSTR(ComponentPose=@n6, Alpha=1) {
    alphaBoolBlend {blendOption=Linear}
    alphaScaleBiasClamp {bClampResult=true, bInterpResult=true, clampMax=0.3, interpSpeedDecreasing=5, interpSpeedIncreasing=7}
    boneToModify {boneName=dyn_shirt_fr_l_MSTR}
    lookAtLocation=<-93.5684 6.2062 -70.4433>
    lookAtTarget {boneReference {boneName=pelvis}}
    lookAt_Axis {axis=<1 0>}
   }
   n8 = Look At - Bone: dyn_shirt_mid_l_MSTR(ComponentPose=@n7, Alpha=1) {
    alphaBoolBlend {blendOption=Linear}
    alphaScaleBiasClamp {bClampResult=true, bInterpResult=true, clampMax=0.4, interpSpeedDecreasing=5, interpSpeedIncreasing=7}
    boneToModify {boneName=dyn_shirt_mid_l_MSTR}
    lookAtLocation=<-172.213 -12.4154 -216.1829>
    lookAtTarget {boneReference {boneName=pelvis}}
    lookAt_Axis {axis=<1 0>}
   }
   n9 = Look At - Bone: dyn_shirt_fr_r_MSTR(ComponentPose=@n8, Alpha=1) {
    alphaBoolBlend {blendOption=Linear}
    alphaScaleBiasClamp {bClampResult=true, bInterpResult=true, clampMax=0.5, interpSpeedDecreasing=5, interpSpeedIncreasing=7}
    boneToModify {boneName=dyn_shirt_fr_r_MSTR}
    lookAtLocation=<-93.9316 -0.7226 112.8809>
    lookAtTarget {boneReference {boneName=pelvis}}
    lookAt_Axis {axis=<1 0>}
   }
   n10 = Look At - Bone: dyn_shirt_mid_r_MSTR(ComponentPose=@n9, Alpha=1) {
    alphaBoolBlend {blendOption=Linear}
    alphaScaleBiasClamp {bClampResult=true, bInterpResult=true, clampMax=0.5, interpSpeedDecreasing=5, interpSpeedIncreasing=7}
    boneToModify {boneName=dyn_shirt_mid_r_MSTR}
    lookAtLocation=<-172.1129 -78.4784 195.3093>
    lookAtTarget {boneReference {boneName=pelvis}}
    lookAt_Axis {axis=<1 0>}
   }
   n11 = Get GravityOverride()
   n12 = RigidBody(ComponentPose=Local To Component(LocalPose=Use cached pose 'Dynamics'()), OverrideWorldGravity=@n11) {
    bClampLinearTranslationLimitToRefPose=true
    bOverrideWorldGravity=true
    overridePhysicsAsset=&/F_MED_Chainmail_Armor_Shirt_Crouch_Physics
    simulationSpace=BaseBoneSpace
    alphaBoolBlend {blendOption=Linear}
    baseBoneRef {boneName=pelvis}
    componentAppliedLinearAccClamp=<3000 3000 3000>
    componentLinearAccScale=<0.25 0.25 0.4>
    componentLinearVelScale=<0.2 0.2 0.2>
    simSpaceSettings {maxAngularVelocity=3, maxLinearAcceleration=700, maxLinearVelocity=500, velocityScaleZ=0, worldAlpha=1, externalLinearDragV=<2 0.2>}
   }
   n13 = RigidBody(ComponentPose=@n12, OverrideWorldGravity=@n11) {
    bClampLinearTranslationLimitToRefPose=true
    bOverrideWorldGravity=true
    overridePhysicsAsset=&F_MED_Chainmail_Armor_Physics
    simulationSpace=BaseBoneSpace
    alphaBoolBlend {blendOption=Linear}
    baseBoneRef {boneName=neck_01}
    componentAppliedLinearAccClamp=<3000 3000 3000>
    componentLinearAccScale=<0.4 0.3 0.3>
    componentLinearVelScale=<0.4 0.3 0.3>
    simSpaceSettings {maxAngularAcceleration=1000, maxAngularVelocity=3, maxLinearAcceleration=250, maxLinearVelocity=500, velocityScaleZ=0, worldAlpha=1}
   }
   n14 = "Constraint - dyn_shirt_bk_l_03 (P)"(ComponentPose=@n13, Alpha=1, ConstraintWeights_0=1) {
    alphaBoolBlend {blendOption=Linear}
    alphaScaleBiasClamp {bClampResult=true, bInterpResult=true, clampMax=0.6, interpSpeedDecreasing=15}
    boneToModify {boneName=dyn_shirt_bk_l_03}
    constraintSetup [{targetBone {boneName=pelvis}}]
    constraintWeights [1]
   }
   n15 = "Constraint - dyn_shirt_bk_r_03 (P)"(ComponentPose=@n14, Alpha=1, ConstraintWeights_0=1) {
    alphaBoolBlend {blendOption=Linear}
    alphaScaleBiasClamp {bClampResult=true, bInterpResult=true, clampMax=0.6, interpSpeedDecreasing=15}
    boneToModify {boneName=dyn_shirt_bk_r_03}
    constraintSetup [{targetBone {boneName=pelvis}}]
    constraintWeights [1]
   }
   n16 = "Constraint - dyn_shirt_mid_l_03 (P)"(ComponentPose=@n15, Alpha=1, ConstraintWeights_0=1) {
    alphaBoolBlend {blendOption=Linear}
    alphaScaleBiasClamp {bClampResult=true, bInterpResult=true, clampMax=0.3, interpSpeedDecreasing=15}
    boneToModify {boneName=dyn_shirt_mid_l_03}
    constraintSetup [{targetBone {boneName=pelvis}}]
    constraintWeights [1]
   }
   n17 = "Constraint - dyn_shirt_mid_r_03 (P)"(ComponentPose=@n16, Alpha=1, ConstraintWeights_0=1) {
    alphaBoolBlend {blendOption=Linear}
    alphaScaleBiasClamp {bClampResult=true, bInterpResult=true, clampMax=0.3, interpSpeedDecreasing=15}
    boneToModify {boneName=dyn_shirt_mid_r_03}
    constraintSetup [{targetBone {boneName=pelvis}}]
    constraintWeights [1]
   }
   n18 = Look At - Bone: dyn_shirt_bk_l_MSTR(ComponentPose=@n17, Alpha=1) {
    alphaBoolBlend {blendOption=Linear}
    alphaScaleBiasClamp {bClampResult=true, bInterpResult=true, interpSpeedDecreasing=5, interpSpeedIncreasing=7}
    boneToModify {boneName=dyn_shirt_bk_l_MSTR}
    lookAtLocation=<-481.868 -1829.396 279.5607>
    lookAtTarget {boneReference {boneName=pelvis}}
    lookAt_Axis {axis=<1 0>}
   }
   n19 = Look At - Bone: dyn_shirt_bk_r_MSTR(ComponentPose=@n18, Alpha=1) {
    alphaBoolBlend {blendOption=Linear}
    alphaScaleBiasClamp {bClampResult=true, bInterpResult=true, interpSpeedDecreasing=5, interpSpeedIncreasing=7}
    boneToModify {boneName=dyn_shirt_bk_r_MSTR}
    lookAtLocation=<-142.6319 -192.8493 26.5133>
    lookAtTarget {boneReference {boneName=pelvis}}
    lookAt_Axis {axis=<1 0>}
   }
   n20 = "Transform (Modify) Bone - Bone: dyn_shirt_fr_r_03"!(ComponentPose=@n19, Alpha=1, Rotation="7.191,14.815,44.739") -> (Pose) {
    rotationMode=BMM_Additive
    rotationSpace=BCS_BoneSpace
    scaleSpace=BCS_BoneSpace
    translationMode=BMM_Additive
    translationSpace=BCS_BoneSpace
    alphaBoolBlend {blendOption=Linear}
    alphaScaleBiasClamp {bInterpResult=true, interpSpeedIncreasing=7}
    boneToModify {boneName=dyn_shirt_fr_r_03}
    rotation=<rot 7.191 14.815 44.739>
    translation=<2.0233 5.6948 -1.0494>
   }
   n21 = "Transform (Modify) Bone - Bone: dyn_shirt_mid_l_MSTR"!(ComponentPose=@n20, Alpha=1, Rotation="0,40,0") -> (Pose) {
    rotationMode=BMM_Additive
    rotationSpace=BCS_BoneSpace
    scaleSpace=BCS_BoneSpace
    translationSpace=BCS_BoneSpace
    alphaBoolBlend {blendOption=Linear}
    alphaScaleBiasClamp {bInterpResult=true, interpSpeedIncreasing=7}
    boneToModify {boneName=dyn_shirt_mid_l_MSTR}
    rotation {yaw=40}
    translation {y=25}
   }
   n22 = Look At - Bone: dyn_shirt_bk_l_MSTR(ComponentPose=@n21, Alpha=1) {
    alphaBoolBlend {blendOption=Linear}
    alphaScaleBiasClamp {bClampResult=true, bInterpResult=true, clampMax=0.3, interpSpeedDecreasing=5, interpSpeedIncreasing=7}
    boneToModify {boneName=dyn_shirt_bk_l_MSTR}
    lookAtLocation=<-104.0086 -75.691 -110.2188>
    lookAtTarget {boneReference {boneName=pelvis}}
    lookAt_Axis {axis=<1 0>}
   }
   n23 = Blend Poses by bool(
    BlendPose_0=Component To Local(ComponentPose=@n22)
    BlendPose_1=Component To Local(ComponentPose=@n10)
    BlendTime_0=0.2
    BlendTime_1=1.4
    bActiveValue=Get bIsCrouching()
   ) {blendType=Linear}
   n24 = Get GravityOverride()
   n25 = RigidBody(ComponentPose=Local To Component(LocalPose=Use cached pose 'Dynamics'()), OverrideWorldGravity=@n24) {
    bClampLinearTranslationLimitToRefPose=true
    bOverrideWorldGravity=true
    overridePhysicsAsset=&/F_MED_Chainmail_Armor_Shirt_FE_Physics
    simulationSpace=BaseBoneSpace
    alphaBoolBlend {blendOption=Linear}
    baseBoneRef {boneName=pelvis}
    componentAppliedLinearAccClamp=<3000 3000 3000>
    componentLinearAccScale=<0.25 0.25 0.4>
    componentLinearVelScale=<0.2 0.2 0.2>
    simSpaceSettings {maxAngularVelocity=3, maxLinearAcceleration=700, maxLinearVelocity=500, velocityScaleZ=0, worldAlpha=1, externalLinearDragV=<2 0.2>}
   }
   n26 = RigidBody(ComponentPose=@n25, OverrideWorldGravity=@n24) {
    bClampLinearTranslationLimitToRefPose=true
    bOverrideWorldGravity=true
    overridePhysicsAsset=&F_MED_Chainmail_Armor_Physics
    simulationSpace=BaseBoneSpace
    alphaBoolBlend {blendOption=Linear}
    baseBoneRef {boneName=neck_01}
    componentAppliedLinearAccClamp=<3000 3000 3000>
    componentLinearAccScale=<0.4 0.3 0.3>
    componentLinearVelScale=<0.4 0.3 0.3>
    simSpaceSettings {maxAngularAcceleration=1000, maxAngularVelocity=3, maxLinearAcceleration=250, maxLinearVelocity=500, velocityScaleZ=0, worldAlpha=1}
   }
   n27 = Look At - Bone: dyn_shirt_bk_l_MSTR(ComponentPose=@n26, Alpha=1) {
    alphaBoolBlend {blendOption=Linear}
    alphaScaleBiasClamp {bClampResult=true, clampMax=0.2, interpSpeedDecreasing=5, interpSpeedIncreasing=7}
    boneToModify {boneName=dyn_shirt_bk_l_MSTR}
    lookAtLocation=<-203.3643 -53.5888 -20.9512>
    lookAtTarget {boneReference {boneName=pelvis}}
    lookAt_Axis {axis=<1 0>}
   }
   n28 = Look At - Bone: dyn_shirt_fr_r_MSTR(ComponentPose=@n27, Alpha=1) {
    alphaBoolBlend {blendOption=Linear}
    alphaScaleBiasClamp {bClampResult=true, clampMax=0.2, interpSpeedDecreasing=5, interpSpeedIncreasing=7}
    boneToModify {boneName=dyn_shirt_fr_r_MSTR}
    lookAtLocation=<-102.8992 36.1722 27.877>
    lookAtTarget {boneReference {boneName=pelvis}}
    lookAt_Axis {axis=<1 0>}
   }
   n29 = Look At - Bone: dyn_shirt_mid_l_MSTR(ComponentPose=@n28, Alpha=1) {
    alphaBoolBlend {blendOption=Linear}
    alphaScaleBiasClamp {bClampResult=true, clampMax=0.2, interpSpeedDecreasing=5, interpSpeedIncreasing=7}
    boneToModify {boneName=dyn_shirt_mid_l_MSTR}
    lookAtLocation=<-96.1136 61.7875 -75.8794>
    lookAtTarget {boneReference {boneName=pelvis}}
    lookAt_Axis {axis=<1 0>}
   }
   n30 = Look At - Bone: dyn_shirt_mid_r_MSTR(ComponentPose=@n29, Alpha=1) {
    alphaBoolBlend {blendOption=Linear}
    alphaScaleBiasClamp {bClampResult=true, clampMax=0.2, interpSpeedDecreasing=5, interpSpeedIncreasing=7}
    boneToModify {boneName=dyn_shirt_mid_r_MSTR}
    lookAtLocation=<-94475.0201 -61738.4867 6068.6785>
    lookAtTarget {boneReference {boneName=pelvis}}
    lookAt_Axis {axis=<1 0>}
   }
   n31 = "Constraint - dyn_necklace_02 (PP)"!(ComponentPose=@n30, Alpha=0.8, ConstraintWeights_0=0.3585, ConstraintWeights_1:float=0.6415) -> (Pose) {
    alphaBoolBlend {blendOption=Linear}
    boneToModify {boneName=dyn_necklace_02}
    constraintSetup [{targetBone {boneName=spine_04}}, {targetBone {boneName=spine_05}}]
    constraintWeights [1,1]
   }
   n32 = "Transform (Modify) Bone - Bone: dyn_necklace_01"!(ComponentPose=@n31, Alpha=0.5, Rotation="0,2,0") -> (Pose) {
    rotationMode=BMM_Replace
    rotationSpace=BCS_BoneSpace
    scaleSpace=BCS_BoneSpace
    alphaBoolBlend {blendOption=Linear}
    alphaScaleBiasClamp {bInterpResult=true, interpSpeedIncreasing=7}
    boneToModify {boneName=dyn_necklace_01}
    rotation {yaw=2}
    scale {y=1.2}
   }
   n33 = "Transform (Modify) Bone - Bone: dyn_shoulder_r_fr"!(ComponentPose=@n32, Alpha=0.5) -> (Pose) {
    rotationMode=BMM_Replace
    rotationSpace=BCS_BoneSpace
    scaleSpace=BCS_BoneSpace
    translationMode=BMM_Additive
    translationSpace=BCS_BoneSpace
    alphaBoolBlend {blendOption=Linear}
    alphaScaleBiasClamp {bInterpResult=true, interpSpeedIncreasing=7}
    boneToModify {boneName=dyn_shoulder_r_fr}
    scale {y=1.2}
    translation=<-4.1095 5.8716 -7.3122>
   }
   n34 = "Transform (Modify) Bone - Bone: dyn_shoulder_l_bk"!(ComponentPose=@n33, Alpha=0.5) -> (Pose) {
    rotationMode=BMM_Replace
    rotationSpace=BCS_BoneSpace
    scaleSpace=BCS_BoneSpace
    translationMode=BMM_Additive
    translationSpace=BCS_BoneSpace
    alphaBoolBlend {blendOption=Linear}
    alphaScaleBiasClamp {bInterpResult=true, interpSpeedIncreasing=7}
    boneToModify {boneName=dyn_shoulder_l_bk}
    scale {y=1.2}
    translation=<0.0419 -0.9352 -1.4703>
   }
   n35 = "Transform (Modify) Bone - Bone: dyn_wrist_l"!(ComponentPose=@n34, Alpha=0.5) -> (Pose) {
    rotationMode=BMM_Replace
    rotationSpace=BCS_BoneSpace
    scaleSpace=BCS_BoneSpace
    translationMode=BMM_Additive
    translationSpace=BCS_BoneSpace
    alphaBoolBlend {blendOption=Linear}
    alphaScaleBiasClamp {bInterpResult=true, interpSpeedIncreasing=7}
    boneToModify {boneName=dyn_wrist_l}
    rotation {yaw=10}
    scale {y=1.2}
    translation {y=3}
   }
   n36 = "Constraint - dyn_shoulder_r_bk (PPPPPPP)"(
    ComponentPose=@n35
    Alpha=1
    ConstraintWeights_0=0
    ConstraintWeights_1=0.0466
    ConstraintWeights_2=0.0003
    ConstraintWeights_3=0.1514
    ConstraintWeights_4=0.0023
    ConstraintWeights_5=0.4705
    ConstraintWeights_6=0.3287
   ) {
    alphaBoolBlend {blendOption=Linear}
    alphaScaleBiasClamp {bInterpResult=true, interpSpeedIncreasing=7}
    boneToModify {boneName=dyn_shoulder_r_bk}
    constraintSetup [
     {targetBone {boneName=spine_02}}
     {targetBone {boneName=spine_04}}
     {targetBone {boneName=spine_03}}
     {targetBone {boneName=spine_05}}
     {targetBone {boneName=clavicle_l}}
     {targetBone {boneName=clavicle_r}}
     {targetBone {boneName=upperarm_twist_01_r}}
    ]
    constraintWeights [1,1,1,1,1,1,1]
   }
   n37 = Component To Local(ComponentPose=@n36)
   n38 = Get bIsInFrontEnd()
   n39 = "Constraint - dyn_rivet_01 (PPPPP)"!(
    ComponentPose=Local To Component(LocalPose=Blend Poses by bool(BlendPose_0=@n37, BlendPose_1=@n23, bActiveValue=@n38) {blendType=Linear})
    Alpha=1
    ConstraintWeights_0=0.0033
    ConstraintWeights_1:float=0
    ConstraintWeights_2:float=0.9899
    ConstraintWeights_3:float=0.0055
    ConstraintWeights_4:float=0.0012
   ) -> (Pose) {
    alphaBoolBlend {blendOption=Linear}
    boneToModify {boneName=dyn_rivet_01}
    constraintSetup [
     {targetBone {boneName=dyn_shirt_mid_r_01}}
     {targetBone {boneName=dyn_shirt_mid_r_02}}
     {targetBone {boneName=dyn_shirt_fr_r_01}}
     {targetBone {boneName=dyn_shirt_fr_r_02}}
     {targetBone {boneName=dyn_shirt_fr_l_01}}
    ]
    constraintWeights [1,1,1,1,1]
   }
   n40 = "Constraint - dyn_rivet_02 (PPPPPP)"!(
    ComponentPose=@n39
    Alpha=1
    ConstraintWeights_0=0.0073
    ConstraintWeights_1:float=0
    ConstraintWeights_2:float=0.59
    ConstraintWeights_3:float=0.3862
    ConstraintWeights_4:float=0.0165
    ConstraintWeights_5:float=0
   ) -> (Pose) {
    alphaBoolBlend {blendOption=Linear}
    boneToModify {boneName=dyn_rivet_02}
    constraintSetup [
     {targetBone {boneName=dyn_shirt_mid_r_01}}
     {targetBone {boneName=dyn_shirt_mid_r_02}}
     {targetBone {boneName=dyn_shirt_fr_r_01}}
     {targetBone {boneName=dyn_shirt_fr_r_02}}
     {targetBone {boneName=dyn_shirt_fr_r_03}}
     {targetBone {boneName=dyn_shirt_fr_l_01}}
    ]
    constraintWeights [1,1,1,1,1,1]
   }
   n41 = "Constraint - dyn_rivet_03 (PPPPPPPP)"!(
    ComponentPose=@n40
    Alpha=1
    ConstraintWeights_0=0.4554
    ConstraintWeights_1:float=0.0324
    ConstraintWeights_2:float=0.0384
    ConstraintWeights_3:float=0.0652
    ConstraintWeights_4:float=0.1413
    ConstraintWeights_5:float=0.1645
    ConstraintWeights_6:float=0.037
    ConstraintWeights_7:float=0.0658
   ) -> (Pose) {
    alphaBoolBlend {blendOption=Linear}
    boneToModify {boneName=dyn_rivet_03}
    constraintSetup [
     {targetBone {boneName=spine_04}}
     {targetBone {boneName=spine_03}}
     {targetBone {boneName=dyn_shirt_fr_r_01}}
     {targetBone {boneName=dyn_shirt_fr_r_02}}
     {targetBone {boneName=dyn_shirt_fr_r_03}}
     {targetBone {boneName=dyn_shirt_fr_r_04}}
     {targetBone {boneName=dyn_shirt_fr_r_05}}
     {targetBone {boneName=dyn_shirt_mid_r_04}}
    ]
    constraintWeights [1,1,1,1,1,1,1,1]
   }
   n42 = Control Rig!(
    Source=Component To Local(ComponentPose=@n41)
    Alpha
    Crouch:bool=Get bIsCrouching()
    Swing:double=Get Swing() -> (Swing:double)
    Jumps:double=Get Jumping() -> (Jumping:double)
    Forward:double=Get Forward() -> (Forward:double)
    Left:double=Get Left() -> (Left:double)
    Right:double=Get Right() -> (Right:double)
    Back:double=Get Backward() -> (Backward:double)
    Tact_Sprint:bool=Get Tactial Jump() -> (Tactial Jump:bool)
   ) -> (Pose) {
    alphaBoolBlend {blendOption=Linear}
    controlRigAssetReference {blueprintRigClass=&/F_MED_Chainmail_Armor_CtrlRig_Shirt.F_MED_Chainmail_Armor_CtrlRig_Shirt_C}
   }
   n43 = "Constraint - dyn_shoulder_l_bk (PPPPPPPP)"(
    ComponentPose=Local To Component(LocalPose=@n42)
    Alpha=1
    ConstraintWeights_0=0.3247
    ConstraintWeights_1=0.0563
    ConstraintWeights_2=0.1149
    ConstraintWeights_3=0.4799
    ConstraintWeights_4=0.0057
    ConstraintWeights_5=0.0066
    ConstraintWeights_6=0.006
    ConstraintWeights_7=0.0057
   ) {
    alphaBoolBlend {blendOption=Linear}
    alphaScaleBiasClamp {bInterpResult=true, interpSpeedIncreasing=7}
    boneToModify {boneName=dyn_shoulder_l_bk}
    constraintSetup [
     {targetBone {boneName=upperarm_twist_01_l}}
     {targetBone {boneName=spine_04}}
     {targetBone {boneName=spine_05}}
     {targetBone {boneName=clavicle_l}}
     {targetBone {boneName=dyn_shirt_mid_l_01}}
     {targetBone {boneName=dyn_shirt_mid_l_02}}
     {targetBone {boneName=dyn_shirt_bk_l_01}}
     {targetBone {boneName=dyn_shirt_bk_l_02}}
    ]
    constraintWeights [1,1,1,1,1,1,1,1]
   }
   n44 = "Constraint - dyn_shoulder_r_bk (PPPPPPP)"(
    ComponentPose=@n43
    Alpha=1
    ConstraintWeights_0=0
    ConstraintWeights_1=0.0466
    ConstraintWeights_2=0.0003
    ConstraintWeights_3=0.1514
    ConstraintWeights_4=0.0023
    ConstraintWeights_5=0.4705
    ConstraintWeights_6=0.3287
   ) {
    alphaBoolBlend {blendOption=Linear}
    alphaScaleBiasClamp {bInterpResult=true, interpSpeedIncreasing=7}
    boneToModify {boneName=dyn_shoulder_r_bk}
    constraintSetup [
     {targetBone {boneName=spine_02}}
     {targetBone {boneName=spine_04}}
     {targetBone {boneName=spine_03}}
     {targetBone {boneName=spine_05}}
     {targetBone {boneName=clavicle_l}}
     {targetBone {boneName=clavicle_r}}
     {targetBone {boneName=upperarm_twist_01_r}}
    ]
    constraintWeights [1,1,1,1,1,1,1]
   }
   n45 = "Constraint - dyn_shoulder_l_fr (PPPPPPPP)"(
    ComponentPose=@n44
    Alpha=1
    ConstraintWeights_0=0.6436
    ConstraintWeights_1=0.0394
    ConstraintWeights_2=0.0591
    ConstraintWeights_3=0.2512
    ConstraintWeights_4=0.0049
    ConstraintWeights_5=0.0003
    ConstraintWeights_6=0.001
    ConstraintWeights_7=0.0005
   ) {
    alphaBoolBlend {blendOption=Linear}
    alphaScaleBiasClamp {bInterpResult=true, interpSpeedIncreasing=7}
    boneToModify {boneName=dyn_shoulder_l_fr}
    constraintSetup [
     {targetBone {boneName=upperarm_twist_01_l}}
     {targetBone {boneName=spine_04}}
     {targetBone {boneName=spine_05}}
     {targetBone {boneName=clavicle_l}}
     {targetBone {boneName=dyn_shirt_fr_l_01}}
     {targetBone {boneName=dyn_shirt_fr_l_02}}
     {targetBone {boneName=dyn_shirt_mid_l_01}}
     {targetBone {boneName=dyn_shirt_mid_l_02}}
    ]
    constraintWeights [1,1,1,1,1,1,1,1]
   }
   n46 = "Constraint - dyn_shoulder_r_fr (PPPP)"(ComponentPose=@n45, Alpha=1, ConstraintWeights_0=0.0177, ConstraintWeights_1=0.0884, ConstraintWeights_2=0.3037, ConstraintWeights_3=0.5902) {
    alphaBoolBlend {blendOption=Linear}
    alphaScaleBiasClamp {bInterpResult=true, interpSpeedIncreasing=7}
    boneToModify {boneName=dyn_shoulder_r_fr}
    constraintSetup [
     {targetBone {boneName=spine_04}}
     {targetBone {boneName=spine_05}}
     {targetBone {boneName=clavicle_r}}
     {targetBone {boneName=upperarm_twist_01_r}}
    ]
    constraintWeights [1,1,1,1]
   }
   n47 = "Constraint - dyn_shirt_mid_l_02 (P)"(ComponentPose=@n46, Alpha=1, ConstraintWeights_0=1) {
    alphaBoolBlend {blendOption=Linear}
    alphaScaleBiasClamp {bClampResult=true, bInterpResult=true, clampMax=0.5, interpSpeedDecreasing=5, interpSpeedIncreasing=7}
    boneToModify {boneName=dyn_shirt_mid_l_02}
    constraintSetup [{targetBone {boneName=spine_02}}]
    constraintWeights [1]
   }
   n48 = "Constraint - dyn_shirt_bk_l_02 (P)"(ComponentPose=@n47, Alpha=1, ConstraintWeights_0=1) {
    alphaBoolBlend {blendOption=Linear}
    alphaScaleBiasClamp {bClampResult=true, bInterpResult=true, clampMax=0.5, interpSpeedDecreasing=5, interpSpeedIncreasing=7}
    boneToModify {boneName=dyn_shirt_bk_l_02}
    constraintSetup [{targetBone {boneName=spine_02}}]
    constraintWeights [1]
   }
   n49 = "Constraint - dyn_shirt_mid_r_02 (P)"(ComponentPose=@n48, Alpha=1, ConstraintWeights_0=1) {
    alphaBoolBlend {blendOption=Linear}
    alphaScaleBiasClamp {bClampResult=true, bInterpResult=true, clampMax=0.5, interpSpeedDecreasing=5, interpSpeedIncreasing=7}
    boneToModify {boneName=dyn_shirt_mid_r_02}
    constraintSetup [{targetBone {boneName=spine_02}}]
    constraintWeights [1]
   }
   n50 = "Constraint - dyn_shirt_bk_r_02 (P)"(ComponentPose=@n49, Alpha=1, ConstraintWeights_0=1) {
    alphaBoolBlend {blendOption=Linear}
    alphaScaleBiasClamp {bClampResult=true, bInterpResult=true, clampMax=0.5, interpSpeedDecreasing=5, interpSpeedIncreasing=7}
    boneToModify {boneName=dyn_shirt_bk_r_02}
    constraintSetup [{targetBone {boneName=spine_02}}]
    constraintWeights [1]
   }
   n51 = "Constraint - dyn_shirt_mid_l_03 (P)"(ComponentPose=@n50, Alpha=1, ConstraintWeights_0=1) {
    alphaBoolBlend {blendOption=Linear}
    alphaScaleBiasClamp {bClampResult=true, bInterpResult=true, interpSpeedDecreasing=5, interpSpeedIncreasing=7}
    boneToModify {boneName=dyn_shirt_mid_l_03}
    constraintSetup [{targetBone {boneName=spine_01}}]
    constraintWeights [1]
   }
   n52 = "Constraint - dyn_shirt_bk_l_03 (P)"(ComponentPose=@n51, Alpha=1, ConstraintWeights_0=1) {
    alphaBoolBlend {blendOption=Linear}
    alphaScaleBiasClamp {bClampResult=true, bInterpResult=true, interpSpeedDecreasing=5, interpSpeedIncreasing=7}
    boneToModify {boneName=dyn_shirt_bk_l_03}
    constraintSetup [{targetBone {boneName=spine_01}}]
    constraintWeights [1]
   }
   n53 = "Constraint - dyn_shirt_mid_r_03 (P)"(ComponentPose=@n52, Alpha=1, ConstraintWeights_0=1) {
    alphaBoolBlend {blendOption=Linear}
    alphaScaleBiasClamp {bClampResult=true, bInterpResult=true, interpSpeedDecreasing=5, interpSpeedIncreasing=7}
    boneToModify {boneName=dyn_shirt_mid_r_03}
    constraintSetup [{targetBone {boneName=spine_01}}]
    constraintWeights [1]
   }
   n54 = "Constraint - dyn_shirt_bk_r_03 (P)"(ComponentPose=@n53, Alpha=1, ConstraintWeights_0=1) {
    alphaBoolBlend {blendOption=Linear}
    alphaScaleBiasClamp {bClampResult=true, bInterpResult=true, interpSpeedDecreasing=5, interpSpeedIncreasing=7}
    boneToModify {boneName=dyn_shirt_bk_r_03}
    constraintSetup [{targetBone {boneName=spine_01}}]
    constraintWeights [1]
   }
   n55 = "Constraint - dyn_shirt_bk_l_04 (P)"(ComponentPose=@n54, Alpha=1, ConstraintWeights_0=1) {
    alphaBoolBlend {blendOption=Linear}
    alphaScaleBiasClamp {bClampResult=true, bInterpResult=true, clampMax=0.4, interpSpeedDecreasing=5, interpSpeedIncreasing=7}
    boneToModify {boneName=dyn_shirt_bk_l_04}
    constraintSetup [{targetBone {boneName=pelvis}}]
    constraintWeights [1]
   }
   n56 = "Constraint - dyn_shirt_bk_r_04 (P)"(ComponentPose=@n55, Alpha=1, ConstraintWeights_0=1) {
    alphaBoolBlend {blendOption=Linear}
    alphaScaleBiasClamp {bClampResult=true, bInterpResult=true, clampMax=0.4, interpSpeedDecreasing=5, interpSpeedIncreasing=7}
    boneToModify {boneName=dyn_shirt_bk_r_04}
    constraintSetup [{targetBone {boneName=pelvis}}]
    constraintWeights [1]
   }
   n57 = "Transform (Modify) Bone - Bone: dyn_shirt_bk_l_03"!(ComponentPose=@n56, Alpha=1, Translation="0,0,3") -> (Pose) {
    rotationMode=BMM_Additive
    rotationSpace=BCS_BoneSpace
    scaleSpace=BCS_BoneSpace
    translationMode=BMM_Additive
    translationSpace=BCS_BoneSpace
    alphaBoolBlend {blendOption=Linear}
    alphaScaleBiasClamp {bInterpResult=true, interpSpeedIncreasing=7}
    boneToModify {boneName=dyn_shirt_bk_l_03}
    translation {z=3}
   }
   n58 = "Transform (Modify) Bone - Bone: dyn_shirt_bk_r_03"!(ComponentPose=@n57, Alpha=1, Translation="0,0,3") -> (Pose) {
    rotationMode=BMM_Additive
    rotationSpace=BCS_BoneSpace
    scaleSpace=BCS_BoneSpace
    translationMode=BMM_Additive
    translationSpace=BCS_BoneSpace
    alphaBoolBlend {blendOption=Linear}
    alphaScaleBiasClamp {bInterpResult=true, interpSpeedIncreasing=7}
    boneToModify {boneName=dyn_shirt_bk_r_03}
    rotation {pitch=-5}
    translation {z=3}
   }
   n59 = "Transform (Modify) Bone - Bone: dyn_shirt_mid_l_03"!(ComponentPose=@n58, Alpha=1, Translation="0,0,6") -> (Pose) {
    rotationMode=BMM_Additive
    rotationSpace=BCS_BoneSpace
    scaleSpace=BCS_BoneSpace
    translationMode=BMM_Additive
    translationSpace=BCS_BoneSpace
    alphaBoolBlend {blendOption=Linear}
    alphaScaleBiasClamp {bInterpResult=true, interpSpeedIncreasing=7}
    boneToModify {boneName=dyn_shirt_mid_l_03}
    translation {z=6}
   }
   n60 = "Transform (Modify) Bone - Bone: dyn_shirt_mid_r_03"!(ComponentPose=@n59, Alpha=1, Translation="0,0,6") -> (Pose) {
    rotationMode=BMM_Additive
    rotationSpace=BCS_BoneSpace
    scaleSpace=BCS_BoneSpace
    translationMode=BMM_Additive
    translationSpace=BCS_BoneSpace
    alphaBoolBlend {blendOption=Linear}
    alphaScaleBiasClamp {bInterpResult=true, interpSpeedIncreasing=7}
    boneToModify {boneName=dyn_shirt_mid_r_03}
    translation {z=6}
   }
   n61 = Look At - Bone: dyn_shirt_bk_l_MSTR(ComponentPose=@n60, Alpha=1) {
    alphaBoolBlend {blendOption=Linear}
    boneToModify {boneName=dyn_shirt_bk_l_MSTR}
    lookAtLocation=<-136.1044 -71.6377 27.8233>
    lookAtTarget {boneReference {boneName=spine_03}}
    lookAt_Axis {axis=<1 0>}
   }
   n62 = Look At - Bone: dyn_shirt_bk_r_MSTR(ComponentPose=@n61, Alpha=1) {
    alphaBoolBlend {blendOption=Linear}
    boneToModify {boneName=dyn_shirt_bk_r_MSTR}
    lookAtLocation=<-124.6278 -74.8795 -26.2139>
    lookAtTarget {boneReference {boneName=spine_03}}
    lookAt_Axis {axis=<1 0>}
   }
   n63 = "Constraint - dyn_shoulder_l_bk (PPPPPPPP)"(
    ComponentPose=@n62
    Alpha=1
    ConstraintWeights_0=0.3247
    ConstraintWeights_1=0.0563
    ConstraintWeights_2=0.1149
    ConstraintWeights_3=0.4799
    ConstraintWeights_4=0.0057
    ConstraintWeights_5=0.0066
    ConstraintWeights_6=0.006
    ConstraintWeights_7=0.0057
   ) {
    alphaBoolBlend {blendOption=Linear}
    alphaScaleBiasClamp {bClampResult=true, bInterpResult=true, clampMax=0.5, interpSpeedIncreasing=7}
    boneToModify {boneName=dyn_shoulder_l_bk}
    constraintSetup [
     {targetBone {boneName=upperarm_twist_01_l}}
     {targetBone {boneName=spine_04}}
     {targetBone {boneName=spine_05}}
     {targetBone {boneName=clavicle_l}}
     {targetBone {boneName=dyn_shirt_mid_l_01}}
     {targetBone {boneName=dyn_shirt_mid_l_02}}
     {targetBone {boneName=dyn_shirt_bk_l_01}}
     {targetBone {boneName=dyn_shirt_bk_l_02}}
    ]
    constraintWeights [1,1,1,1,1,1,1,1]
   }
   n64 = "Constraint - dyn_shoulder_r_bk (PPPPPPP)"(
    ComponentPose=@n63
    Alpha=1
    ConstraintWeights_0=0
    ConstraintWeights_1=0.0466
    ConstraintWeights_2=0.0003
    ConstraintWeights_3=0.1514
    ConstraintWeights_4=0.0023
    ConstraintWeights_5=0.4705
    ConstraintWeights_6=0.3287
   ) {
    alphaBoolBlend {blendOption=Linear}
    alphaScaleBiasClamp {bClampResult=true, bInterpResult=true, clampMax=0.5, interpSpeedIncreasing=7}
    boneToModify {boneName=dyn_shoulder_r_bk}
    constraintSetup [
     {targetBone {boneName=spine_02}}
     {targetBone {boneName=spine_04}}
     {targetBone {boneName=spine_03}}
     {targetBone {boneName=spine_05}}
     {targetBone {boneName=clavicle_l}}
     {targetBone {boneName=clavicle_r}}
     {targetBone {boneName=upperarm_twist_01_r}}
    ]
    constraintWeights [1,1,1,1,1,1,1]
   }
   n65 = "Constraint - dyn_shoulder_l_fr (PPPPPPPP)"(
    ComponentPose=@n64
    Alpha=1
    ConstraintWeights_0=0.6436
    ConstraintWeights_1=0.0394
    ConstraintWeights_2=0.0591
    ConstraintWeights_3=0.2512
    ConstraintWeights_4=0.0049
    ConstraintWeights_5=0.0003
    ConstraintWeights_6=0.001
    ConstraintWeights_7=0.0005
   ) {
    alphaBoolBlend {blendOption=Linear}
    alphaScaleBiasClamp {bClampResult=true, bInterpResult=true, clampMax=0.5, interpSpeedIncreasing=7}
    boneToModify {boneName=dyn_shoulder_l_fr}
    constraintSetup [
     {targetBone {boneName=upperarm_twist_01_l}}
     {targetBone {boneName=spine_04}}
     {targetBone {boneName=spine_05}}
     {targetBone {boneName=clavicle_l}}
     {targetBone {boneName=dyn_shirt_fr_l_01}}
     {targetBone {boneName=dyn_shirt_fr_l_02}}
     {targetBone {boneName=dyn_shirt_mid_l_01}}
     {targetBone {boneName=dyn_shirt_mid_l_02}}
    ]
    constraintWeights [1,1,1,1,1,1,1,1]
   }
   n66 = "Constraint - dyn_shoulder_r_fr (PPPP)"(ComponentPose=@n65, Alpha=1, ConstraintWeights_0=0.0177, ConstraintWeights_1=0.0884, ConstraintWeights_2=0.3037, ConstraintWeights_3=0.5902) {
    alphaBoolBlend {blendOption=Linear}
    alphaScaleBiasClamp {bClampResult=true, bInterpResult=true, clampMax=0.5, interpSpeedIncreasing=7}
    boneToModify {boneName=dyn_shoulder_r_fr}
    constraintSetup [
     {targetBone {boneName=spine_04}}
     {targetBone {boneName=spine_05}}
     {targetBone {boneName=clavicle_r}}
     {targetBone {boneName=upperarm_twist_01_r}}
    ]
    constraintWeights [1,1,1,1]
   }
   n67 = "Constraint - dyn_shirt_bk_l_04 (P)"(ComponentPose=@n66, Alpha=1, ConstraintWeights_0=1) {
    alphaBoolBlend {blendOption=Linear}
    alphaScaleBiasClamp {bClampResult=true, bInterpResult=true, clampMax=0.8, interpSpeedDecreasing=5, interpSpeedIncreasing=7}
    boneToModify {boneName=dyn_shirt_bk_l_04}
    constraintSetup [{targetBone {boneName=pelvis}}]
    constraintWeights [1]
   }
   n68 = "Constraint - dyn_shirt_bk_r_04 (P)"(ComponentPose=@n67, Alpha=1, ConstraintWeights_0=1) {
    alphaBoolBlend {blendOption=Linear}
    alphaScaleBiasClamp {bClampResult=true, bInterpResult=true, clampMax=0.8, interpSpeedDecreasing=5, interpSpeedIncreasing=7}
    boneToModify {boneName=dyn_shirt_bk_r_04}
    constraintSetup [{targetBone {boneName=pelvis}}]
    constraintWeights [1]
   }
   n69 = "Transform (Modify) Bone - Bone: dyn_shoulder_l"!(ComponentPose=@n68, Alpha=1, Scale="1,1.2,1") -> (Pose) {
    scaleMode=BMM_Additive
    scaleSpace=BCS_BoneSpace
    alphaBoolBlend {blendOption=Linear}
    alphaScaleBiasClamp {bInterpResult=true, interpSpeedIncreasing=7}
    boneToModify {boneName=dyn_shoulder_l}
    scale {y=1.2}
   }
   n70 = "Transform (Modify) Bone - Bone: dyn_shoulder_r"!(ComponentPose=@n69, Alpha=1, Scale="1,1.3,1") -> (Pose) {
    scaleMode=BMM_Additive
    scaleSpace=BCS_BoneSpace
    alphaBoolBlend {blendOption=Linear}
    alphaScaleBiasClamp {bInterpResult=true, interpSpeedIncreasing=7}
    boneToModify {boneName=dyn_shoulder_r}
    scale {y=1.3}
   }
   n71 = "Constraint - dyn_shoulder_l_bk (PPPPPPPP)"(
    ComponentPose=@n70
    Alpha=1
    ConstraintWeights_0=0.3247
    ConstraintWeights_1=0.0563
    ConstraintWeights_2=0.1149
    ConstraintWeights_3=0.4799
    ConstraintWeights_4=0.0057
    ConstraintWeights_5=0.0066
    ConstraintWeights_6=0.006
    ConstraintWeights_7=0.0057
   ) {
    alphaBoolBlend {blendOption=Linear}
    alphaScaleBiasClamp {bInterpResult=true, interpSpeedIncreasing=7}
    boneToModify {boneName=dyn_shoulder_l_bk}
    constraintSetup [
     {targetBone {boneName=upperarm_twist_01_l}}
     {targetBone {boneName=spine_04}}
     {targetBone {boneName=spine_05}}
     {targetBone {boneName=clavicle_l}}
     {targetBone {boneName=dyn_shirt_mid_l_01}}
     {targetBone {boneName=dyn_shirt_mid_l_02}}
     {targetBone {boneName=dyn_shirt_bk_l_01}}
     {targetBone {boneName=dyn_shirt_bk_l_02}}
    ]
    constraintWeights [1,1,1,1,1,1,1,1]
   }
   n72 = "Constraint - dyn_shoulder_r_bk (PPPPPPP)"(
    ComponentPose=@n71
    Alpha=1
    ConstraintWeights_0=0
    ConstraintWeights_1=0.0466
    ConstraintWeights_2=0.0003
    ConstraintWeights_3=0.1514
    ConstraintWeights_4=0.0023
    ConstraintWeights_5=0.4705
    ConstraintWeights_6=0.3287
   ) {
    alphaBoolBlend {blendOption=Linear}
    alphaScaleBiasClamp {bInterpResult=true, interpSpeedIncreasing=7}
    boneToModify {boneName=dyn_shoulder_r_bk}
    constraintSetup [
     {targetBone {boneName=spine_02}}
     {targetBone {boneName=spine_04}}
     {targetBone {boneName=spine_03}}
     {targetBone {boneName=spine_05}}
     {targetBone {boneName=clavicle_l}}
     {targetBone {boneName=clavicle_r}}
     {targetBone {boneName=upperarm_twist_01_r}}
    ]
    constraintWeights [1,1,1,1,1,1,1]
   }
   n73 = "Constraint - dyn_shoulder_l_fr (PPPPPPPP)"(
    ComponentPose=@n72
    Alpha=1
    ConstraintWeights_0=0.6436
    ConstraintWeights_1=0.0394
    ConstraintWeights_2=0.0591
    ConstraintWeights_3=0.2512
    ConstraintWeights_4=0.0049
    ConstraintWeights_5=0.0003
    ConstraintWeights_6=0.001
    ConstraintWeights_7=0.0005
   ) {
    alphaBoolBlend {blendOption=Linear}
    alphaScaleBiasClamp {bInterpResult=true, interpSpeedIncreasing=7}
    boneToModify {boneName=dyn_shoulder_l_fr}
    constraintSetup [
     {targetBone {boneName=upperarm_twist_01_l}}
     {targetBone {boneName=spine_04}}
     {targetBone {boneName=spine_05}}
     {targetBone {boneName=clavicle_l}}
     {targetBone {boneName=dyn_shirt_fr_l_01}}
     {targetBone {boneName=dyn_shirt_fr_l_02}}
     {targetBone {boneName=dyn_shirt_mid_l_01}}
     {targetBone {boneName=dyn_shirt_mid_l_02}}
    ]
    constraintWeights [1,1,1,1,1,1,1,1]
   }
   n74 = "Constraint - dyn_shoulder_r_fr (PPPP)"(ComponentPose=@n73, Alpha=1, ConstraintWeights_0=0.0177, ConstraintWeights_1=0.0884, ConstraintWeights_2=0.3037, ConstraintWeights_3=0.5902) {
    alphaBoolBlend {blendOption=Linear}
    alphaScaleBiasClamp {bInterpResult=true, interpSpeedIncreasing=7}
    boneToModify {boneName=dyn_shoulder_r_fr}
    constraintSetup [
     {targetBone {boneName=spine_04}}
     {targetBone {boneName=spine_05}}
     {targetBone {boneName=clavicle_r}}
     {targetBone {boneName=upperarm_twist_01_r}}
    ]
    constraintWeights [1,1,1,1]
   }
   n75 = "Constraint - dyn_necklace_01 (PPPP)"!(
    ComponentPose=@n74
    Alpha=1
    ConstraintWeights_0=0.0016
    ConstraintWeights_1:float=0.894
    ConstraintWeights_2:float=0.0624
    ConstraintWeights_3:float=0.042
   ) -> (Pose) {
    alphaBoolBlend {blendOption=Linear}
    boneToModify {boneName=dyn_necklace_01}
    constraintSetup [{targetBone {boneName=spine_04}}, {targetBone {boneName=spine_05}}, {targetBone {boneName=clavicle_l}}, {targetBone {boneName=clavicle_r}}]
    constraintWeights [1,1,1,1]
   }
   n76 = "Constraint - dyn_necklace_02 (PP)"(ComponentPose=@n75, Alpha=1, ConstraintWeights_0=0.3585, ConstraintWeights_1=0.6415) {
    alphaBoolBlend {blendOption=Linear}
    boneToModify {boneName=dyn_necklace_02}
    constraintSetup [{targetBone {boneName=spine_04}}, {targetBone {boneName=spine_05}}]
    constraintWeights [1,1]
   }
   n77 = "Constraint - dyn_wrist_l (PPPPP)"!(
    ComponentPose=@n76
    Alpha=1
    ConstraintWeights_0=0.5778
    ConstraintWeights_1:float=0.0427
    ConstraintWeights_2:float=0.0989
    ConstraintWeights_3:float=0.0354
    ConstraintWeights_4:float=0.2452
   ) -> (Pose) {
    alphaBoolBlend {blendOption=Linear}
    boneToModify {boneName=dyn_wrist_l}
    constraintSetup [
     {targetBone {boneName=hand_l}}
     {targetBone {boneName=ring_metacarpal_l}}
     {targetBone {boneName=middle_metacarpal_l}}
     {targetBone {boneName=index_metacarpal_l}}
     {targetBone {boneName=lowerarm_twist_01_l}}
    ]
    constraintWeights [1,1,1,1,1]
   }
   n78 = "Constraint - dyn_wrist_r (PPPPPPPPP)"!(
    ComponentPose=@n77
    Alpha=1
    ConstraintWeights_0=0.051
    ConstraintWeights_1:float=0.6815
    ConstraintWeights_2:float=0.055
    ConstraintWeights_3:float=0.1751
    ConstraintWeights_4:float=0.0016
    ConstraintWeights_5:float=0.0069
    ConstraintWeights_6:float=0.0103
    ConstraintWeights_7:float=0.0069
    ConstraintWeights_8:float=0.0116
   ) -> (Pose) {
    alphaBoolBlend {blendOption=Linear}
    boneToModify {boneName=dyn_wrist_r}
    constraintSetup [
     {targetBone {boneName=lowerarm_twist_02_r}}
     {targetBone {boneName=lowerarm_twist_01_r}}
     {targetBone {boneName=lowerarm_r}}
     {targetBone {boneName=hand_r}}
     {targetBone {boneName=pinky_metacarpal_r}}
     {targetBone {boneName=ring_metacarpal_r}}
     {targetBone {boneName=middle_metacarpal_r}}
     {targetBone {boneName=index_metacarpal_r}}
     {targetBone {boneName=thumb_01_r}}
    ]
    constraintWeights [1,1,1,1,1,1,1,1,1]
   }
   n79 = Look At - Bone: dyn_wrist_l(ComponentPose=@n78, Alpha=1) {
    alphaBoolBlend {blendOption=Linear}
    alphaScaleBiasClamp {interpSpeedDecreasing=5, interpSpeedIncreasing=7}
    boneToModify {boneName=dyn_wrist_l}
    lookAtLocation=<-3.4019 1.3831 18.5481>
    lookAtTarget {boneReference {boneName=lowerarm_l}}
    lookAt_Axis {axis=<1 0>}
   }
   n80 = Look At - Bone: dyn_wrist_r(ComponentPose=@n79, Alpha=1) {
    alphaBoolBlend {blendOption=Linear}
    alphaScaleBiasClamp {interpSpeedDecreasing=5, interpSpeedIncreasing=7}
    boneToModify {boneName=dyn_wrist_r}
    lookAtLocation=<30.4482 -34.6184 -2.2128>
    lookAtTarget {boneReference {boneName=hand_r}}
    lookAt_Axis {axis=<1 0>}
   }
   n81 = "Transform (Modify) Bone - Bone: dyn_shoulder_l_bk"!(ComponentPose=@n80, Alpha=1, Rotation:Rotator="0.087,-24.993,9.885") -> (Pose) {
    rotationMode=BMM_Additive
    rotationSpace=BCS_BoneSpace
    alphaBoolBlend {blendOption=Linear}
    alphaScaleBiasClamp {bInterpResult=true, interpSpeedDecreasing=5, interpSpeedIncreasing=7}
    boneToModify {boneName=dyn_shoulder_l_bk}
    rotation=<rot 0.087 -24.993 9.885>
   }
   n82 = "Constraint - dyn_shirt_mid_l_03 (P)"(ComponentPose=@n81, Alpha=1, ConstraintWeights_0=1) {
    alphaBoolBlend {blendOption=Linear}
    alphaScaleBiasClamp {bClampResult=true, bInterpResult=true, clampMax=0.3, interpSpeedDecreasing=5, interpSpeedIncreasing=7}
    boneToModify {boneName=dyn_shirt_mid_l_03}
    constraintSetup [{targetBone {boneName=spine_02}}]
    constraintWeights [1]
   }
   n83 = "Constraint - dyn_shirt_bk_l_03 (P)"(ComponentPose=@n82, Alpha=1, ConstraintWeights_0=1) {
    alphaBoolBlend {blendOption=Linear}
    alphaScaleBiasClamp {bClampResult=true, bInterpResult=true, clampMax=0.3, interpSpeedDecreasing=5, interpSpeedIncreasing=7}
    boneToModify {boneName=dyn_shirt_bk_l_03}
    constraintSetup [{targetBone {boneName=spine_02}}]
    constraintWeights [1]
   }
   n84 = "Constraint - dyn_shirt_mid_r_03 (P)"(ComponentPose=@n83, Alpha=1, ConstraintWeights_0=1) {
    alphaBoolBlend {blendOption=Linear}
    alphaScaleBiasClamp {bClampResult=true, bInterpResult=true, clampMax=0.3, interpSpeedDecreasing=5, interpSpeedIncreasing=7}
    boneToModify {boneName=dyn_shirt_mid_r_03}
    constraintSetup [{targetBone {boneName=spine_02}}]
    constraintWeights [1]
   }
   n85 = "Constraint - dyn_shirt_bk_r_03 (P)"(ComponentPose=@n84, Alpha=1, ConstraintWeights_0=1) {
    alphaBoolBlend {blendOption=Linear}
    alphaScaleBiasClamp {bClampResult=true, bInterpResult=true, clampMax=0.3, interpSpeedDecreasing=5, interpSpeedIncreasing=7}
    boneToModify {boneName=dyn_shirt_bk_r_03}
    constraintSetup [{targetBone {boneName=spine_02}}]
    constraintWeights [1]
   }
   n86 = "Constraint - dyn_shirt_mid_l_03 (P)"(ComponentPose=@n85, Alpha=1, ConstraintWeights_0=1) {
    alphaBoolBlend {blendOption=Linear}
    alphaScaleBiasClamp {bClampResult=true, bInterpResult=true, clampMax=0.3, interpSpeedDecreasing=5, interpSpeedIncreasing=7}
    boneToModify {boneName=dyn_shirt_mid_l_03}
    constraintSetup [{targetBone {boneName=spine_02}}]
    constraintWeights [1]
   }
   n87 = "Constraint - dyn_shirt_bk_l_03 (P)"(ComponentPose=@n86, Alpha=1, ConstraintWeights_0=1) {
    alphaBoolBlend {blendOption=Linear}
    alphaScaleBiasClamp {bClampResult=true, bInterpResult=true, clampMax=0.3, interpSpeedDecreasing=5, interpSpeedIncreasing=7}
    boneToModify {boneName=dyn_shirt_bk_l_03}
    constraintSetup [{targetBone {boneName=spine_02}}]
    constraintWeights [1]
   }
   n88 = "Constraint - dyn_shirt_mid_r_03 (P)"(ComponentPose=@n87, Alpha=1, ConstraintWeights_0=1) {
    alphaBoolBlend {blendOption=Linear}
    alphaScaleBiasClamp {bClampResult=true, bInterpResult=true, clampMax=0.3, interpSpeedDecreasing=5, interpSpeedIncreasing=7}
    boneToModify {boneName=dyn_shirt_mid_r_03}
    constraintSetup [{targetBone {boneName=spine_02}}]
    constraintWeights [1]
   }
   n89 = "Constraint - dyn_shirt_bk_r_03 (P)"(ComponentPose=@n88, Alpha=1, ConstraintWeights_0=1) {
    alphaBoolBlend {blendOption=Linear}
    alphaScaleBiasClamp {bClampResult=true, bInterpResult=true, clampMax=0.3, interpSpeedDecreasing=5, interpSpeedIncreasing=7}
    boneToModify {boneName=dyn_shirt_bk_r_03}
    constraintSetup [{targetBone {boneName=spine_02}}]
    constraintWeights [1]
   }
   n90 = "Transform (Modify) Bone - Bone: dyn_shirt_fr_l_MSTR"!(ComponentPose=@n89, Alpha=1, Rotation="63.287,-0.478,1.895") -> (Pose) {
    rotationMode=BMM_Additive
    rotationSpace=BCS_BoneSpace
    scaleMode=BMM_Additive
    scaleSpace=BCS_BoneSpace
    translationMode=BMM_Additive
    translationSpace=BCS_BoneSpace
    alphaBoolBlend {blendOption=Linear}
    alphaScaleBiasClamp {bInterpResult=true, interpSpeedIncreasing=7}
    boneToModify {boneName=dyn_shirt_fr_l_MSTR}
    rotation=<rot 63.287 -0.478 1.895>
   }
   n91 = "Transform (Modify) Bone - Bone: dyn_shirt_mid_l_MSTR"(ComponentPose=@n90, Alpha=1, Rotation="15.547,-8.222,4.169") {
    rotationMode=BMM_Additive
    rotationSpace=BCS_BoneSpace
    scaleSpace=BCS_BoneSpace
    translationMode=BMM_Additive
    translationSpace=BCS_BoneSpace
    alphaBoolBlend {blendOption=Linear}
    alphaScaleBiasClamp {bInterpResult=true, interpSpeedIncreasing=7}
    boneToModify {boneName=dyn_shirt_mid_l_MSTR}
    rotation=<rot 15.547 -8.222 4.1689>
   }
   n92 = "Transform (Modify) Bone - Bone: dyn_sleeve_r_fr"!(ComponentPose=@n91, Alpha=1, Translation="-0.003,-1.714,0.397") -> (Pose) {
    rotationMode=BMM_Additive
    rotationSpace=BCS_BoneSpace
    scaleMode=BMM_Additive
    scaleSpace=BCS_BoneSpace
    translationMode=BMM_Additive
    translationSpace=BCS_BoneSpace
    alphaBoolBlend {blendOption=Linear}
    alphaScaleBiasClamp {bInterpResult=true, interpSpeedIncreasing=7}
    boneToModify {boneName=dyn_sleeve_r_fr}
    translation=<-0.003 -1.714 0.397>
   }
   n93 = "Transform (Modify) Bone - Bone: dyn_sleeve_r_top"!(ComponentPose=@n92, Alpha=1, Translation="-0.000,-0.373,-0.110") -> (Pose) {
    rotationMode=BMM_Additive
    rotationSpace=BCS_BoneSpace
    scaleSpace=BCS_BoneSpace
    translationMode=BMM_Additive
    translationSpace=BCS_BoneSpace
    alphaBoolBlend {blendOption=Linear}
    alphaScaleBiasClamp {bInterpResult=true, interpSpeedIncreasing=7}
    boneToModify {boneName=dyn_sleeve_r_top}
    rotation=<rot 0 0 0>
    translation {y=-0.373, z=-0.11}
   }
   n94 = "Transform (Modify) Bone - Bone: dyn_sleeve_r_bk"!(ComponentPose=@n93, Alpha=1, Translation="0.007,1.210,-2.510") -> (Pose) {
    rotationMode=BMM_Additive
    rotationSpace=BCS_BoneSpace
    scaleMode=BMM_Additive
    scaleSpace=BCS_BoneSpace
    translationMode=BMM_Additive
    translationSpace=BCS_BoneSpace
    alphaBoolBlend {blendOption=Linear}
    alphaScaleBiasClamp {bInterpResult=true, interpSpeedIncreasing=7}
    boneToModify {boneName=dyn_sleeve_r_bk}
    translation=<0.007 1.21 -2.51>
   }
   n95 = "Transform (Modify) Bone - Bone: dyn_sleeve_r_bot"!(ComponentPose=@n94, Alpha=1, Translation="0.014,-3.548,3.034") -> (Pose) {
    rotationMode=BMM_Additive
    rotationSpace=BCS_BoneSpace
    scaleSpace=BCS_BoneSpace
    translationMode=BMM_Additive
    translationSpace=BCS_BoneSpace
    alphaBoolBlend {blendOption=Linear}
    alphaScaleBiasClamp {bInterpResult=true, interpSpeedIncreasing=7}
    boneToModify {boneName=dyn_sleeve_r_bot}
    translation=<0.014 -3.548 3.034>
   }
   n96 = "Transform (Modify) Bone - Bone: dyn_shoulder_r_bk"!(ComponentPose=@n95, Alpha=1, Translation="-0.000,-2.979,0.000") -> (Pose) {
    rotationMode=BMM_Additive
    rotationSpace=BCS_BoneSpace
    scaleSpace=BCS_BoneSpace
    translationMode=BMM_Additive
    translationSpace=BCS_BoneSpace
    alphaBoolBlend {blendOption=Linear}
    alphaScaleBiasClamp {bInterpResult=true, interpSpeedIncreasing=7}
    boneToModify {boneName=dyn_shoulder_r_bk}
    translation {y=-2.979}
   }
   n97 = "Constraint - dyn_shirt_mid_l_04 (P)"(ComponentPose=@n96, Alpha=1, ConstraintWeights_0=1) {
    alphaBoolBlend {blendOption=Linear}
    alphaScaleBiasClamp {bInterpResult=true, clampMax=0.3, interpSpeedDecreasing=5, interpSpeedIncreasing=7}
    boneToModify {boneName=dyn_shirt_mid_l_04}
    constraintSetup [{targetBone {boneName=thigh_l}}]
    constraintWeights [1]
   }
   n98 = "Constraint - dyn_shirt_fr_l_04 (P)"(ComponentPose=@n97, Alpha=1, ConstraintWeights_0=1) {
    alphaBoolBlend {blendOption=Linear}
    alphaScaleBiasClamp {bInterpResult=true, clampMax=0.3, interpSpeedDecreasing=5, interpSpeedIncreasing=7}
    boneToModify {boneName=dyn_shirt_fr_l_04}
    constraintSetup [{targetBone {boneName=thigh_l}}]
    constraintWeights [1]
   }
   n99 = "Constraint - dyn_shirt_mid_r_03 (P)"(ComponentPose=@n98, Alpha=1, ConstraintWeights_0=1) {
    alphaBoolBlend {blendOption=Linear}
    alphaScaleBiasClamp {bInterpResult=true, clampMax=0.3, interpSpeedDecreasing=5, interpSpeedIncreasing=7}
    boneToModify {boneName=dyn_shirt_mid_r_03}
    constraintSetup [{targetBone {boneName=thigh_r}}]
    constraintWeights [1]
   }
   n100 = "Constraint - dyn_shirt_fr_r_04 (P)"(ComponentPose=@n99, Alpha=1, ConstraintWeights_0=1) {
    alphaBoolBlend {blendOption=Linear}
    alphaScaleBiasClamp {bInterpResult=true, clampMax=0.3, interpSpeedDecreasing=5, interpSpeedIncreasing=7}
    boneToModify {boneName=dyn_shirt_fr_r_04}
    constraintSetup [{targetBone {boneName=thigh_r}}]
    constraintWeights [1]
   }
   Output Pose(Result=Component To Local(ComponentPose=@n100)) {layerGroup=None}
   n101 = "Constraint - dyn_shoulder_l_bk (PPPPPPPP)"(
    ComponentPose=Local To Component(LocalPose=Use cached pose 'Input'())
    Alpha=1
    ConstraintWeights_0=0.3247
    ConstraintWeights_1=0.0563
    ConstraintWeights_2=0.1149
    ConstraintWeights_3=0.4799
    ConstraintWeights_4=0.0057
    ConstraintWeights_5=0.0066
    ConstraintWeights_6=0.006
    ConstraintWeights_7=0.0057
   ) {
    alphaBoolBlend {blendOption=Linear}
    alphaScaleBiasClamp {bInterpResult=true, interpSpeedIncreasing=7}
    boneToModify {boneName=dyn_shoulder_l_bk}
    constraintSetup [
     {targetBone {boneName=upperarm_twist_01_l}}
     {targetBone {boneName=spine_04}}
     {targetBone {boneName=spine_05}}
     {targetBone {boneName=clavicle_l}}
     {targetBone {boneName=dyn_shirt_mid_l_01}}
     {targetBone {boneName=dyn_shirt_mid_l_02}}
     {targetBone {boneName=dyn_shirt_bk_l_01}}
     {targetBone {boneName=dyn_shirt_bk_l_02}}
    ]
    constraintWeights [1,1,1,1,1,1,1,1]
   }
   n102 = "Constraint - dyn_shoulder_r_bk (PPPPPPP)"(
    ComponentPose=@n101
    Alpha=1
    ConstraintWeights_0=0
    ConstraintWeights_1=0.0466
    ConstraintWeights_2=0.0003
    ConstraintWeights_3=0.1514
    ConstraintWeights_4=0.0023
    ConstraintWeights_5=0.4705
    ConstraintWeights_6=0.3287
   ) {
    alphaBoolBlend {blendOption=Linear}
    alphaScaleBiasClamp {bInterpResult=true, interpSpeedIncreasing=7}
    boneToModify {boneName=dyn_shoulder_r_bk}
    constraintSetup [
     {targetBone {boneName=spine_02}}
     {targetBone {boneName=spine_04}}
     {targetBone {boneName=spine_03}}
     {targetBone {boneName=spine_05}}
     {targetBone {boneName=clavicle_l}}
     {targetBone {boneName=clavicle_r}}
     {targetBone {boneName=upperarm_twist_01_r}}
    ]
    constraintWeights [1,1,1,1,1,1,1]
   }
   n103 = "Constraint - dyn_shoulder_l_fr (PPPPPPPP)"(
    ComponentPose=@n102
    Alpha=1
    ConstraintWeights_0=0.6436
    ConstraintWeights_1=0.0394
    ConstraintWeights_2=0.0591
    ConstraintWeights_3=0.2512
    ConstraintWeights_4=0.0049
    ConstraintWeights_5=0.0003
    ConstraintWeights_6=0.001
    ConstraintWeights_7=0.0005
   ) {
    alphaBoolBlend {blendOption=Linear}
    alphaScaleBiasClamp {bInterpResult=true, interpSpeedIncreasing=7}
    boneToModify {boneName=dyn_shoulder_l_fr}
    constraintSetup [
     {targetBone {boneName=upperarm_twist_01_l}}
     {targetBone {boneName=spine_04}}
     {targetBone {boneName=spine_05}}
     {targetBone {boneName=clavicle_l}}
     {targetBone {boneName=dyn_shirt_fr_l_01}}
     {targetBone {boneName=dyn_shirt_fr_l_02}}
     {targetBone {boneName=dyn_shirt_mid_l_01}}
     {targetBone {boneName=dyn_shirt_mid_l_02}}
    ]
    constraintWeights [1,1,1,1,1,1,1,1]
   }
   n104 = "Constraint - dyn_shoulder_r_fr (PPPP)"(ComponentPose=@n103, Alpha=1, ConstraintWeights_0=0.0177, ConstraintWeights_1=0.0884, ConstraintWeights_2=0.3037, ConstraintWeights_3=0.5902) {
    alphaBoolBlend {blendOption=Linear}
    alphaScaleBiasClamp {bInterpResult=true, interpSpeedIncreasing=7}
    boneToModify {boneName=dyn_shoulder_r_fr}
    constraintSetup [
     {targetBone {boneName=spine_04}}
     {targetBone {boneName=spine_05}}
     {targetBone {boneName=clavicle_r}}
     {targetBone {boneName=upperarm_twist_01_r}}
    ]
    constraintWeights [1,1,1,1]
   }
   n105 = "Transform (Modify) Bone - Bone: dyn_shoulder_l"!(ComponentPose=@n104, Alpha=1, Translation:Vector="1.792,6.075,-5.722", Rotation:Rotator="1.727,16.420,68.391", Scale="1,1.2,1") -> (Pose) {
    rotationMode=BMM_Additive
    rotationSpace=BCS_BoneSpace
    scaleMode=BMM_Additive
    scaleSpace=BCS_BoneSpace
    translationMode=BMM_Additive
    translationSpace=BCS_BoneSpace
    alphaBoolBlend {blendOption=Linear}
    alphaScaleBiasClamp {bInterpResult=true, interpSpeedIncreasing=7}
    boneToModify {boneName=dyn_shoulder_l}
    rotation=<rot 1.727 16.42 68.391>
    scale {y=1.2}
    translation=<1.792 6.075 -5.722>
   }
   n106 = "Transform (Modify) Bone - Bone: dyn_shoulder_l_bk"!(ComponentPose=@n105, Alpha=1, Translation:Vector="6.011,4.301,0.343", Rotation="0.865,44.539,4.538") -> (Pose) {
    rotationMode=BMM_Additive
    rotationSpace=BCS_BoneSpace
    scaleSpace=BCS_BoneSpace
    translationMode=BMM_Additive
    translationSpace=BCS_BoneSpace
    alphaBoolBlend {blendOption=Linear}
    alphaScaleBiasClamp {bInterpResult=true, interpSpeedIncreasing=7}
    boneToModify {boneName=dyn_shoulder_l_bk}
    rotation=<rot 0.865 44.539 4.538>
    scale {y=0}
    translation=<6.011 4.301 0.343>
   }
   n107 = Look At - Bone: dyn_shirt_fr_l_02(ComponentPose=@n106, Alpha=1) {
    alphaBoolBlend {blendOption=Linear}
    alphaScaleBiasClamp {bInterpResult=true, interpSpeedDecreasing=5, interpSpeedIncreasing=7}
    boneToModify {boneName=dyn_shirt_fr_l_02}
    lookAtLocation=<-106.8791 -9.7035 -711.1313>
    lookAtTarget {boneReference {boneName=spine_03}}
    lookAt_Axis {axis=<1 0>}
   }
   n108 = Look At - Bone: dyn_shirt_mid_l_MSTR(ComponentPose=@n107, Alpha=1) {
    alphaBoolBlend {blendOption=Linear}
    alphaScaleBiasClamp {bInterpResult=true, interpSpeedDecreasing=5, interpSpeedIncreasing=7}
    boneToModify {boneName=dyn_shirt_mid_l_MSTR}
    lookAtLocation=<-101.134 -373.7236 -490.5109>
    lookAtTarget {boneReference {boneName=spine_03}}
    lookAt_Axis {axis=<1 0>}
   }
   n109 = Look At - Bone: dyn_shirt_bk_l_MSTR(ComponentPose=@n108, Alpha=1) {
    alphaBoolBlend {blendOption=Linear}
    alphaScaleBiasClamp {bInterpResult=true, interpSpeedDecreasing=5, interpSpeedIncreasing=7}
    boneToModify {boneName=dyn_shirt_bk_l_MSTR}
    lookAtLocation=<-94.8149 -595.0538 -59.7493>
    lookAtTarget {boneReference {boneName=pelvis}}
    lookAt_Axis {axis=<1 0>}
   }
   n110 = Look At - Bone: dyn_shirt_fr_r_MSTR(ComponentPose=@n109, Alpha=1) {
    alphaBoolBlend {blendOption=Linear}
    alphaScaleBiasClamp {bInterpResult=true, interpSpeedDecreasing=5, interpSpeedIncreasing=7}
    boneToModify {boneName=dyn_shirt_fr_r_MSTR}
    lookAtLocation=<-209.01 -22.7032 630.5749>
    lookAtTarget {boneReference {boneName=spine_03}}
    lookAt_Axis {axis=<1 0>}
   }
   n111 = Look At - Bone: dyn_shirt_mid_r_MSTR(ComponentPose=@n110, Alpha=1) {
    alphaBoolBlend {blendOption=Linear}
    alphaScaleBiasClamp {bInterpResult=true, interpSpeedDecreasing=5, interpSpeedIncreasing=7}
    boneToModify {boneName=dyn_shirt_mid_r_MSTR}
    lookAtLocation=<-238.8847 -346.5756 518.2114>
    lookAtTarget {boneReference {boneName=spine_03}}
    lookAt_Axis {axis=<1 0>}
   }
   n112 = Look At - Bone: dyn_shirt_bk_r_MSTR(ComponentPose=@n111, Alpha=1) {
    alphaBoolBlend {blendOption=Linear}
    alphaScaleBiasClamp {bInterpResult=true, interpSpeedDecreasing=5, interpSpeedIncreasing=7}
    boneToModify {boneName=dyn_shirt_bk_r_MSTR}
    lookAtLocation=<-86.7403 -367.3367 41.5077>
    lookAtTarget {boneReference {boneName=pelvis}}
    lookAt_Axis {axis=<1 0>}
   }
   n113 = Component To Local(ComponentPose=@n112)
   n114 = Control Rig!(
    Source=Use cached pose 'Input'()
    Alpha
    Crouch=Get bIsCrouching()
    Forward=Get Forward()
    Backward:double=Get Backward()
    Left=Get Left()
    Right=Get Right()
    Jump:double=Get Jumping()
    Crouch Idle:bool=Get Crouch_Idle() -> (Crouch_Idle:bool)
    DBNO:bool=Get bIsDBNO()
    Disable Front End:bool=Get Front End Off() -> (Front End Off:bool)
   ) -> (Pose) {
    alphaBoolBlend {blendOption=Linear}
    controlRigAssetReference {blueprintRigClass=&/F_MED_Chainmail_Armor_CtrlRig.F_MED_Chainmail_Armor_CtrlRig_C}
   }
   n115 = Get bIsSkyDiving()
   n116 = Anim Dynamics - Bone: dyn_sleeve_r_fr(
    ComponentPose=Local To Component(LocalPose=Blend Poses by bool(BlendPose_0=@n113, BlendPose_1=@n114, bActiveValue=@n115) {blendType=Linear})
    Alpha=1
    ExternalForce="0,0,0"
   ) {
    angularDampingOverride=1
    angularSpringConstant=20000
    bAngularSpring=true
    bOverrideAngularDamping=true
    bUseGravityOverride=true
    gravityScale=0.35
    simulationSpace=RootRelative
    alphaBoolBlend {blendOption=Linear}
    boundBone {boneName=dyn_sleeve_r_fr}
    componentLinearAccScale=<0.7 0.7 0.7>
    componentLinearVelScale=<0.7 0.7 0.7>
    gravityOverride {z=-980}
    physicsBodyDefinitions [{
      sphereCollisionRadius=0
      boundBone {boneName=dyn_sleeve_r_fr}
      boxExtents=<20 30 30>
      constraintSetup {angularLimitsMax=<0 5 10>, angularLimitsMin=<0 -2 -10>, angularTarget=<1 0 0>}
      localJointOffset=<4 0 0>
     }]
   }
   n117 = Anim Dynamics - Bone: dyn_sleeve_r_top(ComponentPose=@n116, Alpha=1, ExternalForce="0,0,0") {
    angularDampingOverride=1
    angularSpringConstant=20000
    bAngularSpring=true
    bOverrideAngularDamping=true
    bUseGravityOverride=true
    gravityScale=0.35
    simulationSpace=RootRelative
    alphaBoolBlend {blendOption=Linear}
    boundBone {boneName=dyn_sleeve_r_top}
    componentLinearAccScale=<0.7 0.7 0.7>
    componentLinearVelScale=<0.7 0.7 0.7>
    gravityOverride {z=-980}
    physicsBodyDefinitions [{
      sphereCollisionRadius=0
      boundBone {boneName=dyn_sleeve_r_top}
      boxExtents=<20 30 30>
      constraintSetup {angularLimitsMax=<0 5 10>, angularLimitsMin=<0 -2 -10>, angularTarget=<1 0 0>}
      localJointOffset=<4 0 0>
     }]
   }
   n118 = Anim Dynamics - Bone: dyn_sleeve_r_bk(ComponentPose=@n117, Alpha=1, ExternalForce="0,0,0") {
    angularDampingOverride=1
    angularSpringConstant=20000
    bAngularSpring=true
    bOverrideAngularDamping=true
    bUseGravityOverride=true
    gravityScale=0.35
    simulationSpace=RootRelative
    alphaBoolBlend {blendOption=Linear}
    boundBone {boneName=dyn_sleeve_r_bk}
    componentLinearAccScale=<0.7 0.7 0.7>
    componentLinearVelScale=<0.7 0.7 0.7>
    gravityOverride {z=-980}
    physicsBodyDefinitions [{
      sphereCollisionRadius=0
      boundBone {boneName=dyn_sleeve_r_bk}
      boxExtents=<20 30 30>
      constraintSetup {angularLimitsMax=<0 5 8>, angularLimitsMin=<0 -5 -8>, angularTarget=<1 0 0>}
      localJointOffset=<4 0 0>
     }]
   }
   n119 = Anim Dynamics - Bone: dyn_sleeve_r_bot(ComponentPose=@n118, Alpha=1, ExternalForce="0,0,0") {
    angularDampingOverride=1
    angularSpringConstant=20000
    bAngularSpring=true
    bOverrideAngularDamping=true
    bUseGravityOverride=true
    gravityScale=0.35
    simulationSpace=RootRelative
    alphaBoolBlend {blendOption=Linear}
    boundBone {boneName=dyn_sleeve_r_bot}
    componentLinearAccScale=<0.7 0.7 0.7>
    componentLinearVelScale=<0.7 0.7 0.7>
    gravityOverride {z=-980}
    physicsBodyDefinitions [{
      sphereCollisionRadius=0
      boundBone {boneName=dyn_sleeve_r_bot}
      boxExtents=<20 30 30>
      constraintSetup {angularLimitsMax=<0 5 8>, angularLimitsMin=<0 -5 -8>, angularTarget=<1 0 0>}
      localJointOffset=<4 0 0>
     }]
   }
   n120 = Anim Dynamics - Bone: dyn_shoelace_l_out_01!(ComponentPose=@n119, bAlphaBoolEnabled:bool=True, GravityOverride:Vector="0,0,-980", ExternalForce="0,0,0") -> (Pose) {
    alphaInputType=Bool
    angularDampingOverride=0.99
    angularSpringConstant=100
    bAngularSpring=true
    bOverrideAngularDamping=true
    bUseGravityOverride=true
    bUseSphericalLimits=true
    simulationSpace=RootRelative
    alphaBoolBlend {blendOption=Linear}
    boundBone {boneName=dyn_shoelace_l_out_01}
    componentLinearAccScale=<0.3 0.3 0.3>
    componentLinearVelScale=<0.3 0.3 0.3>
    gravityOverride {z=-980}
    physicsBodyDefinitions [{
      collisionType=CustomSphere
      sphereCollisionRadius=4.5
      boundBone {boneName=dyn_shoelace_l_out_01}
      boxExtents=<6 15 15>
      constraintSetup {angularLimitsMax=<1 20 20>, angularLimitsMin=<-1 -20 -20>, angularTarget=<1 0 0>}
      localJointOffset=<2 0 0>
     }]
    relativeSpaceBone {boneName=pelvis}
    sphericalLimits [
     {limitRadius=5, drivingBone {boneName=calf_l}, sphereLocalOffset=<-10 0 0>}
     {limitRadius=8, drivingBone {boneName=foot_l}, sphereLocalOffset=<0 0 0>}
    ]
   }
   n121 = Anim Dynamics - Bone: dyn_shoelace_l_out_02!(ComponentPose=@n120, bAlphaBoolEnabled:bool=True, GravityOverride:Vector="0,0,-980", ExternalForce="0,0,0") -> (Pose) {
    alphaInputType=Bool
    angularDampingOverride=0.99
    angularSpringConstant=100
    bAngularSpring=true
    bOverrideAngularDamping=true
    bUseGravityOverride=true
    bUseSphericalLimits=true
    simulationSpace=RootRelative
    alphaBoolBlend {blendOption=Linear}
    boundBone {boneName=dyn_shoelace_l_out_02}
    componentLinearAccScale=<0.3 0.3 0.3>
    componentLinearVelScale=<0.3 0.3 0.3>
    gravityOverride {z=-980}
    physicsBodyDefinitions [{
      collisionType=CustomSphere
      sphereCollisionRadius=4.5
      boundBone {boneName=dyn_shoelace_l_out_02}
      boxExtents=<6 15 15>
      constraintSetup {angularLimitsMax=<1 20 20>, angularLimitsMin=<-1 -20 -20>, angularTarget=<1 0 0>}
      localJointOffset=<2 0 0>
     }]
    relativeSpaceBone {boneName=pelvis}
    sphericalLimits [
     {limitRadius=5, drivingBone {boneName=calf_l}, sphereLocalOffset=<-10 0 0>}
     {limitRadius=8, drivingBone {boneName=foot_l}, sphereLocalOffset=<0 0 0>}
    ]
   }
   n122 = Anim Dynamics - Bone: dyn_shoelace_l_out_03!(ComponentPose=@n121, bAlphaBoolEnabled:bool=True, GravityOverride:Vector="0,0,-980", ExternalForce="0,0,0") -> (Pose) {
    alphaInputType=Bool
    angularDampingOverride=0.99
    angularSpringConstant=100
    bAngularSpring=true
    bOverrideAngularDamping=true
    bUseGravityOverride=true
    bUseSphericalLimits=true
    simulationSpace=RootRelative
    alphaBoolBlend {blendOption=Linear}
    boundBone {boneName=dyn_shoelace_l_out_03}
    componentLinearAccScale=<0.3 0.3 0.3>
    componentLinearVelScale=<0.3 0.3 0.3>
    gravityOverride {z=-980}
    physicsBodyDefinitions [{
      collisionType=CustomSphere
      sphereCollisionRadius=4.5
      boundBone {boneName=dyn_shoelace_l_out_03}
      boxExtents=<6 15 15>
      constraintSetup {angularLimitsMax=<1 20 20>, angularLimitsMin=<-1 -20 -20>, angularTarget=<1 0 0>}
      localJointOffset=<2 0 0>
     }]
    relativeSpaceBone {boneName=pelvis}
    sphericalLimits [
     {limitRadius=5, drivingBone {boneName=calf_l}, sphereLocalOffset=<-10 0 0>}
     {limitRadius=8, drivingBone {boneName=foot_l}, sphereLocalOffset=<0 0 0>}
    ]
   }
   n123 = Anim Dynamics - Bone: dyn_shoelace_l_out_04!(ComponentPose=@n122, bAlphaBoolEnabled:bool=True, GravityOverride:Vector="0,0,-980", ExternalForce="0,0,0") -> (Pose) {
    alphaInputType=Bool
    angularDampingOverride=0.99
    angularSpringConstant=100
    bAngularSpring=true
    bOverrideAngularDamping=true
    bUseGravityOverride=true
    bUseSphericalLimits=true
    simulationSpace=RootRelative
    alphaBoolBlend {blendOption=Linear}
    boundBone {boneName=dyn_shoelace_l_out_04}
    componentLinearAccScale=<0.3 0.3 0.3>
    componentLinearVelScale=<0.3 0.3 0.3>
    gravityOverride {z=-980}
    physicsBodyDefinitions [{
      collisionType=CustomSphere
      sphereCollisionRadius=4.5
      boundBone {boneName=dyn_shoelace_l_out_04}
      boxExtents=<6 15 15>
      constraintSetup {angularLimitsMax=<1 20 20>, angularLimitsMin=<-1 -20 -20>, angularTarget=<1 0 0>}
      localJointOffset=<2 0 0>
     }]
    relativeSpaceBone {boneName=pelvis}
    sphericalLimits [
     {limitRadius=5, drivingBone {boneName=calf_l}, sphereLocalOffset=<-10 0 0>}
     {limitRadius=8, drivingBone {boneName=foot_l}, sphereLocalOffset=<0 0 0>}
    ]
   }
   n124 = Anim Dynamics - Bone: dyn_shoelace_l_out_05!(ComponentPose=@n123, bAlphaBoolEnabled:bool=True, GravityOverride:Vector="0,0,-980", ExternalForce="0,0,0") -> (Pose) {
    alphaInputType=Bool
    angularDampingOverride=0.99
    angularSpringConstant=100
    bAngularSpring=true
    bOverrideAngularDamping=true
    bUseGravityOverride=true
    bUseSphericalLimits=true
    simulationSpace=RootRelative
    alphaBoolBlend {blendOption=Linear}
    boundBone {boneName=dyn_shoelace_l_out_05}
    componentLinearAccScale=<0.3 0.3 0.3>
    componentLinearVelScale=<0.3 0.3 0.3>
    gravityOverride {z=-980}
    physicsBodyDefinitions [{
      collisionType=CustomSphere
      sphereCollisionRadius=4.5
      boundBone {boneName=dyn_shoelace_l_out_05}
      boxExtents=<6 15 15>
      constraintSetup {angularLimitsMax=<1 20 20>, angularLimitsMin=<-1 -20 -20>, angularTarget=<1 0 0>}
      localJointOffset=<2 0 0>
     }]
    relativeSpaceBone {boneName=pelvis}
    sphericalLimits [
     {limitRadius=5, drivingBone {boneName=calf_l}, sphereLocalOffset=<-10 0 0>}
     {limitRadius=8, drivingBone {boneName=foot_l}, sphereLocalOffset=<0 0 0>}
    ]
   }
   n125 = Anim Dynamics - Bone: dyn_shoelace_l_in_01!(ComponentPose=@n124, bAlphaBoolEnabled:bool=True, GravityOverride:Vector="0,0,-980", ExternalForce="0,0,0") -> (Pose) {
    alphaInputType=Bool
    angularDampingOverride=0.99
    angularSpringConstant=100
    bAngularSpring=true
    bOverrideAngularDamping=true
    bUseGravityOverride=true
    bUseSphericalLimits=true
    simulationSpace=RootRelative
    alphaBoolBlend {blendOption=Linear}
    boundBone {boneName=dyn_shoelace_l_in_01}
    componentLinearAccScale=<0.3 0.3 0.3>
    componentLinearVelScale=<0.3 0.3 0.3>
    gravityOverride {z=-980}
    physicsBodyDefinitions [{
      collisionType=CustomSphere
      sphereCollisionRadius=4.5
      boundBone {boneName=dyn_shoelace_l_in_01}
      boxExtents=<6 15 15>
      constraintSetup {angularLimitsMax=<1 20 20>, angularLimitsMin=<-1 -20 -20>, angularTarget=<1 0 0>}
      localJointOffset=<2 0 0>
     }]
    relativeSpaceBone {boneName=pelvis}
    sphericalLimits [
     {limitRadius=5, drivingBone {boneName=calf_l}, sphereLocalOffset=<-10 0 0>}
     {limitRadius=8, drivingBone {boneName=foot_l}, sphereLocalOffset=<0 0 0>}
    ]
   }
   n126 = Anim Dynamics - Bone: dyn_shoelace_l_in_02!(ComponentPose=@n125, bAlphaBoolEnabled:bool=True, GravityOverride:Vector="0,0,-980", ExternalForce="0,0,0") -> (Pose) {
    alphaInputType=Bool
    angularDampingOverride=0.99
    angularSpringConstant=100
    bAngularSpring=true
    bOverrideAngularDamping=true
    bUseGravityOverride=true
    bUseSphericalLimits=true
    simulationSpace=RootRelative
    alphaBoolBlend {blendOption=Linear}
    boundBone {boneName=dyn_shoelace_l_in_02}
    componentLinearAccScale=<0.3 0.3 0.3>
    componentLinearVelScale=<0.3 0.3 0.3>
    gravityOverride {z=-980}
    physicsBodyDefinitions [{
      collisionType=CustomSphere
      sphereCollisionRadius=4.5
      boundBone {boneName=dyn_shoelace_l_in_02}
      boxExtents=<6 15 15>
      constraintSetup {angularLimitsMax=<1 20 20>, angularLimitsMin=<-1 -20 -20>, angularTarget=<1 0 0>}
      localJointOffset=<2 0 0>
     }]
    relativeSpaceBone {boneName=pelvis}
    sphericalLimits [
     {limitRadius=5, drivingBone {boneName=calf_l}, sphereLocalOffset=<-10 0 0>}
     {limitRadius=8, drivingBone {boneName=foot_l}, sphereLocalOffset=<0 0 0>}
    ]
   }
   n127 = Anim Dynamics - Bone: dyn_shoelace_l_in_03!(ComponentPose=@n126, bAlphaBoolEnabled:bool=True, GravityOverride:Vector="0,0,-980", ExternalForce="0,0,0") -> (Pose) {
    alphaInputType=Bool
    angularDampingOverride=0.99
    angularSpringConstant=100
    bAngularSpring=true
    bOverrideAngularDamping=true
    bUseGravityOverride=true
    bUseSphericalLimits=true
    simulationSpace=RootRelative
    alphaBoolBlend {blendOption=Linear}
    boundBone {boneName=dyn_shoelace_l_in_03}
    componentLinearAccScale=<0.3 0.3 0.3>
    componentLinearVelScale=<0.3 0.3 0.3>
    gravityOverride {z=-980}
    physicsBodyDefinitions [{
      collisionType=CustomSphere
      sphereCollisionRadius=4.5
      boundBone {boneName=dyn_shoelace_l_in_03}
      boxExtents=<6 15 15>
      constraintSetup {angularLimitsMax=<1 20 20>, angularLimitsMin=<-1 -20 -20>, angularTarget=<1 0 0>}
      localJointOffset=<2 0 0>
     }]
    relativeSpaceBone {boneName=pelvis}
    sphericalLimits [
     {limitRadius=5, drivingBone {boneName=calf_l}, sphereLocalOffset=<-10 0 0>}
     {limitRadius=8, drivingBone {boneName=foot_l}, sphereLocalOffset=<0 0 0>}
    ]
   }
   n128 = Anim Dynamics - Bone: dyn_shoelace_l_in_04!(ComponentPose=@n127, bAlphaBoolEnabled:bool=True, GravityOverride:Vector="0,0,-980", ExternalForce="0,0,0") -> (Pose) {
    alphaInputType=Bool
    angularDampingOverride=0.99
    angularSpringConstant=100
    bAngularSpring=true
    bOverrideAngularDamping=true
    bUseGravityOverride=true
    bUseSphericalLimits=true
    simulationSpace=RootRelative
    alphaBoolBlend {blendOption=Linear}
    boundBone {boneName=dyn_shoelace_l_in_04}
    componentLinearAccScale=<0.3 0.3 0.3>
    componentLinearVelScale=<0.3 0.3 0.3>
    gravityOverride {z=-980}
    physicsBodyDefinitions [{
      collisionType=CustomSphere
      sphereCollisionRadius=4.5
      boundBone {boneName=dyn_shoelace_l_in_04}
      boxExtents=<6 15 15>
      constraintSetup {angularLimitsMax=<1 20 20>, angularLimitsMin=<-1 -20 -20>, angularTarget=<1 0 0>}
      localJointOffset=<2 0 0>
     }]
    relativeSpaceBone {boneName=pelvis}
    sphericalLimits [
     {limitRadius=5, drivingBone {boneName=calf_l}, sphereLocalOffset=<-10 0 0>}
     {limitRadius=8, drivingBone {boneName=foot_l}, sphereLocalOffset=<0 0 0>}
    ]
   }
   n129 = Anim Dynamics - Bone: dyn_shoelace_l_in_05!(ComponentPose=@n128, bAlphaBoolEnabled:bool=True, GravityOverride:Vector="0,0,-980", ExternalForce="0,0,0") -> (Pose) {
    alphaInputType=Bool
    angularDampingOverride=0.99
    angularSpringConstant=100
    bAngularSpring=true
    bOverrideAngularDamping=true
    bUseGravityOverride=true
    bUseSphericalLimits=true
    simulationSpace=RootRelative
    alphaBoolBlend {blendOption=Linear}
    boundBone {boneName=dyn_shoelace_l_in_05}
    componentLinearAccScale=<0.3 0.3 0.3>
    componentLinearVelScale=<0.3 0.3 0.3>
    gravityOverride {z=-980}
    physicsBodyDefinitions [{
      collisionType=CustomSphere
      sphereCollisionRadius=4.5
      boundBone {boneName=dyn_shoelace_l_in_05}
      boxExtents=<6 15 15>
      constraintSetup {angularLimitsMax=<1 20 20>, angularLimitsMin=<-1 -20 -20>, angularTarget=<1 0 0>}
      localJointOffset=<2 0 0>
     }]
    relativeSpaceBone {boneName=pelvis}
    sphericalLimits [
     {limitRadius=5, drivingBone {boneName=calf_l}, sphereLocalOffset=<-10 0 0>}
     {limitRadius=8, drivingBone {boneName=foot_l}, sphereLocalOffset=<0 0 0>}
    ]
   }
   n130 = Anim Dynamics - Bone: dyn_shoelace_r_in_01!(ComponentPose=@n129, bAlphaBoolEnabled:bool=True, GravityOverride:Vector="0,0,-980", ExternalForce="0,0,0") -> (Pose) {
    alphaInputType=Bool
    angularDampingOverride=0.99
    angularSpringConstant=100
    bAngularSpring=true
    bOverrideAngularDamping=true
    bUseGravityOverride=true
    bUseSphericalLimits=true
    simulationSpace=RootRelative
    alphaBoolBlend {blendOption=Linear}
    boundBone {boneName=dyn_shoelace_r_in_01}
    componentLinearAccScale=<0.3 0.3 0.3>
    componentLinearVelScale=<0.3 0.3 0.3>
    gravityOverride {z=-980}
    physicsBodyDefinitions [{
      collisionType=CustomSphere
      sphereCollisionRadius=4.5
      boundBone {boneName=dyn_shoelace_r_in_01}
      boxExtents=<6 15 15>
      constraintSetup {angularLimitsMax=<1 20 20>, angularLimitsMin=<-1 -20 -20>, angularTarget=<1 0 0>}
      localJointOffset=<2 0 0>
     }]
    relativeSpaceBone {boneName=thigh_r}
    sphericalLimits [
     {limitRadius=3, drivingBone {boneName=calf_twist_02_r}, sphereLocalOffset=<0 0 0>}
     {limitRadius=8, drivingBone {boneName=foot_r}, sphereLocalOffset=<0 0 0>}
    ]
   }
   n131 = Anim Dynamics - Bone: dyn_shoelace_r_in_02!(ComponentPose=@n130, bAlphaBoolEnabled:bool=True, GravityOverride:Vector="0,0,-980", ExternalForce="0,0,0") -> (Pose) {
    alphaInputType=Bool
    angularDampingOverride=0.99
    angularSpringConstant=100
    bAngularSpring=true
    bOverrideAngularDamping=true
    bUseGravityOverride=true
    bUseSphericalLimits=true
    simulationSpace=RootRelative
    alphaBoolBlend {blendOption=Linear}
    boundBone {boneName=dyn_shoelace_r_in_02}
    componentLinearAccScale=<0.3 0.3 0.3>
    componentLinearVelScale=<0.3 0.3 0.3>
    gravityOverride {z=-980}
    physicsBodyDefinitions [{
      collisionType=CustomSphere
      sphereCollisionRadius=4.5
      boundBone {boneName=dyn_shoelace_r_in_02}
      boxExtents=<6 15 15>
      constraintSetup {angularLimitsMax=<1 20 20>, angularLimitsMin=<-1 -20 -20>, angularTarget=<1 0 0>}
      localJointOffset=<2 0 0>
     }]
    relativeSpaceBone {boneName=thigh_r}
    sphericalLimits [
     {limitRadius=3, drivingBone {boneName=calf_twist_02_r}, sphereLocalOffset=<0 0 0>}
     {limitRadius=8, drivingBone {boneName=foot_r}, sphereLocalOffset=<0 0 0>}
    ]
   }
   n132 = Anim Dynamics - Bone: dyn_shoelace_r_in_03!(ComponentPose=@n131, bAlphaBoolEnabled:bool=True, GravityOverride:Vector="0,0,-980", ExternalForce="0,0,0") -> (Pose) {
    alphaInputType=Bool
    angularDampingOverride=0.99
    angularSpringConstant=100
    bAngularSpring=true
    bOverrideAngularDamping=true
    bUseGravityOverride=true
    bUseSphericalLimits=true
    simulationSpace=RootRelative
    alphaBoolBlend {blendOption=Linear}
    boundBone {boneName=dyn_shoelace_r_in_03}
    componentLinearAccScale=<0.3 0.3 0.3>
    componentLinearVelScale=<0.3 0.3 0.3>
    gravityOverride {z=-980}
    physicsBodyDefinitions [{
      collisionType=CustomSphere
      sphereCollisionRadius=4.5
      boundBone {boneName=dyn_shoelace_r_in_03}
      boxExtents=<6 15 15>
      constraintSetup {angularLimitsMax=<1 20 20>, angularLimitsMin=<-1 -20 -20>, angularTarget=<1 0 0>}
      localJointOffset=<2 0 0>
     }]
    relativeSpaceBone {boneName=thigh_r}
    sphericalLimits [
     {limitRadius=3, drivingBone {boneName=calf_twist_02_r}, sphereLocalOffset=<0 0 0>}
     {limitRadius=8, drivingBone {boneName=foot_r}, sphereLocalOffset=<0 0 0>}
    ]
   }
   n133 = Anim Dynamics - Bone: dyn_shoelace_r_in_04!(ComponentPose=@n132, bAlphaBoolEnabled:bool=True, GravityOverride:Vector="0,0,-980", ExternalForce="0,0,0") -> (Pose) {
    alphaInputType=Bool
    angularDampingOverride=0.99
    angularSpringConstant=100
    bAngularSpring=true
    bOverrideAngularDamping=true
    bUseGravityOverride=true
    bUseSphericalLimits=true
    simulationSpace=RootRelative
    alphaBoolBlend {blendOption=Linear}
    boundBone {boneName=dyn_shoelace_r_in_04}
    componentLinearAccScale=<0.3 0.3 0.3>
    componentLinearVelScale=<0.3 0.3 0.3>
    gravityOverride {z=-980}
    physicsBodyDefinitions [{
      collisionType=CustomSphere
      sphereCollisionRadius=4.5
      boundBone {boneName=dyn_shoelace_r_in_04}
      boxExtents=<6 15 15>
      constraintSetup {angularLimitsMax=<1 20 20>, angularLimitsMin=<-1 -20 -20>, angularTarget=<1 0 0>}
      localJointOffset=<2 0 0>
     }]
    relativeSpaceBone {boneName=thigh_r}
    sphericalLimits [
     {limitRadius=3, drivingBone {boneName=calf_twist_02_r}, sphereLocalOffset=<0 0 0>}
     {limitRadius=8, drivingBone {boneName=foot_r}, sphereLocalOffset=<0 0 0>}
    ]
   }
   n134 = Anim Dynamics - Bone: dyn_shoelace_r_in_05!(ComponentPose=@n133, bAlphaBoolEnabled:bool=True, GravityOverride:Vector="0,0,-980", ExternalForce="0,0,0") -> (Pose) {
    alphaInputType=Bool
    angularDampingOverride=0.99
    angularSpringConstant=100
    bAngularSpring=true
    bOverrideAngularDamping=true
    bUseGravityOverride=true
    bUseSphericalLimits=true
    simulationSpace=RootRelative
    alphaBoolBlend {blendOption=Linear}
    boundBone {boneName=dyn_shoelace_r_in_05}
    componentLinearAccScale=<0.3 0.3 0.3>
    componentLinearVelScale=<0.3 0.3 0.3>
    gravityOverride {z=-980}
    physicsBodyDefinitions [{
      collisionType=CustomSphere
      sphereCollisionRadius=4.5
      boundBone {boneName=dyn_shoelace_r_in_05}
      boxExtents=<6 15 15>
      constraintSetup {angularLimitsMax=<1 20 20>, angularLimitsMin=<-1 -20 -20>, angularTarget=<1 0 0>}
      localJointOffset=<2 0 0>
     }]
    relativeSpaceBone {boneName=thigh_r}
    sphericalLimits [
     {limitRadius=3, drivingBone {boneName=calf_twist_02_r}, sphereLocalOffset=<0 0 0>}
     {limitRadius=8, drivingBone {boneName=foot_r}, sphereLocalOffset=<0 0 0>}
    ]
   }
   n135 = Anim Dynamics - Bone: dyn_shoelace_r_in_06!(ComponentPose=@n134, bAlphaBoolEnabled:bool=True, GravityOverride:Vector="0,0,-980", ExternalForce="0,0,0") -> (Pose) {
    alphaInputType=Bool
    angularDampingOverride=0.99
    angularSpringConstant=100
    bAngularSpring=true
    bOverrideAngularDamping=true
    bUseGravityOverride=true
    bUseSphericalLimits=true
    simulationSpace=RootRelative
    alphaBoolBlend {blendOption=Linear}
    boundBone {boneName=dyn_shoelace_r_in_06}
    componentLinearAccScale=<0.3 0.3 0.3>
    componentLinearVelScale=<0.3 0.3 0.3>
    gravityOverride {z=-980}
    physicsBodyDefinitions [{
      collisionType=CustomSphere
      sphereCollisionRadius=4.5
      boundBone {boneName=dyn_shoelace_r_in_06}
      boxExtents=<6 15 15>
      constraintSetup {angularLimitsMax=<1 20 20>, angularLimitsMin=<-1 -20 -20>, angularTarget=<1 0 0>}
      localJointOffset=<2 0 0>
     }]
    relativeSpaceBone {boneName=thigh_r}
    sphericalLimits [
     {limitRadius=3, drivingBone {boneName=calf_twist_02_r}, sphereLocalOffset=<0 0 0>}
     {limitRadius=8, drivingBone {boneName=foot_r}, sphereLocalOffset=<0 0 0>}
    ]
   }
   n136 = Anim Dynamics - Bone: dyn_shoelace_r_out_01!(ComponentPose=@n135, bAlphaBoolEnabled:bool=True, GravityOverride:Vector="0,0,-980", ExternalForce="0,0,0") -> (Pose) {
    alphaInputType=Bool
    angularDampingOverride=0.99
    angularSpringConstant=100
    bAngularSpring=true
    bOverrideAngularDamping=true
    bUseGravityOverride=true
    bUseSphericalLimits=true
    simulationSpace=RootRelative
    alphaBoolBlend {blendOption=Linear}
    boundBone {boneName=dyn_shoelace_r_out_01}
    componentLinearAccScale=<0.3 0.3 0.3>
    componentLinearVelScale=<0.3 0.3 0.3>
    gravityOverride {z=-980}
    physicsBodyDefinitions [{
      collisionType=CustomSphere
      sphereCollisionRadius=4.5
      boundBone {boneName=dyn_shoelace_r_out_01}
      boxExtents=<6 15 15>
      constraintSetup {angularLimitsMax=<1 20 20>, angularLimitsMin=<-1 -20 -20>, angularTarget=<1 0 0>}
      localJointOffset=<2 0 0>
     }]
    relativeSpaceBone {boneName=thigh_r}
    sphericalLimits [
     {limitRadius=3, drivingBone {boneName=calf_twist_02_r}, sphereLocalOffset=<0 0 0>}
     {limitRadius=8, drivingBone {boneName=foot_r}, sphereLocalOffset=<0 0 0>}
    ]
   }
   n137 = Anim Dynamics - Bone: dyn_shoelace_r_out_02!(ComponentPose=@n136, bAlphaBoolEnabled:bool=True, GravityOverride:Vector="0,0,-980", ExternalForce="0,0,0") -> (Pose) {
    alphaInputType=Bool
    angularDampingOverride=0.99
    angularSpringConstant=100
    bAngularSpring=true
    bOverrideAngularDamping=true
    bUseGravityOverride=true
    bUseSphericalLimits=true
    simulationSpace=RootRelative
    alphaBoolBlend {blendOption=Linear}
    boundBone {boneName=dyn_shoelace_r_out_02}
    componentLinearAccScale=<0.3 0.3 0.3>
    componentLinearVelScale=<0.3 0.3 0.3>
    gravityOverride {z=-980}
    physicsBodyDefinitions [{
      collisionType=CustomSphere
      sphereCollisionRadius=4.5
      boundBone {boneName=dyn_shoelace_r_out_02}
      boxExtents=<6 15 15>
      constraintSetup {angularLimitsMax=<1 20 20>, angularLimitsMin=<-1 -20 -20>, angularTarget=<1 0 0>}
      localJointOffset=<2 0 0>
     }]
    relativeSpaceBone {boneName=thigh_r}
    sphericalLimits [
     {limitRadius=3, drivingBone {boneName=calf_twist_02_r}, sphereLocalOffset=<0 0 0>}
     {limitRadius=8, drivingBone {boneName=foot_r}, sphereLocalOffset=<0 0 0>}
    ]
   }
   n138 = Anim Dynamics - Bone: dyn_shoelace_r_out_03!(ComponentPose=@n137, bAlphaBoolEnabled:bool=True, GravityOverride:Vector="0,0,-980", ExternalForce="0,0,0") -> (Pose) {
    alphaInputType=Bool
    angularDampingOverride=0.99
    angularSpringConstant=100
    bAngularSpring=true
    bOverrideAngularDamping=true
    bUseGravityOverride=true
    bUseSphericalLimits=true
    simulationSpace=RootRelative
    alphaBoolBlend {blendOption=Linear}
    boundBone {boneName=dyn_shoelace_r_out_03}
    componentLinearAccScale=<0.3 0.3 0.3>
    componentLinearVelScale=<0.3 0.3 0.3>
    gravityOverride {z=-980}
    physicsBodyDefinitions [{
      collisionType=CustomSphere
      sphereCollisionRadius=4.5
      boundBone {boneName=dyn_shoelace_r_out_03}
      boxExtents=<6 15 15>
      constraintSetup {angularLimitsMax=<1 20 20>, angularLimitsMin=<-1 -20 -20>, angularTarget=<1 0 0>}
      localJointOffset=<2 0 0>
     }]
    relativeSpaceBone {boneName=thigh_r}
    sphericalLimits [
     {limitRadius=3, drivingBone {boneName=calf_twist_02_r}, sphereLocalOffset=<0 0 0>}
     {limitRadius=8, drivingBone {boneName=foot_r}, sphereLocalOffset=<0 0 0>}
    ]
   }
   n139 = Anim Dynamics - Bone: dyn_shoelace_r_out_04!(ComponentPose=@n138, bAlphaBoolEnabled:bool=True, GravityOverride:Vector="0,0,-980", ExternalForce="0,0,0") -> (Pose) {
    alphaInputType=Bool
    angularDampingOverride=0.99
    angularSpringConstant=100
    bAngularSpring=true
    bOverrideAngularDamping=true
    bUseGravityOverride=true
    bUseSphericalLimits=true
    simulationSpace=RootRelative
    alphaBoolBlend {blendOption=Linear}
    boundBone {boneName=dyn_shoelace_r_out_04}
    componentLinearAccScale=<0.3 0.3 0.3>
    componentLinearVelScale=<0.3 0.3 0.3>
    gravityOverride {z=-980}
    physicsBodyDefinitions [{
      collisionType=CustomSphere
      sphereCollisionRadius=4.5
      boundBone {boneName=dyn_shoelace_r_out_04}
      boxExtents=<6 15 15>
      constraintSetup {angularLimitsMax=<1 20 20>, angularLimitsMin=<-1 -20 -20>, angularTarget=<1 0 0>}
      localJointOffset=<2 0 0>
     }]
    relativeSpaceBone {boneName=thigh_r}
    sphericalLimits [
     {limitRadius=3, drivingBone {boneName=calf_twist_02_r}, sphereLocalOffset=<0 0 0>}
     {limitRadius=8, drivingBone {boneName=foot_r}, sphereLocalOffset=<0 0 0>}
    ]
   }
   n140 = Anim Dynamics - Bone: dyn_shoelace_r_out_05!(ComponentPose=@n139, bAlphaBoolEnabled:bool=True, GravityOverride:Vector="0,0,-980", ExternalForce="0,0,0") -> (Pose) {
    alphaInputType=Bool
    angularDampingOverride=0.99
    angularSpringConstant=100
    bAngularSpring=true
    bOverrideAngularDamping=true
    bUseGravityOverride=true
    bUseSphericalLimits=true
    simulationSpace=RootRelative
    alphaBoolBlend {blendOption=Linear}
    boundBone {boneName=dyn_shoelace_r_out_05}
    componentLinearAccScale=<0.3 0.3 0.3>
    componentLinearVelScale=<0.3 0.3 0.3>
    gravityOverride {z=-980}
    physicsBodyDefinitions [{
      collisionType=CustomSphere
      sphereCollisionRadius=4.5
      boundBone {boneName=dyn_shoelace_r_out_05}
      boxExtents=<6 15 15>
      constraintSetup {angularLimitsMax=<1 20 20>, angularLimitsMin=<-1 -20 -20>, angularTarget=<1 0 0>}
      localJointOffset=<2 0 0>
     }]
    relativeSpaceBone {boneName=thigh_r}
    sphericalLimits [
     {limitRadius=3, drivingBone {boneName=calf_twist_02_r}, sphereLocalOffset=<0 0 0>}
     {limitRadius=8, drivingBone {boneName=foot_r}, sphereLocalOffset=<0 0 0>}
    ]
   }
   n141 = Anim Dynamics - Bone: dyn_shoelace_r_out_06!(ComponentPose=@n140, bAlphaBoolEnabled:bool=True, GravityOverride:Vector="0,0,-980", ExternalForce="0,0,0") -> (Pose) {
    alphaInputType=Bool
    angularDampingOverride=0.99
    angularSpringConstant=100
    bAngularSpring=true
    bOverrideAngularDamping=true
    bUseGravityOverride=true
    bUseSphericalLimits=true
    simulationSpace=RootRelative
    alphaBoolBlend {blendOption=Linear}
    boundBone {boneName=dyn_shoelace_r_out_06}
    componentLinearAccScale=<0.3 0.3 0.3>
    componentLinearVelScale=<0.3 0.3 0.3>
    gravityOverride {z=-980}
    physicsBodyDefinitions [{
      collisionType=CustomSphere
      sphereCollisionRadius=4.5
      boundBone {boneName=dyn_shoelace_r_out_06}
      boxExtents=<6 15 15>
      constraintSetup {angularLimitsMax=<1 20 20>, angularLimitsMin=<-1 -20 -20>, angularTarget=<1 0 0>}
      localJointOffset=<2 0 0>
     }]
    relativeSpaceBone {boneName=thigh_r}
    sphericalLimits [
     {limitRadius=3, drivingBone {boneName=calf_twist_02_r}, sphereLocalOffset=<0 0 0>}
     {limitRadius=8, drivingBone {boneName=foot_r}, sphereLocalOffset=<0 0 0>}
    ]
   }
   n142 = "Constraint - dyn_loop_l_top (PPP)"!(ComponentPose=@n141, Alpha=1, ConstraintWeights_0=0.3178, ConstraintWeights_1:float=0.6616, ConstraintWeights_2:float=0.0206) -> (Pose) {
    alphaBoolBlend {blendOption=Linear}
    boneToModify {boneName=dyn_loop_l_top}
    constraintSetup [{targetBone {boneName=pelvis}}, {targetBone {boneName=thigh_twist_01_l}}, {targetBone {boneName=thigh_l}}]
    constraintWeights [1,1,1]
   }
   n143 = "Constraint - dyn_loop_l_in (PPP)"!(ComponentPose=@n142, Alpha=1, ConstraintWeights_0=0.013, ConstraintWeights_1:float=0.918, ConstraintWeights_2:float=0.069) -> (Pose) {
    alphaBoolBlend {blendOption=Linear}
    boneToModify {boneName=dyn_loop_l_in}
    constraintSetup [{targetBone {boneName=pelvis}}, {targetBone {boneName=thigh_twist_01_l}}, {targetBone {boneName=thigh_l}}]
    constraintWeights [1,1,1]
   }
   n144 = "Constraint - dyn_loop_l_out (PPP)"!(ComponentPose=@n143, Alpha=1, ConstraintWeights_0=0.042, ConstraintWeights_1:float=0.677, ConstraintWeights_2:float=0.281) -> (Pose) {
    alphaBoolBlend {blendOption=Linear}
    boneToModify {boneName=dyn_loop_l_out}
    constraintSetup [{targetBone {boneName=pelvis}}, {targetBone {boneName=thigh_twist_01_l}}, {targetBone {boneName=thigh_l}}]
    constraintWeights [1,1,1]
   }
   n145 = "Constraint - dyn_loop_c (PPP)"!(ComponentPose=@n144, Alpha=1, ConstraintWeights_0=0.96, ConstraintWeights_1:float=0.02, ConstraintWeights_2:float=0.02) -> (Pose) {
    alphaBoolBlend {blendOption=Linear}
    boneToModify {boneName=dyn_loop_c}
    constraintSetup [{targetBone {boneName=pelvis}}, {targetBone {boneName=thigh_twist_01_l}}, {targetBone {boneName=thigh_twist_01_r}}]
    constraintWeights [1,1,1]
   }
   n146 = Anim Dynamics - Bone: dyn_loop_l_top(ComponentPose=@n145, Alpha=1, ExternalForce="0,0,0") {
    angularDampingOverride=0.9999
    angularSpringConstant=750
    bAngularSpring=true
    bGravityOverrideInSimSpace=true
    bOverrideAngularDamping=true
    simulationSpace=BoneRelative
    alphaBoolBlend {blendOption=Linear}
    boundBone {boneName=dyn_loop_l_top}
    componentAppliedLinearAccClamp=<5000 5000 5000>
    componentLinearAccScale=<1.2 1.2 1.2>
    componentLinearVelScale=<1.2 1.2 1.2>
    physicsBodyDefinitions [{
      sphereCollisionRadius=0
      boundBone {boneName=dyn_loop_l_top}
      boxExtents=<12.5 12.5 12.5>
      constraintSetup {angularLimitsMax=<1 3 25>, angularLimitsMin=<-1 -3 0>, angularTarget=<1 0 0>}
      localJointOffset=<1.5 0.1 0>
     }]
    relativeSpaceBone {boneName=root}
   }
   n147 = Anim Dynamics - Bone: dyn_loop_l_in(ComponentPose=@n146, Alpha=1, ExternalForce="0,0,0") {
    angularDampingOverride=0.9999
    angularSpringConstant=750
    bAngularSpring=true
    bGravityOverrideInSimSpace=true
    bOverrideAngularDamping=true
    simulationSpace=BoneRelative
    alphaBoolBlend {blendOption=Linear}
    boundBone {boneName=dyn_loop_l_in}
    componentAppliedLinearAccClamp=<5000 5000 5000>
    componentLinearAccScale=<1.2 1.2 1.2>
    componentLinearVelScale=<1.2 1.2 1.2>
    physicsBodyDefinitions [{
      sphereCollisionRadius=0
      boundBone {boneName=dyn_loop_l_in}
      boxExtents=<12.25 12.25 12.25>
      constraintSetup {angularLimitsMax=<1 3 25>, angularLimitsMin=<-1 -3 0>, angularTarget=<1 0 0>}
      localJointOffset=<1.45 0.1 0>
     }]
    relativeSpaceBone {boneName=root}
   }
   n148 = Anim Dynamics - Bone: dyn_loop_l_out(ComponentPose=@n147, Alpha=1, ExternalForce="0,0,0") {
    angularDampingOverride=0.9999
    angularSpringConstant=750
    bAngularSpring=true
    bGravityOverrideInSimSpace=true
    bOverrideAngularDamping=true
    simulationSpace=BoneRelative
    alphaBoolBlend {blendOption=Linear}
    boundBone {boneName=dyn_loop_l_out}
    componentAppliedLinearAccClamp=<5000 5000 5000>
    componentLinearAccScale=<1.2 1.2 1.2>
    componentLinearVelScale=<1.2 1.2 1.2>
    physicsBodyDefinitions [{
      sphereCollisionRadius=0
      boundBone {boneName=dyn_loop_l_out}
      boxExtents=<12.6 12.6 12.6>
      constraintSetup {angularLimitsMax=<1 3 25>, angularLimitsMin=<-1 -3 0>, angularTarget=<1 0 0>}
      localJointOffset=<1.55 0.1 0>
     }]
    relativeSpaceBone {boneName=root}
   }
   n149 = Anim Dynamics - Bone: dyn_loop_c(ComponentPose=@n148, Alpha=1, ExternalForce="0,0,0") {
    angularDampingOverride=0.9999
    angularSpringConstant=750
    bAngularSpring=true
    bGravityOverrideInSimSpace=true
    bOverrideAngularDamping=true
    simulationSpace=BoneRelative
    alphaBoolBlend {blendOption=Linear}
    boundBone {boneName=dyn_loop_c}
    componentAppliedLinearAccClamp=<5000 5000 5000>
    componentLinearAccScale=<1.2 1.2 1.2>
    componentLinearVelScale=<1.2 1.2 1.2>
    physicsBodyDefinitions [{
      sphereCollisionRadius=0
      boundBone {boneName=dyn_loop_c}
      boxExtents=<12.5 12.5 12.5>
      constraintSetup {angularLimitsMax=<1 3 25>, angularLimitsMin=<-1 -3 0>, angularTarget=<1 0 0>}
      localJointOffset=<1.5 0.1 0>
     }]
    relativeSpaceBone {boneName=root}
   }
   Save cached pose 'Dynamics'(Pose=Component To Local(ComponentPose=@n149)) {cacheName=Dynamics}
   Save cached pose 'Input'(Pose=Copy Pose From Mesh() {bUseAttachedParent=true}) {cacheName=Input}
  ]}
 EventGraph {nodes [
   n1 = Event BlueprintUpdateAnimation()
   > Set Delta Time X(Delta Time X:double=@n1.DeltaTimeX) -> (Output_Get:double)
   > n2 = Sequence!(execute) -> (then_0, then_1, then_2:Exec, then_3:Exec, then_4:Exec, then_5:Exec)
   > n3 = Gravity Override State Function DEPRECATED(CharacterPartAnimInstance=Self-Reference(), GameTimeInSeconds=GetGameTimeInSeconds(), GravityOverrideParams=Get Gravity Override Params())
   > Set GravityOverride(GravityOverride=@n3)
   n4 = PlayerMovement(execute=@n2.then_1, AnimInstance=Self-Reference(), XDotProductCoefficient=100, YDotProductCoefficient=100, ZDotProductCoefficient=100)
   n5 = BreakVector(InVec=MakeVector(X=@n4, Y=@n4.YMovement))
   > n6 = Set Left(Left:double="To Float (Boolean)"(InBool="float < float"(A=@n5, B=-50))) -> (Output_Get:double)
   n7 = "To Float (Boolean)"(InBool="float > float"(A=@n5, B=50))
   > n8 = Set Right(Right:double=@n7) -> (Output_Get:double)
   > n9 = Set Forward(Forward:double="To Float (Boolean)"(InBool="float > float"(A=@n5.Y, B=40))) -> (Output_Get:double)
   n10 = "To Float (Boolean)"(InBool="float < float"(A=@n5.Y, B=-40))
   > n11 = Set Backward(Backward:double=@n10) -> (Output_Get:double)
   n12 = Get bIsCrouching()
   > Set Crouch_Left(Crouch_Left:double=SelectFloat(A=@n6, B=0, bPickA=@n12)) -> (Output_Get:double)
   > Set Crouch_RIght(Crouch_RIght:double=SelectFloat(A=@n8, B=0, bPickA=@n12)) -> (Output_Get:double)
   > Set Crouch_Forward(Crouch_Forward:double=SelectFloat(A=@n9, B=0, bPickA=@n12)) -> (Output_Get:double)
   > Set Crouch_Backward(Crouch_Backward:double=SelectFloat(A=@n11, B=0, bPickA=@n12)) -> (Output_Get:double)
   > Set BackRight(BackRight:double=float + float(A=@n7, B=@n10)) -> (Output_Get:double)
   Set Is Wearing Backbling(execute=@n2.then_2, Is Wearing Backbling=CheckForBackbling())
   n13 = NOT Boolean(A=Get bIsSkyDiving())
   n14 = OR Boolean!(A=Get bIsJumping(), B=Get bIsFalling(), C:bool) -> (ReturnValue)
   Set Jumping(execute=@n2.then_3, Jumping:double=SelectFloat(A=1, B=0, bPickA=AND Boolean(A=@n14, B=@n13))) -> (Output_Get:double)
   n15 = Get bIsMoving2D()
   Set Crouch_Idle(execute=@n2.then_4, Crouch_Idle:bool=AND Boolean(A=Get bIsCrouching(), B=NOT Boolean(A=@n15))) -> (Output_Get:bool)
   > Set Swing(Swing:double=SelectFloat(A=1, B=0, bPickA=Get bIsPlayingMeleeAnim())) -> (Output_Get:double)
   n16 = NOT Boolean(A=Get bIsSkyDiving())
   n17 = OR Boolean(A=Get bIsJumping(), B=Get bIsFalling())
   > Set Tactial Jump(Tactial Jump:bool=AND Boolean(A=Get bIsTacticalSprinting(), B=AND Boolean(A=@n17, B=@n16))) -> (Output_Get:bool)
   n18 = Get bIsPlayingEmote()
   > n19 = Branch(Condition=AND Boolean(A=Get bIsInFrontEnd(), B=NOT Boolean(A=@n18)))
   > Set Front End Off(Front End Off:bool=true) -> (Output_Get:bool)
   Set Front End Off(execute=@n19.else, Front End Off=false)
  ]}
]
variables [
 BackRight {type=number}
 Backward {type=number}
 Crouch_Backward {type=number}
 Crouch_Forward {type=number}
 Crouch_Idle {type=boolean, default=true}
 Crouch_Left {type=number}
 Crouch_RIght {type=number}
 Delta Time X {type=number}
 Forward {type=number}
 Front End Off {type=boolean}
 Gravity Override Params {type=GravityOverrideParamsStruct}
 GravityOverride {type=Vector}
 Is Wearing Backbling {type=boolean}
 Jumping {type=number}
 Left {type=number}
 Right {type=number}
 Swing {type=number}
 Tactial Jump {type=boolean}
]
