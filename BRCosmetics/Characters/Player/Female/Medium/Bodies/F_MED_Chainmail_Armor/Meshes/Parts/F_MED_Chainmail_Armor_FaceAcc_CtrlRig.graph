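# Control Rig for the female medium Chainmail Armor face accessory that applies pose-driven corrective offsets to the back and right-middle hair bones.
# Spherical pose readers measure head orientation relative to neck_01, and their outputs weight Modify Transforms nodes on each chain's first bone.
# The base correction always runs.
# When Crouch is true a second reader layers on further offsets, and MOving, checked only within that branch, adds a third set.
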ControlRigBlueprint /BRCosmetics/Characters/Player/Female/Medium/Bodies/F_MED_Chainmail_Armor/Meshes/Parts/F_MED_Chainmail_Armor_FaceAcc_CtrlRig
graphs [
 RigVMFunctionLibrary {links [], nodes []}
 RigVMModel {nodes [
   RigUnit_BeginExecution = &RigUnit_BeginExecution()
   SequenceAggregate = &RigVMFunction_Sequence(ExecuteContext=@RigUnit_BeginExecution.ExecutePin)
   n1 = &RigUnit_SphericalPoseReader(
    ActiveRegionScaleFactors=(PositiveWidth=0.7,NegativeWidth=0.2,PositiveHeight=0.7,NegativeHeight=0.5)
    ActiveRegionSize=0.2
    Debug=(bDrawDebug=false,DebugScale=25,DebugSegments=20,DebugThickness=0.25)
    DriverAxis=(X=1,Y=0,Z=0)
    DriverItem=(Type=Bone,Name="head")
    FalloffRegionScaleFactors=(PositiveWidth=0.4,NegativeWidth=0.1,PositiveHeight=0.4,NegativeHeight=0.2)
    FalloffSize=0.45
    OptionalParentItem=(Type=Bone,Name="neck_01")
    RotationOffset=(X=-90,Y=0,Z=45)
    ExecutePin=@SequenceAggregate.A
   )
   > &RigUnit_ModifyTransforms(
    ItemToModify=((Item=(Type=Bone,Name="dyn_hair_bk_01"),Transform=(Rotation=(X=0,Y=0,Z=0,W=1),Translation=(X=0,Y=0,Z=0),Scale3D=(X=1,Y=1,Z=1))))
    ItemToModify.0.Transform.Rotation=&RigVMFunction_MathQuaternionFromEuler(Euler=(X=0,Y=-30,Z=0)).Result
    Weight=@n1.OutputParam
   )
   > &RigUnit_ModifyTransforms(
    ItemToModify=((Item=(Type=Bone,Name="dyn_hair_bk_l_01"),Transform=(Rotation=(X=0,Y=0,Z=0,W=1),Translation=(X=0,Y=0,Z=0),Scale3D=(X=1,Y=1,Z=1))))
    ItemToModify.0.Transform.Rotation=&RigVMFunction_MathQuaternionFromEuler(Euler=(X=0,Y=-20,Z=-5)).Result
    Weight=@n1.OutputParam
   )
   > &RigUnit_ModifyTransforms(
    ItemToModify=((Item=(Type=Bone,Name="dyn_hair_bk_r_01"),Transform=(Rotation=(X=0,Y=0,Z=0,W=1),Translation=(X=-2,Y=0,Z=0),Scale3D=(X=1,Y=1,Z=1))))
    ItemToModify.0.Transform.Rotation=&RigVMFunction_MathQuaternionFromEuler(Euler=(X=0,Y=-50,Z=5)).Result
    Weight=@n1.OutputParam
   )
   > &RigUnit_ModifyTransforms(
    ItemToModify=((Item=(Type=Bone,Name="dyn_hair_mid_r_01"),Transform=(Rotation=(X=0,Y=0,Z=0,W=1),Translation=(X=-3,Y=0,Z=0),Scale3D=(X=1,Y=1,Z=1))))
    ItemToModify.0.Transform.Rotation=&RigVMFunction_MathQuaternionFromEuler(Euler=(X=0,Y=-40,Z=30)).Result
    Weight=@n1.OutputParam
   )
   Branch_RigVMUnitNode = &RigVMFunction_ControlFlowBranch(ExecuteContext=@SequenceAggregate.B, Condition=RigVMVariableNode() {variable=Crouch}.Value)
   n2 = &RigUnit_SphericalPoseReader(
    ActiveRegionScaleFactors=(PositiveWidth=0.7,NegativeWidth=0.2,PositiveHeight=0.7,NegativeHeight=0.5)
    ActiveRegionSize=0.2
    Debug=(bDrawDebug=false,DebugScale=25,DebugSegments=20,DebugThickness=0.25)
    DriverAxis=(X=1,Y=0,Z=0)
    DriverItem=(Type=Bone,Name="head")
    FalloffRegionScaleFactors=(PositiveWidth=0.2,NegativeWidth=0.1,PositiveHeight=0.2,NegativeHeight=0.2)
    FalloffSize=0.45
    OptionalParentItem=(Type=Bone,Name="neck_01")
    RotationOffset=(X=-90,Y=0,Z=40)
    ExecutePin=@Branch_RigVMUnitNode.True
   )
   > &RigUnit_ModifyTransforms(
    ItemToModify=((Item=(Type=Bone,Name="dyn_hair_bk_l_01"),Transform=(Rotation=(X=0,Y=0,Z=0,W=1),Translation=(X=0,Y=0,Z=0),Scale3D=(X=1,Y=1,Z=1))))
    ItemToModify.0.Transform.Rotation=&RigVMFunction_MathQuaternionFromEuler(Euler=(X=0,Y=-25,Z=-20)).Result
    Weight=@n2.OutputParam
   )
   > &RigUnit_ModifyTransforms(
    ItemToModify=((Item=(Type=Bone,Name="dyn_hair_bk_01"),Transform=(Rotation=(X=0,Y=0,Z=0,W=1),Translation=(X=0,Y=0,Z=0),Scale3D=(X=1,Y=1,Z=1))))
    ItemToModify.0.Transform.Rotation=&RigVMFunction_MathQuaternionFromEuler(Euler=(X=0,Y=-45,Z=5)).Result
    Weight=@n2.OutputParam
   )
   > &RigUnit_ModifyTransforms(
    ItemToModify=((Item=(Type=Bone,Name="dyn_hair_bk_r_01"),Transform=(Rotation=(X=0,Y=0,Z=0,W=1),Translation=(X=2,Y=0,Z=0),Scale3D=(X=1,Y=1,Z=1))))
    ItemToModify.0.Transform.Rotation=&RigVMFunction_MathQuaternionFromEuler(Euler=(X=0,Y=0,Z=0)).Result
    Weight=@n2.OutputParam
   )
   > n3 = &RigUnit_ModifyTransforms(
    ItemToModify=((Item=(Type=Bone,Name="dyn_hair_mid_r_01"),Transform=(Rotation=(X=0,Y=0,Z=0,W=1),Translation=(X=0,Y=0,Z=0),Scale3D=(X=1,Y=1,Z=1))))
    ItemToModify.0.Transform.Rotation=&RigVMFunction_MathQuaternionFromEuler(Euler=(X=0,Y=0,Z=0)).Result
    Weight=@n2.OutputParam
   )
   Branch_1_RigVMUnitNode = &RigVMFunction_ControlFlowBranch(ExecuteContext=@n3.ExecutePin, Condition=RigVMVariableNode() {variable=MOving}.Value)
   n4 = &RigUnit_SphericalPoseReader(
    ActiveRegionScaleFactors=(PositiveWidth=0.6,NegativeWidth=0.2,PositiveHeight=1,NegativeHeight=0.5)
    ActiveRegionSize=0.2
    Debug=(bDrawDebug=false,DebugScale=25,DebugSegments=20,DebugThickness=0.25)
    DriverAxis=(X=1,Y=0,Z=0)
    DriverItem=(Type=Bone,Name="head")
    FalloffRegionScaleFactors=(PositiveWidth=0.4,NegativeWidth=0.1,PositiveHeight=0.4,NegativeHeight=0.2)
    FalloffSize=0.45
    OptionalParentItem=(Type=Bone,Name="neck_01")
    RotationOffset=(X=-90,Y=0,Z=40)
    ExecutePin=@Branch_1_RigVMUnitNode.True
   )
   > &RigUnit_ModifyTransforms(
    ItemToModify=((Item=(Type=Bone,Name="dyn_hair_bk_l_01"),Transform=(Rotation=(X=0,Y=0,Z=0,W=1),Translation=(X=0,Y=0,Z=0),Scale3D=(X=1,Y=1,Z=1))))
    ItemToModify.0.Transform.Rotation=&RigVMFunction_MathQuaternionFromEuler(Euler=(X=0,Y=-10,Z=0)).Result
    Weight=@n4.OutputParam
   )
   > &RigUnit_ModifyTransforms(
    ItemToModify=((Item=(Type=Bone,Name="dyn_hair_bk_01"),Transform=(Rotation=(X=0,Y=0,Z=0,W=1),Translation=(X=0,Y=0,Z=2),Scale3D=(X=1,Y=1,Z=1))))
    ItemToModify.0.Transform.Rotation=&RigVMFunction_MathQuaternionFromEuler(Euler=(X=0,Y=5,Z=0)).Result
    Weight=@n4.OutputParam
   )
   > &RigUnit_ModifyTransforms(
    ItemToModify=((Item=(Type=Bone,Name="dyn_hair_bk_r_01"),Transform=(Rotation=(X=0,Y=0,Z=0,W=1),Translation=(X=0,Y=2,Z=5),Scale3D=(X=1,Y=1,Z=1))))
    ItemToModify.0.Transform.Rotation=&RigVMFunction_MathQuaternionFromEuler(Euler=(X=0,Y=60,Z=-20)).Result
    Weight=@n4.OutputParam
   )
   > &RigUnit_ModifyTransforms(
    ItemToModify=((Item=(Type=Bone,Name="dyn_hair_mid_r_01"),Transform=(Rotation=(X=0,Y=0,Z=0,W=1),Translation=(X=3,Y=0,Z=2),Scale3D=(X=1,Y=1,Z=1))))
    ItemToModify.0.Transform.Rotation=&RigVMFunction_MathQuaternionFromEuler(Euler=(X=0,Y=100,Z=0)).Result
    Weight=@n4.OutputParam
   )
  ]}
]
hierarchy [
 root {loc=<0 0 0>, rot=<0 0 0 1>}
 pelvis {parent=root, loc=<0 -0.5301 98.0571>, rot=<0.0127 0.707 -0.0127 -0.707>}
 spine_01 {parent=pelvis, loc=<6.1208 -0.9328 0>, rot=<0 0 0.034 -0.9994>}
 spine_02 {parent=spine_01, loc=<6.5996 -0.2095 0>, rot=<0 0 0.0238 -0.9997>}
 spine_03 {parent=spine_02, loc=<6.4914 -0.1237 0>, rot=<0 0 0.0149 -0.9999>}
 spine_04 {parent=spine_03, loc=<6.9598 0.0794 0>, rot=<0 0 -0.0132 -0.9999>}
 spine_05 {parent=spine_04, loc=<6.9987 0 0>, rot=<0 0 -0.0015 -1>}
 clavicle_l {parent=spine_05, loc=<5.4831 -0.1053 -3.2946>, rot=<0.0465 -0.7289 -0.009 -0.6829>}
 neck_01 {parent=spine_05, loc=<11.3138 0.1529 0>, rot=<0 0 -0.2003 -0.9797>}
 neck_02 {parent=neck_01, loc=<4.5627 0 0>, rot=<0 0 0.0504 -0.9987>}
 head {parent=neck_02, loc=<5.3914 0 0>, rot=<0 0 0.115 -0.9934>}
 dyn_knot {parent=head, loc=<12.8975 -9.141 -0.0509>, rot=<0.4996 -0.4918 0.5002 -0.5082>}
 dyn_string_l_01 {parent=dyn_knot, loc=<0.2558 0.5447 -0.2634>, rot=<0.576 -0.3833 -0.4062 -0.597>}
 dyn_string_l_02 {parent=dyn_string_l_01, loc=<2.8695 0 0.0822>, rot=<0.0001 -0.0003 -0.1176 -0.9931>}
 dyn_string_l_03 {parent=dyn_string_l_02, loc=<4.5049 0 0>, rot=<-0.0277 -0.0014 0.0819 -0.9963>}
 dyn_string_r_01 {parent=dyn_knot, loc=<0.3181 -0.609 -0.2421>, rot=<0.5882 -0.4081 -0.3981 -0.5736>}
 dyn_string_r_02 {parent=dyn_string_r_01, loc=<3.1809 -0.0056 -0.2126>, rot=<0.009 -0.0355 -0.1161 -0.9926>}
 dyn_string_r_03 {parent=dyn_string_r_02, loc=<4.2803 0 0>, rot=<0.0575 0.0305 0.1021 -0.9926>}
 dyn_loop_l_01 {parent=dyn_knot, loc=<-0.0351 0.5314 0.1811>, rot=<0.412 -0.191 -0.4365 -0.7767>}
 dyn_loop_l_02 {parent=dyn_loop_l_01, loc=<1.5268 0.1795 -0.0279>, rot=<0.104 -0.1802 -0.1558 -0.9656>}
 dyn_loop_l_03 {parent=dyn_loop_l_02, loc=<3.5064 0 0>, rot=<-0.0074 0.0008 -0.1276 -0.9918>}
 dyn_loop_r_01 {parent=dyn_knot, loc=<0.0675 -0.4907 0.211>, rot=<0.7369 -0.4832 -0.188 -0.4338>}
 dyn_loop_r_02 {parent=dyn_loop_r_01, loc=<1.7376 0.1819 0.0253>, rot=<-0.0949 0.1868 -0.1409 -0.9676>}
 dyn_loop_r_03 {parent=dyn_loop_r_02, loc=<3.8697 0 0>, rot=<0.0293 -0.0239 -0.0987 -0.9944>}
 dyn_hair_fr_r_01 {parent=head, loc=<0.8755 -1.3101 5.7883>, rot=<-0.0078 -0.1579 0.9872 -0.021>}
 dyn_hair_fr_r_02 {parent=dyn_hair_fr_r_01, loc=<3.9417 0.005 -0.0006>, rot=<0.0006 -0.0106 0.0302 -0.9995>}
 dyn_hair_fr_r_03 {parent=dyn_hair_fr_r_02, loc=<2.5284 0 0>, rot=<0.0182 -0.1557 0.0709 -0.9851>}
 dyn_hair_mid_r_01 {parent=head, loc=<-0.2912 -4.4145 5.6768>, rot=<0.0568 -0.2132 0.9376 -0.2686>}
 dyn_hair_mid_r_02 {parent=dyn_hair_mid_r_01, loc=<2.0828 0 0>, rot=<-0.013 0.0479 0.0747 -0.996>}
 dyn_hair_mid_r_03 {parent=dyn_hair_mid_r_02, loc=<3.7555 0 0>, rot=<0.0362 -0.0787 0.0919 -0.992>}
 dyn_hair_bk_r_01 {parent=head, loc=<-0.6136 -6.0548 3.6285>, rot=<0.1829 -0.5798 0.769 -0.1974>}
 dyn_hair_bk_r_02 {parent=dyn_hair_bk_r_01, loc=<4.0458 -0.0449 0.1244>, rot=<-0.0581 0.1318 -0.1794 -0.9732>}
 dyn_hair_bk_r_03 {parent=dyn_hair_bk_r_02, loc=<2.8778 -2.6019 -0.0635>, rot=<0.0657 -0.0162 0.3325 -0.9407>}
 dyn_hair_bk_01 {parent=head, loc=<1.8648 -7.486 -0.1489>, rot=<0.1161 -0.654 0.7391 -0.1118>}
 dyn_hair_bk_02 {parent=dyn_hair_bk_01, loc=<2.9596 0 0>, rot=<0.0009 -0.051 -0.1513 -0.9872>}
 dyn_hair_bk_03 {parent=dyn_hair_bk_02, loc=<3.7533 -0.1557 0.001>, rot=<-0.0009 -0.0076 0.0165 -0.9998>}
 dyn_hair_bk_l_01 {parent=head, loc=<0.411 -5.4178 -5.2032>, rot=<0.1029 -0.9067 0.4023 -0.0737>}
 dyn_hair_bk_l_02 {parent=dyn_hair_bk_l_01, loc=<3.5618 0 0>, rot=<-0.041 0.1207 0.0287 -0.9914>}
 dyn_hair_bk_l_03 {parent=dyn_hair_bk_l_02, loc=<2.6376 0 0>, rot=<0.0563 -0.1472 0.0086 -0.9875>}
 dyn_hair_fr_l_01 {parent=head, loc=<0.0935 -1.829 -6.6985>, rot=<-0.1778 -0.9781 -0.0937 0.054>}
 dyn_hair_fr_l_02 {parent=dyn_hair_fr_l_01, loc=<3.0629 0 0>, rot=<-0.0894 -0.5597 0.1934 -0.8009>}
 dyn_hair_fr_l_03 {parent=dyn_hair_fr_l_02, loc=<0.8206 1.004 2.3804>, rot=<0.0392 0.2643 -0.2115 -0.9401>}
 clavicle_r {parent=spine_05, loc=<5.4831 -0.1053 3.2946>, rot=<0.7289 0.0465 -0.6829 0.009>}
]
variables [
 Crouch {type=bool, default=False}
 MOving {type=bool, default=False}
]
